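# Masked material for the SeasonPass Reboot custom character background, built on the shared Vault World UI background functions.
# A screen-space texture or fallback two-color radial gradient forms the base visuals.
# Fake fog tinted by a scene MPC color and an optional shop color override adjust the color.
# A radially faded grid of rotated rounded tile outlines multiplies the result, which feeds Vault World calibration and transition functions.
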Material /SeasonPass/Widgets/Reboot/Materials/M_UI_CustomCharacter_Background
graph {
 properties {bAutomaticallySetUsageInEditor=false, blendMode=BLEND_Masked, usedShadingModels=Default Lit}
 nodes [
  n1 = NamedRerouteUsage()
  n2 = CollectionParameter() {collection=/Game/UI/Foundation/Environment/Shell/Materials/M_UI_SceneBackgroundMCP, parameterName=RGBA1}
  n3 = LinearInterpolate(
   A=Divide(A=NamedRerouteUsage()) {constB=0.3}
   B=Multiply(A=ComponentMask(None=@n2) {b=true, g=true, r=true}) {constB=0.7}
   Alpha=NamedRerouteUsage()
  )
  n4 = TextureSample() {
   samplerSource=SSM_Clamp_WorldGroupSettings
   samplerType=SAMPLERTYPE_Alpha
   texture=/Game/UI/Foundation/Environment/Shell/Materials/Textures/T_UI_Vaultworld_Background_Premium_FakeFog
  }
  n5 = RayTracingQualitySwitch(
   Normal=Multiply(
    A=NamedRerouteDeclaration(None=@n4.R) {name=FakeFogMask, nodeColor=<rgba 0 0 0 1>}
    B=ScalarParameter() {defaultValue=1, group=Vehicle Settings, parameterName=Display Fog, sliderMax=1}
   )
   RayTraced=Constant()
  )
  n6 = VectorParameter() {
   desc=Alpha is used to Lerp OverrideValue
   group=Shop Settings
   parameterName=BaseColorOverride
   channelNames {a=UseOverride?, b="", g="", r=""}
  }
  n7 = ScalarParameter() {group=Shop Settings, parameterName=IsUsingShopTexture}
  n8 = MaterialFunctionCall!(RotationAmount=ScalarParameter() {defaultValue=0.048, group=Floor Tiles, parameterName=TileRotation}, Coords, Pivot) {materialFunction=/BaseMaterial/Materials/Functions/MF_Rotate2D}
  n9 = ScalarParameter() {defaultValue=85, group=Floor Tiles, parameterName=TilingAmount}
  n10 = MaterialFunctionCall(
   UVs=Frac(None=Multiply(A=@n8.Result, B=@n9) {constB=160, desc=Tiling})
   Size=ScalarParameter() {defaultValue=0.9, group=Floor Tiles, parameterName=TileSeparationSize}
   CornerRoundness=ScalarParameter() {defaultValue=0.2, group=Floor Tiles, parameterName=TileCornerRoundness}
   OutlineThickness=ScalarParameter() {defaultValue=0.035, group=Floor Tiles, parameterName=TileTickness}
  ) {materialFunction=&MF_UI_Box}
  n11 = ScalarParameter() {group=Floor Tiles, parameterName=FloorThreshold}
  n12 = ScalarParameter() {defaultValue=1.9, group=Floor Tiles, parameterName=FloorSmootthing}
  n13 = Length(None=Add(A=TextureCoordinate()) {constB=-0.5}) {is=class MaterialExpressionLength}
  n14 = ScalarParameter() {defaultValue=11, group=Floor Tiles, parameterName=MaskSIze}
  n15 = Saturate(None=OneMinus(None=Multiply(A=@n13, B=@n14) {constB=5}))
  n16 = Multiply(A=@n15, B=@n15)
  n17 = Saturate(None=Multiply(
    A=MaterialFunctionCall(DistanceField=@n10.Outline, Threshold=@n11, Smoothing=@n12) {materialFunction=&MF_UI_AntiAliasing}.Outer
    B=Multiply(A=@n16, B=@n16)
   ))
  n18 = Multiply(
   A=LinearInterpolate(
    A=LinearInterpolate(A=@n1, B=@n3, Alpha=@n5)
    B=ComponentMask(None=@n6) {b=true, g=true, r=true}
    Alpha=LinearInterpolate(A=@n6.UseOverride?, Alpha=@n7) {constB=0}
   )
   B=Add(A=Multiply(A=@n17) {desc=opacity})
  )
  n19 = MaterialFunctionCall!(Vault World Color=@n18) {materialFunction=/Game/Environments/FrontEnd/Blueprints/ItemPreview/VaultWorldTransitions/MaterialFunctions/MF_UI_VaultWorld_PostProcessCalibration}
  MaterialFunctionCall!(Vault World Color=LinearInterpolate(A=@n19.Calibrated Color, B=@n18, Alpha=NamedRerouteUsage())) {materialFunction=/Game/Environments/FrontEnd/Blueprints/ItemPreview/VaultWorldTransitions/MaterialFunctions/MF_UI_VaultWorld_Transition}
  MaterialFunctionCall() {materialFunction=/FortUILibrary/Materials/Background/DefaultFortnite/MF_UI_Background_DefaultFortnite}
  n20 = MaterialFunctionCall!(Use Local UV=Constant(), Texture Size, Scale Pivot Point) {materialFunction=/Game/Environments/FrontEnd/Blueprints/ItemPreview/VaultWorldTransitions/MaterialFunctions/MF_UI_VaultWorld_UV}
  NamedRerouteDeclaration(None=@n20.UV) {name=ScreenUV, nodeColor=<rgba 0 0 0 1>}
  n21 = TextureSampleParameter2D(UVs=ScreenPosition().ViewportUV) {
   group=Background Settings
   parameterName=BackgroundTexture
   samplerSource=SSM_Wrap_WorldGroupSettings
   texture=/Game/Athena/UI/Frontend/Spatial/Art/T_Ui_Empty
  }
  n22 = Multiply(A=NamedRerouteUsage(), B=NamedRerouteUsage())
  n23 = Subtract(A=NamedRerouteUsage(), B=AppendVector(A=Constant() {r=0.65}, B=Constant() {r=0.35}))
  n24 = Constant() {r=2.5}
  n25 = Constant() {r=1.5}
  n26 = ComponentMask(None=LinearInterpolate(
    A=VectorParameter() {group=Background colors, parameterName=Color2, defaultValue=<rgba 0 0.0052 0.0056 1>}.RGB
    B=VectorParameter() {group=Background colors, parameterName=Color1, defaultValue=<rgba 0.0222 0.4621 0.2346 1>}.RGB
    Alpha=MaterialFunctionCall(UVs=@n23, Radius=@n24, Density=@n25) {materialFunction=&RadialGradientExponential}.RadialGradientExponential
   )) {b=true, g=true, r=true}
  n27 = ScalarParameter() {defaultValue=1, group=Background Settings, parameterName=UseFallbackGradient}
  NamedRerouteDeclaration(None=LinearInterpolate(A=@n21.RGB, B=Multiply(A=@n22, B=@n26), Alpha=NamedRerouteDeclaration(None=@n27) {name=Use Fallback, nodeColor=<rgba 0 0 0 1>})) {name=Background Visuals, nodeColor=<rgba 0 0 0 1>}
  n28 = NamedRerouteUsage()
  n29 = OneMinus(None=@n28)
  n30 = Multiply(A=Add(A=ComponentMask(None=@n29) {g=true}) {constB=0.125}) {constB=1.5}
  n31 = Add(A=ComponentMask(None=@n28) {g=true}) {constB=0.15}
  NamedRerouteDeclaration(None=Multiply(A=Saturate(None=@n30), B=Saturate(None=@n31))) {name=TopGradient, nodeColor=<rgba 0 0 0 1>}
  n32 = Add(A=@n28) {constB=-0.5}
  n33 = OneMinus(None=ComponentMask(None=Abs(None=@n32)) {r=true})
  NamedRerouteDeclaration(None=Saturate(None=Add(A=@n33) {constB=0.025})) {name=MidGradient, nodeColor=<rgba 0 0 0 1>}
  NamedRerouteDeclaration(None=@n20.Scale Y) {name=ScaleY, nodeColor=<rgba 0 0 0 1>}
 ]
}
