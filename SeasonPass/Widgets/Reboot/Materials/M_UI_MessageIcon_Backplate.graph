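# UI material that draws the backplate behind a message icon in the Season Pass Reboot widgets, as a pale gold, slightly translucent shape.
# A rotated box SDF forms a diamond, and an offset, rotated, scaled triangle SDF is built alongside it.
# SDF subtraction combines the two, then anti-aliasing yields the final mask, which is tinted, multiplied by vertex alpha and alpha-composited.
# Scalar parameters control the placement and proportions of both shapes.
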Material /SeasonPass/Widgets/Reboot/Materials/M_UI_MessageIcon_Backplate
graph {
 properties {bAutomaticallySetUsageInEditor=false, blendMode=BLEND_AlphaComposite, materialDomain=MD_UI, usedShadingModels=Default Lit}
 nodes [
  n1 = NamedRerouteUsage()
  Multiply(A=VertexColor().A, B=Multiply(A=Constant3Vector() {constant=<rgba 1 0.8714 0.3813 1>}, B=@n1))
  Multiply(A=@n1, B=Constant() {r=0.8})
  n2 = MaterialFunctionCall() {materialFunction=&MF_UI_UVScaling}
  n3 = ScalarParameter() {defaultValue=0.125, parameterName=Diamond Mask Rotation}
  n4 = ScalarParameter() {defaultValue=0.7, parameterName=Backplate Size X}
  n5 = ScalarParameter() {defaultValue=0.7, parameterName=Backplate Size Y}
  NamedRerouteDeclaration(None=MaterialFunctionCall(
    UVs=MaterialFunctionCall(UVs=@n2.Scaled to Fit, "Rotation Angle (0-1)"=@n3) {materialFunction=&CustomRotator}.Rotated Values
    Size=AppendVector(A=@n4, B=@n5)
   ) {materialFunction=&MF_UI_Box}.SDF) {name=DiamondShape, nodeColor=<rgba 0 0.9766 1 1>}
  n6 = TextureCoordinate()
  n7 = ScalarParameter() {defaultValue=0.202, parameterName=Position X}
  n8 = ScalarParameter() {defaultValue=-0.628, parameterName=Position Y}
  n9 = MaterialFunctionCall(
   UVs=Add(
    A=MaterialFunctionCall(UV=@n6) {materialFunction=&MF_UI_UVScaling}.Scaled to Fit
    B=AppendVector(A=@n7, B=@n8)
   )
   "Rotation Angle (0-1)"=ScalarParameter() {defaultValue=0.125, parameterName=Triangle Rotation}
  ) {materialFunction=&CustomRotator}
  n10 = AppendVector(A=ScalarParameter() {defaultValue=1.0574, parameterName=Scale X}, B=ScalarParameter() {defaultValue=0.816, parameterName=Scale Y})
  NamedRerouteDeclaration(None=MaterialFunctionCall!(
    UVs=MaterialFunctionCall(UVs=@n9.Rotated Values, Texture Scale=@n10) {materialFunction=&ScaleUVsByCenter}.UVs
    Corner 01
    Corner 02
    Corner 03
    Glow Min
    Glow Max
    Outline Thickness
   ) {materialFunction=/FortUILibrary/MaterialFunctions/Shapes/MF_UI_WIP_Triangle}.SDF) {name=TriangleMask, nodeColor=<rgba 1 0.8203 0 1>}
  n11 = NamedRerouteUsage()
  n12 = NamedRerouteUsage()
  NamedRerouteDeclaration(None=MaterialFunctionCall(
    DistanceField=MaterialFunctionCall!(DistanceFieldA=@n11, DistanceFieldB=@n12, BlendAmount) {materialFunction=/FortUILibrary/MaterialFunctions/Shapes/Modifiers/MF_UI_SDF_Subtraction}.Exact
    Smoothing=Constant() {r=0.01}
   ) {materialFunction=&MF_UI_AntiAliasing}.Inner) {name=Final Backplate Shape, nodeColor=<rgba 1 0 0.2109 1>}
 ]
}
