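# Season Pass disclaimer widget showing a title and body text beside an icon on a tilted backplate.
# SetDisclamerData fills both texts, then SetIcon and SetBackground blocking-load soft references, accepting material instances or textures with default-material fallbacks.
# PreConstruct runs SetStatusIndicator with a fixed type, which looks up the StatusIndicators map to toggle Background and Icon visibility, set materials and tint by indicator type.
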WidgetBlueprint /SeasonPass/Widgets/Reboot/WBP_ItemDisclaimer
graphs [
 EventGraph {nodes [
   n1 = SetBackground!() -> (OutputDelegate, then, Disclaimer Bckground Icon:Soft Object Reference) {is=title K2Node_CustomEvent}
   > n2 = LoadAsset_Blocking(Asset=@n1.Disclaimer Bckground Icon)
   > n3 = Branch(Condition=IsValid(Object=@n2))
   > n4 = Cast To MaterialInstance!(execute, Object=@n2) -> (then, CastFailed, AsMaterial Instance:Material Instance Object Reference) {is=title K2Node_DynamicCast}
   > SetBrushFromMaterial(self=Get IconBackground() -> (IconBackground:Image Object Reference), Material=@n4)
   n5 = Cast To Texture2D(execute=@n4.CastFailed, Object=@n2)
   > SetBrushFromTexture!(execute, self:Image Object Reference=Get IconBackground(), Texture:Texture 2D Object Reference=@n5, bMatchSize:bool=false) -> (then)
   SetBrushFromSoftMaterial!(
    execute=[@n5.CastFailed, @n3.else]
    self:Image Object Reference=Get IconBackground()
    SoftMaterial:Material Interface Soft Object Reference=/SeasonPass/Widgets/Reboot/Materials/MI_UI_MessageIcon_Backplate
   ) -> (then)
   n6 = SetIcon!() -> (OutputDelegate, then, Disclaimer Icon:Soft Object Reference) {is=title K2Node_CustomEvent}
   > n7 = LoadAsset_Blocking(Asset=@n6.Disclaimer Icon)
   > n8 = Branch(Condition=IsValid(Object=@n7))
   > n9 = Cast To MaterialInstance(Object=@n7)
   > SetBrushFromMaterial(self=Get Icon() -> (Icon:Image Object Reference), Material=@n9)
   n10 = Cast To Texture2D(execute=@n9.CastFailed, Object=@n7)
   > SetBrushFromTexture(self=Get Icon(), Texture=@n10, bMatchSize=false)
   SetBrushFromSoftMaterial(execute=[@n10.CastFailed, @n8.else], self=Get Icon(), SoftMaterial=/FortUILibrary/Textures/MaterialInstances/MI_UI_Icon_Customizable)
   Event PreConstruct()
   > SetStatusIndicator!(execute, self, IndicatorType:E_UIKit_ItemCardStatusIndicator Enum=NewEnumerator12) -> (then)
  ]}
 SetDisclamerData {nodes [
   n1 = SetDisclamerData!() -> (then, DisclamerData:Season Pass Element Disclaimer Data Structure) {is=title K2Node_FunctionEntry}
   n2 = Break Season Pass Element Disclaimer Data!(SeasonPassElementDisclaimerData:"Season Pass Element Disclaimer Data Structure (by ref)"=@n1) -> (Header:Text, Text:Text, Icon:Soft Object Reference, Background:Soft Object Reference) {is=title K2Node_BreakStruct}
   > SetHeader!(execute, self, InText:Text=@n2) -> (then)
   > SetText(self:Self Object Reference, InText:Text=@n2.Text)
   > SetIcon!(execute, self, Disclaimer Icon:Soft Object Reference=@n2.Icon) -> (then) {is=title K2Node_CallFunction}
   > SetBackground!(execute, self, Disclaimer Bckground Icon:Soft Object Reference=@n2.Background) -> (then) {is=title K2Node_CallFunction}
  ]}
 SetHeader {nodes [
   n1 = SetHeader!() -> (then, InText:Text) {is=title K2Node_FunctionEntry}
   > "SetText (Text)"(self=Get Text_Title() -> (Text_Title:Common Text Object Reference), InText=@n1)
  ]}
 SetStatusIndicator {nodes [
   n1 = SetStatusIndicator!() -> (then, IndicatorType:E_UIKit_ItemCardStatusIndicator Enum=NewEnumerator10) {is=title K2Node_FunctionEntry}
   n2 = Find(
    TargetMap:Map of E_UIKit_ItemCardStatusIndicator Enums to S UIKit Status Indicator Spawn Params Structures=Get StatusIndicators() -> (StatusIndicators:Map of E_UIKit_ItemCardStatusIndicator Enums to S UIKit Status Indicator Spawn Params Structures=NewEnumerator10)
    Key:"E_UIKit_ItemCardStatusIndicator Enum (by ref)"=@n1
   ) -> (Value:S UIKit Status Indicator Spawn Params Structure)
   > SetVisibility(
    self=[Get Background() -> (Background:WBP UIKit Backplate Object Reference), Get Icon()]
    InVisibility=Select(Option 0=Collapsed, Option 1=SelfHitTestInvisible, Index=@n2.ReturnValue)
   )
   > Branch(Condition=@n2.ReturnValue)
   n3 = Break S UIKit Status Indicator Spawn Params!(S_UIKit_StatusIndicatorSpawnParams:"S UIKit Status Indicator Spawn Params Structure (by ref)"=@n2) -> (
    Configuration_2_#1:E_UI_StatusIndicator_Configuration Enum
    Size_9_#2:E_UI_StatusIndicator_Size Enum
    Type_11_#3:E_UI_StatusIndicator_Type Enum
    Icon_28_#4:Material Instance Soft Object Reference
    Placement_36_#5:E_UIKit_StatusIndicatorPlacement Enum
    OverrideColumn_39_#6:int
    Background_47_#7:Material Instance Soft Object Reference
   ) {is=title K2Node_BreakStruct}
   > SetBrushFromSoftMaterial(self=Get Icon(), SoftMaterial=@n3.Icon_28_#4)
   > n4 = Branch(Condition=IsValidSoftObjectReference(SoftObjectReference=@n3.Background_47_#7))
   > n5 = SetBrushFromSoftMaterial(self=Get IconBackground(), SoftMaterial=@n3.Background_47_#7)
   n6 = GetDynamicMaterial(execute=[@n5.then, @n4.else], self=Get IconBackground())
   > SetVectorParameterValue(
    self=@n6
    ParameterName=Color
    Value=Select!(
     NewEnumerator0:Linear Color Structure=(R=0.0395,G=1,B=0.2307,A=1)
     NewEnumerator2:Linear Color Structure=(R=0.9301,G=1,B=0.0103,A=1)
     NewEnumerator3:Linear Color Structure=(R=0.0048,G=0.004,B=0.0116,A=1)
     NewEnumerator4:Linear Color Structure=(R=1,G=0.147,B=0.1329,A=1)
     NewEnumerator5:Linear Color Structure=(R=1,G=0.8714,B=0.3813,A=1)
     Index:E_UI_StatusIndicator_Type Enum=@n3.Type_11_#3
    ) -> (ReturnValue:Linear Color Structure)
   )
   > SetColorAndOpacity(
    self:Image Object Reference=Get Icon()
    InColorAndOpacity:Linear Color Structure=Select(
     NewEnumerator0=(R=0,G=0.0331,B=0.0296,A=1)
     NewEnumerator2=(R=0.0048,G=0.004,B=0.0116,A=1)
     NewEnumerator3=(R=1,G=1,B=1,A=1)
     NewEnumerator4=(R=0.1413,G=0.0212,B=0.0194,A=1)
     NewEnumerator5=(R=0,G=0,B=0,A=1)
     Index=@n3.Type_11_#3
    )
   )
  ]}
 SetText {nodes [
   n1 = SetText!() -> (then, InText:Text) {is=title K2Node_FunctionEntry}
   > "SetText (Text)"(self=Get Text_Body() -> (Text_Body:Common Text Object Reference), InText=@n1)
  ]}
]
variables [StatusIndicators {default {
   Archived {
    type=Utility
    background=null
    configuration=Icon
    icon=/FortUILibrary/Textures/MaterialInstances/MI_UIKit_Icon_Archive
    overrideColumn=1
    placement=TopRight
    size=Default
   }
   Crew {
    type=Utility
    background=null
    configuration=Icon
    icon=/FortUILibrary/Textures/MaterialInstances/MI_UIKit_Icon_CrewLoan
    overrideColumn=2
    placement=TopLeft
    size=Compact
   }
   Customizable {
    type=Creative
    background=&MI_UI_MessageIcon_Backplate
    configuration=Icon
    icon=&MI_UI_Icon_Customizable
    overrideColumn=0
    placement=BottomLeft
    size=Default
   }
   EquippedInOtherCategory {
    type=Utility
    background=null
    configuration=Icon
    icon=/FortUILibrary/Textures/MaterialInstances/MI_UIKit_Icon_LobbyImage
    overrideColumn=0
    placement=TopRight
    size=Default
   }
   Favorite {
    type=Utility
    background=null
    configuration=Icon
    icon=/FortUILibrary/Textures/MaterialInstances/MI_UIKit_Icon_Favorite
    overrideColumn=1
    placement=TopRight
    size=Default
   }
   HasConflict {
    type=Utility
    background=null
    configuration=Icon
    icon=/FortUILibrary/Textures/MaterialInstances/MI_UIKit_Icon_Warning
    overrideColumn=0
    placement=Center
    size=Default
   }
   Juno {
    type=Utility
    background=null
    configuration=Icon
    icon=/FortUILibrary/Textures/MaterialInstances/MI_UIKit_Icon_JunoStyle
    overrideColumn=1
    placement=TopLeft
    size=Compact
   }
   Locked {
    type=Utility
    background=null
    configuration=Icon
    icon=/FortUILibrary/Textures/MaterialInstances/MI_UIKit_Icon_Locked
    overrideColumn=0
    placement=Center
    size=Default
   }
   New {
    type=Attention
    background=null
    configuration=Icon
    icon=/FortUILibrary/Textures/MaterialInstances/MI_UIKit_Icon_ExclamationMark
    overrideColumn=2
    placement=TopRight
    size=Compact
   }
   Shuffle {
    type=Utility
    background=null
    configuration=Icon
    icon=/FortUILibrary/Textures/MaterialInstances/MI_UIKit_Icon_Shuffle
    overrideColumn=0
    placement=BottomLeft
    size=Default
   }
   Temporary {
    type=Utility
    background=null
    configuration=Icon
    icon=/FortUILibrary/Textures/MaterialInstances/MI_UIKit_Icon_UpdatePending_Small
    overrideColumn=1
    placement=BottomLeft
    size=Default
   }
  }}]
widgets {
 info {inheritedWidgetCount=0, parentClass=&CommonUserWidget, rootWidgetClass=&GridPanel, widgetCount=8}
 widgets [
  Root {widgetClassPath=&GridPanel, widget=Root, properties {bOverride_Cursor=true, columnFill [0,1], rowFill [1]}}
  Background {
   widgetClassPath=&WBP_UIKit_Backplate
   bIsVariable=true
   parent=Root
   widget=Background
   properties {brightness=Dark}
   slot {class=GridSlot, properties {columnSpan=2, padding=<ltrb 50 0 0 0>}}
  }
  Spacer_Image {
   widgetClassPath=&Spacer
   parent=Root
   widget=Spacer_Image
   properties {size=<100 100>}
   slot {class=GridSlot, properties {columnSpan=1, layer=1, rowSpan=1}}
  }
  IconBackground {
   widgetClassPath=&Image
   bIsVariable=true
   parent=Root
   widget=IconBackground
   properties {
    visibility=HitTestInvisible
    brush {
     bIsDynamicallyLoaded=false
     drawAs=Image
     imageType=FullColor
     mirroring=NoMirror
     resourceName=None
     resourceObject=&MI_UI_MessageIcon_Backplate
     tiling=NoTile
     imageSize=<92 92>
     margin=<ltrb 0 0 0 0>
     outlineSettings {
      bUseBrushTransparency=false
      roundingType=HalfHeightRadius
      width=0
      color {colorUseRule=UseColor_Specified, specifiedColor=<rgba 0 0 0 0>}
      cornerRadii=<0 0 0 0>
     }
     tintColor {colorUseRule=UseColor_Specified, specifiedColor=<rgba 1 1 1 1>}
     uVRegion {bIsValid=false, max=<0 0>, min=<0 0>}
    }
    renderTransform {angle=-5, scale=<1 1>, shear=<0 0>, translation=<0 0>}
   }
   slot {class=GridSlot, properties {columnSpan=1, horizontalAlignment=HAlign_Center, layer=1, rowSpan=1, verticalAlignment=VAlign_Center}}
  }
  Icon {
   widgetClassPath=&Image
   bIsVariable=true
   parent=Root
   widget=Icon
   properties {brush {
     bIsDynamicallyLoaded=false
     drawAs=Image
     imageType=FullColor
     mirroring=NoMirror
     resourceName=None
     resourceObject=&MI_UI_Icon_Customizable
     tiling=NoTile
     imageSize=<46 46>
     margin=<ltrb 0 0 0 0>
     outlineSettings {
      bUseBrushTransparency=false
      roundingType=HalfHeightRadius
      width=0
      color {colorUseRule=UseColor_Specified, specifiedColor=<rgba 0 0 0 0>}
      cornerRadii=<0 0 0 0>
     }
     tintColor {colorUseRule=UseColor_Specified, specifiedColor=<rgba 1 1 1 1>}
     uVRegion {bIsValid=false, max=<0 0>, min=<0 0>}
    }}
   slot {class=GridSlot, properties {horizontalAlignment=HAlign_Center, layer=5, verticalAlignment=VAlign_Center}}
  }
  Text {
   widgetClassPath=&VerticalBox
   parent=Root
   widget=Text
   slot {
    class=GridSlot
    properties {column=1, columnSpan=1, horizontalAlignment=HAlign_Left, rowSpan=1, verticalAlignment=VAlign_Center, padding=<ltrb 12 8 12 8>}
   }
  }
  Text_Title {
   widgetClassPath=&CommonTextBlock
   bIsVariable=true
   parent=Text
   widget=Text_Title
   properties {
    autoWrapText=true
    bAutoCollapseWithEmptyText=true
    mobileFontSizeMultiplier=1.4
    text="[PH] Disclaimer Header"
    textTransformPolicy=ToUpper
    font {
     bForceMonospaced=false
     bMaterialIsStencil=false
     fontMaterial=null
     fontObject=&HeadingNow
     letterSpacing=-30
     monospacedWidth=1
     size=12.75
     skewAmount=0
     typefaceFontName=75 Medium Italic
     outlineSettings {
      bApplyOutlineToDropShadows=false
      bMiteredCorners=false
      bSeparateFillAlpha=false
      outlineMaterial=null
      outlineSize=0
      outlineColor=<rgba 0 0 0 1>
     }
    }
    shadowOffset=<0 0>
   }
   slot {class=VerticalBoxSlot}
  }
  Text_Body {
   widgetClassPath=&CommonTextBlock
   bIsVariable=true
   parent=Text
   widget=Text_Body
   properties {
    bAutoCollapseWithEmptyText=true
    mobileFontSizeMultiplier=1.4
    text="[PH] Long long long really long Disclaimer text  [PH] Long long long really long Disclaimer text "
    wrapTextAt=400
    colorAndOpacity {colorUseRule=UseColor_Specified, specifiedColor=<rgba 1 1 1 0.7>}
    font {
     bForceMonospaced=false
     bMaterialIsStencil=false
     fontMaterial=null
     fontObject=&HeadingNow
     letterSpacing=0
     monospacedWidth=1
     size=11.25
     skewAmount=0
     typefaceFontName=74 Regular
     outlineSettings {
      bApplyOutlineToDropShadows=false
      bMiteredCorners=false
      bSeparateFillAlpha=false
      outlineMaterial=null
      outlineSize=0
      outlineColor=<rgba 0 0 0 1>
     }
    }
    shadowOffset=<0 0>
   }
   slot {class=VerticalBoxSlot, properties {padding=<ltrb 0 5 0 0>}}
  }
 ]
}
