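# Translucent UI material for a circular character portrait in the Reboot season pass reward track top navigation.
# It composites RewardTexture over a hue-shifted, eased gradient from CharacterColor, using an antialiased distance-field circle as mask.
# Locked darkens the portrait, while Selected blends in a soft glow derived from a distance field.
# Centered UV scaling and a Scale-driven vertex transform adjust the drawn area around the circle.
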Material /SeasonPass/Widgets/Reboot/Materials/M_UI_BPReboot_RewardTrack_TopNav_Portrait
graph {
 properties {
  bAutomaticallySetUsageInEditor=false
  bUsedWithStaticMesh=false
  blendMode=BLEND_Translucent
  materialDomain=MD_UI
  usedShadingModels=Default Lit
 }
 nodes [
  n1 = TextureSampleParameter2D(UVs=MaterialFunctionCall(UV=NamedRerouteUsage()) {materialFunction=&MF_UI_UVScaling}.Scaled to Fit) {
   parameterName=RewardTexture
   texture=/BRCosmetics/UI/Foundation/Textures/Icons/Heroes/Athena/Soldier/S36/T_Soldier_PolarityWinnField_HeadphonesOFF
  }
  n2 = MaterialFunctionCall(DistanceField=NamedRerouteUsage(), Threshold=Constant() {r=1.05}) {materialFunction=&MF_UI_AntiAliasing}
  n3 = Power(Base=NamedRerouteUsage()) {constExponent=0.25}
  LinearInterpolate(
   A=LinearInterpolate(A=NamedRerouteUsage(), B=@n1.RGB, Alpha=Min(A=@n1.A, B=@n2.Inner))
   Alpha=LinearInterpolate(B=Multiply(A=@n3) {constB=0.7}, Alpha=ScalarParameter() {parameterName=Locked}) {constB=0}
  ) {constB=0}
  n4 = MaterialFunctionCall(DistanceField=NamedRerouteUsage(), Threshold=Constant() {r=1}) {materialFunction=&MF_UI_AntiAliasing}
  n5 = ScalarParameter() {parameterName=Selected}
  LinearInterpolate(
   A=LinearInterpolate(A=@n4.Inner, B=Multiply(A=@n4.Inner) {constB=0.85}, Alpha=@n5)
   Alpha=LinearInterpolate(B=NamedRerouteUsage(), Alpha=@n5)
  )
  MaterialFunctionCall!(Anchor, Translation, Padding, Scale=ScalarParameter() {defaultValue=2, parameterName=Scale}) {materialFunction=/FortUILibrary/MaterialFunctions/Basics/MF_UI_VertexTransform}
  n6 = MaterialFunctionCall(UV=NamedRerouteUsage()) {materialFunction=&MF_UI_UVScaling}
  n7 = Constant() {r=0.5}
  NamedRerouteDeclaration(None=Multiply(A=Distance(A=@n6.Scaled to Fit, B=@n7)) {constB=2}) {name=Circle SDF, nodeColor=<rgba 0 0 0 1>}
  n8 = Constant() {r=0.02}
  n9 = VectorParameter() {parameterName=CharacterColor, defaultValue=<rgba 0.1714 0.0561 0.521 1>}
  n10 = Power(Base=@n9.RGB) {constExponent=0.5}
  n11 = MaterialFunctionCall(
   UVs=MaterialFunctionCall(UV=NamedRerouteUsage()) {materialFunction=&MF_UI_UVScaling}.Scaled to Fit
   RotationAngle=Constant() {r=0.35}
  ) {materialFunction=&MF_UI_LinearGradient}
  NamedRerouteDeclaration(None=LinearInterpolate(
    A=MaterialFunctionCall(Hue Shift Percentage=@n8, Texture=@n10) {materialFunction=&HueShift}.Result
    B=@n9.RGB
    Alpha=MaterialFunctionCall("Value (0-1)"=@n11.Result) {materialFunction=&MF_UI_EaseCurves}.EaseInOutSine
   )) {name=Gradient, nodeColor=<rgba 0 0 0 1>}
  NamedRerouteDeclaration(None=MaterialFunctionCall(UVs=TextureCoordinate(), Texture Scale=Constant() {r=0.5}) {materialFunction=&ScaleUVsByCenter}.UVs) {name=UV, nodeColor=<rgba 0 0 0 1>}
  NamedRerouteDeclaration(None=MaterialFunctionCall(DistanceField=NamedRerouteUsage(), Threshold=Constant() {r=1}) {materialFunction=&MF_UI_AntiAliasing}.Inner) {name=Circle, nodeColor=<rgba 0 0 0 1>}
  n12 = NamedRerouteUsage()
  n13 = OneMinus(None=Saturate(None=SmoothStep(Value=@n12) {constMax=2}))
  NamedRerouteDeclaration(None=Multiply(A=MaterialFunctionCall("Value (0-1)"=@n13) {materialFunction=&MF_UI_EaseCurves}.EaseInOutSine) {constB=0.5}) {name=Glow, nodeColor=<rgba 0 0 0 1>}
 ]
}
