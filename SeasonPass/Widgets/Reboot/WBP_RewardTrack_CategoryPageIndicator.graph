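# Season pass reward track widget showing a portrait for the previous or next category in page navigation.
# Season pass and selected page view-model updates rerun Initialize, which gets the linear nav target per IsPreviousIndicator, excluding locked categories.
# HandleTargetCategoryUpdated lazy-loads that category's image into the portrait, or collapses the widget without a category.
# CF_FindBattlepass looks up the battle pass by template ID; IndicatorNavDestinationSet is declared but never broadcast here.
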WidgetBlueprint /SeasonPass/Widgets/Reboot/WBP_RewardTrack_CategoryPageIndicator
graphs [
 CF_FindBattlepass {nodes [
   CF_FindBattlepass!() -> (then, SeasonPasses:Array of Season Pass VM Object References) {is=title K2Node_FunctionEntry}
   n1 = Get SeasonPassSystemVM() -> (SeasonPassSystemVM:Season Pass System VM Object Reference)
   > Is Valid(InputObject=@n1)
   > Return Node!(
    execute
    ReturnValue:Season Pass VM Object Reference=FindSeasonPass!(self:Season Pass System VM Object Reference=@n1, TemplateId:String=battlepassdata) -> (ReturnValue:Season Pass VM Object Reference)
   )
  ]}
 EventGraph {nodes [
   Event PreConstruct()
   Event Construct()
   Event Tick()
  ]}
 HandleTargetCategoryUpdated {nodes [
   n1 = HandleTargetCategoryUpdated!() -> (then, SeasonPassCategoryVM:Season Pass Category VM Object Reference) {is=title K2Node_FunctionEntry}
   > n2 = Is Valid(InputObject=@n1)
   > SetBrushFromLazyDisplayAsset_DontMatchSize!(
    execute
    self:Fort Lazy Image Object Reference=Get LImg_Portrait() -> (LImg_Portrait:Fort Lazy Image Object Reference)
    LazyObject:"Soft Object Reference (by ref)"=GetImage!(self:Season Pass Category VM Object Reference=@n1) -> (ReturnValue:Soft Object Reference)
   ) -> (then)
   > SetVisibility(self=Get Root() -> (Root:Overlay Object Reference), InVisibility=SelfHitTestInvisible)
   SetVisibility(execute=@n2.Is Not Valid, self=Get Root(), InVisibility=Collapsed)
  ]}
 IndicatorNavDestinationSet {nodes [IndicatorNavDestinationSet!() -> (then, CategoryDestination:Season Pass Category VM Object Reference) {is=title K2Node_FunctionEntry}]}
 Initialize {nodes [
   Initialize() {is=title K2Node_FunctionEntry}
   > n1 = GetLinearSeasonPassPageNavTarget!(
    execute
    SeasonPassSelectionVM:Season Pass Selection VM Object Reference=Get SeasonPassSelectionVM() -> (SeasonPassSelectionVM:Season Pass Selection VM Object Reference)
    IsPrevious:bool=Get IsPreviousIndicator() -> (IsPreviousIndicator:bool)
    ShowLockedCategories:bool=false
    UseMainSetOnly:bool=false
   ) -> (
    then
    NavigateToRewardTarget:Season Pass Reward VM Object Reference
    SeasonPassCategoryTarget:Season Pass Category VM Object Reference
    SelectedPage:int=0
   )
   > HandleTargetCategoryUpdated!(execute, self, SeasonPassCategoryVM:Season Pass Category VM Object Reference=@n1.SeasonPassCategoryTarget) -> (then) {is=title K2Node_CallFunction}
  ]}
 VM_HandleSeasonPassUpdated {nodes [
   VM_HandleSeasonPassUpdated!() -> (then, SeasonPass:Season Pass VM Object Reference) {is=title K2Node_FunctionEntry}
   > Is Valid(InputObject=Get SeasonPassSelectionVM())
   > Initialize() {is=title K2Node_CallFunction}
  ]}
 VM_HandleSelectedPageUpdated {nodes [
   VM_HandleSelectedPageUpdated!() -> (then, SelectedPage:Season Pass Page VM Object Reference) {is=title K2Node_FunctionEntry}
   > Is Valid(InputObject=Get SeasonPassVM() -> (SeasonPassVM:Season Pass VM Object Reference))
   > Initialize()
  ]}
]
variables [
 IndicatorNavDestinationSet {}
 IsPreviousIndicator {type=boolean}
]
widgets {
 info {inheritedWidgetCount=0, parentClass=&UserWidget, rootWidgetClass=&Overlay, widgetCount=2}
 widgets [
  Root {widgetClassPath=&Overlay, bIsVariable=true, widget=Root}
  LImg_Portrait {
   widgetClassPath=&FortLazyImage
   bIsVariable=true
   parent=Root
   widget=LImg_Portrait
   properties {brush {
     bIsDynamicallyLoaded=false
     drawAs=Image
     imageType=FullColor
     mirroring=NoMirror
     resourceName=None
     resourceObject=/SeasonPass/Widgets/Reboot/Materials/M_UI_BPReboot_RewardTrack_TopNav_Portrait
     tiling=NoTile
     imageSize=<48 48>
     margin=<ltrb 0 0 0 0>
     outlineSettings {
      bUseBrushTransparency=false
      roundingType=HalfHeightRadius
      width=0
      color {colorUseRule=UseColor_Specified, specifiedColor=<rgba 0 0 0 0>}
      cornerRadii=<0 0 0 0>
     }
     tintColor {colorUseRule=UseColor_Specified, specifiedColor=<rgba 1 1 1 1>}
     uVRegion {bIsValid=false, max=<0 0>, min=<0 0>}
    }}
   slot {class=OverlaySlot, properties {horizontalAlignment=HAlign_Center, padding=<ltrb 4 4 4 4>}}
  }
 ]
}
