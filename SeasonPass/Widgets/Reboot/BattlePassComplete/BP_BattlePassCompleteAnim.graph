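# Activatable overlay widget that celebrates a completed Battle Pass with an animated icon, pass name and "Completed!" text.
# SetPass fills the name and lazily loaded icon material from a Season Pass view model.
# On activation it plays a sound and UnlockSequence, which animates transforms, opacity and material parameters (background transition, icon outline, sheen, tint), then deactivates itself when finished.
# The OnUnlockAnimationFinished dispatcher is declared but not called in these graphs.
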WidgetBlueprint /SeasonPass/Widgets/Reboot/BattlePassComplete/BP_BattlePassCompleteAnim
animations [UnlockSequence {
  binding_count=9
  length_seconds=5.3334
  object=&self:UnlockSequence
  sequence {
   bindings [
    Flare {
     id=#1
     possessed_class=Image
     tracks [
      {
       type=MovieSceneFloatTrack
       display_name=Render Opacity
       sections [{
         type=MovieSceneFloatSection
         end_frame=329
         start_frame=12
         channels [None {type=MovieSceneScriptingFloatChannel, keys [[171,0],[191,0.0415],[317,0]]}]
        }]
      }
      {
       type=MovieScene2DTransformTrack
       sections [{
         type=MovieScene2DTransformSection
         end_frame=318
         start_frame=12
         channels [
          Translation.X {type=MovieSceneScriptingFloatChannel, default=0}
          Translation.Y {type=MovieSceneScriptingFloatChannel, default=0}
          Angle {type=MovieSceneScriptingFloatChannel, default=0}
          Scale.X {type=MovieSceneScriptingFloatChannel, keys [[174,0.55],[187,1.2536],[306,1]]}
          Scale.Y {type=MovieSceneScriptingFloatChannel, keys [[174,0.47],[187,1]]}
          Shear.X {type=MovieSceneScriptingFloatChannel, default=0}
          Shear.Y {type=MovieSceneScriptingFloatChannel, default=0}
         ]
        }]
      }
     ]
    }
    Backing {
     id=#2
     possessed_class=Image
     tracks [
      {
       type=MovieSceneWidgetMaterialTrack
       sections [{
         type=MovieSceneParameterSection
         channels [
          Triangles_Amount {type=MovieSceneScriptingFloatChannel, keys [[33,0],[122,0],[182,2.5],[266,1],[323,0]]}
          Density {type=MovieSceneScriptingFloatChannel, thinned_from=5, keys [[33,0.666],[122,1],[290,1],[323,0.8]]}
          TransitionBackground {type=MovieSceneScriptingFloatChannel, thinned_from=6, keys [[34,0],[152,0],[182,1],[267,1]]}
          TrianglesMask_Radius {type=MovieSceneScriptingFloatChannel, thinned_from=6, keys [[33,2],[122,2],[152,4],[290,4],[323,8]]}
          Hotspot_Intensity {type=MovieSceneScriptingFloatChannel, keys [[33,1],[115,1],[175,5],[274,1],[298,1],[331,0.9]]}
         ]
        }]
      }
      {
       type=MovieSceneFloatTrack
       display_name=Render Opacity
       sections [{
         type=MovieSceneFloatSection
         channels [None {type=MovieSceneScriptingFloatChannel, thinned_from=5, keys [[12,0],[33,1],[312.716,1],[325.767,0]]}]
        }]
      }
     ]
    }
    LargeTriangles {
     id=#3
     possessed_class=HorizontalBox
     tracks [
      {
       type=MovieSceneFloatTrack
       display_name=Render Opacity
       sections [{
         type=MovieSceneFloatSection
         end_frame=318
         start_frame=12
         channels [None {type=MovieSceneScriptingFloatChannel, keys [[170,0],[196,1],[276,1],[306,0]]}]
        }]
      }
      {
       type=MovieScene2DTransformTrack
       sections [{
         type=MovieScene2DTransformSection
         end_frame=321
         start_frame=12
         channels [
          Translation.X {type=MovieSceneScriptingFloatChannel, default=0}
          Translation.Y {type=MovieSceneScriptingFloatChannel, default=0}
          Angle {type=MovieSceneScriptingFloatChannel, default=0}
          Scale.X {type=MovieSceneScriptingFloatChannel, keys [[166,0.29],[189,1],[309,1.2078]]}
          Scale.Y {type=MovieSceneScriptingFloatChannel, keys [[166,0.58],[189,1]]}
          Shear.X {type=MovieSceneScriptingFloatChannel, default=0}
          Shear.Y {type=MovieSceneScriptingFloatChannel, default=0}
         ]
        }]
      }
     ]
    }
    Glow {
     id=#4
     possessed_class=Image
     tracks [
      {
       type=MovieSceneFloatTrack
       display_name=Render Opacity
       sections [{
         type=MovieSceneFloatSection
         end_frame=329
         start_frame=12
         channels [None {type=MovieSceneScriptingFloatChannel, keys [[173,0],[182,0.0415],[317,0]]}]
        }]
      }
      {
       type=MovieScene2DTransformTrack
       sections [{
         type=MovieScene2DTransformSection
         end_frame=194
         start_frame=12
         channels [
          Translation.X {type=MovieSceneScriptingFloatChannel, const [0,1]}
          Translation.Y {type=MovieSceneScriptingFloatChannel, const [0,1]}
          Angle {type=MovieSceneScriptingFloatChannel, const [0,1]}
          Scale.X {type=MovieSceneScriptingFloatChannel, keys [[173,0],[182,1]]}
          Scale.Y {type=MovieSceneScriptingFloatChannel, keys [[173,0],[182,1]]}
          Shear.X {type=MovieSceneScriptingFloatChannel, const [0,1]}
          Shear.Y {type=MovieSceneScriptingFloatChannel, const [0,1]}
         ]
        }]
      }
      {type=MovieSceneWidgetMaterialTrack, sections [{type=MovieSceneParameterSection, end_frame=270, start_frame=12}]}
     ]
    }
    "Backing (5)" {
     id=#5
     possessed_class=CommonTextBlock
     tracks [
      {
       type=MovieScene2DTransformTrack
       sections [{
         type=MovieScene2DTransformSection
         channels [
          Translation.X {type=MovieSceneScriptingFloatChannel}
          Translation.Y {type=MovieSceneScriptingFloatChannel, keys [[158,64],[185,0],[274,0],[307,144]]}
          Angle {type=MovieSceneScriptingFloatChannel}
          Scale.X {type=MovieSceneScriptingFloatChannel, keys [[274,1],[307,3]]}
          Scale.Y {type=MovieSceneScriptingFloatChannel, keys [[274,1],[307,3]]}
          Shear.X {type=MovieSceneScriptingFloatChannel}
          Shear.Y {type=MovieSceneScriptingFloatChannel}
         ]
        }]
      }
      {
       type=MovieSceneFloatTrack
       display_name=Render Opacity
       sections [{type=MovieSceneFloatSection, channels [None {type=MovieSceneScriptingFloatChannel, keys [[158,0],[206,1],[293,1],[303,0]]}]}]
      }
     ]
    }
    Icon {
     id=#6
     possessed_class=FortLazyImage
     tracks [{
       type=MovieSceneWidgetMaterialTrack
       sections [{
         type=MovieSceneParameterSection
         end_frame=332
         start_frame=12
         channels [
          SheenSwipe {type=MovieSceneScriptingFloatChannel, keys [[249,0.75],[312,-0.75]]}
          OutlineSwipe {type=MovieSceneScriptingFloatChannel, keys [[38,0],[83,0.4611],[275,1]]}
          Under Glow Opacity {type=MovieSceneScriptingFloatChannel, keys [[157,0],[180,0.35],[234,0.15]]}
          UsePureTint {type=MovieSceneScriptingFloatChannel, keys [[12,1],[65,0],[295,0],[320,1]]}
          Outline Opacity {type=MovieSceneScriptingFloatChannel, keys [[38,0],[93,1],[212,1],[275,0]]}
          PureTintColorR {type=MovieSceneScriptingFloatChannel, keys [[84,1],[293,-0.47]]}
          PureTintColorG {type=MovieSceneScriptingFloatChannel, keys [[12,0.8417],[283,0.8417],[293,1.2692]]}
          PureTintColorB {type=MovieSceneScriptingFloatChannel, keys [[12,0.05],[283,0.05],[293,0.31]]}
          PureTintColorA {type=MovieSceneScriptingFloatChannel, const [1,2]}
         ]
        }]
      }]
    }
    Completed {
     id=#7
     possessed_class=CommonTextBlock
     tracks [
      {
       type=MovieScene2DTransformTrack
       sections [{
         type=MovieScene2DTransformSection
         channels [
          Translation.X {type=MovieSceneScriptingFloatChannel}
          Translation.Y {type=MovieSceneScriptingFloatChannel, keys [[270,0],[299,160]]}
          Angle {type=MovieSceneScriptingFloatChannel}
          Scale.X {type=MovieSceneScriptingFloatChannel, keys [[176,4],[191,0.95],[197,1.025],[206,1],[270,1],[299,4]]}
          Scale.Y {type=MovieSceneScriptingFloatChannel, keys [[176,2],[191,0.975],[197,1.025],[206,1],[270,1],[299,4]]}
          Shear.X {type=MovieSceneScriptingFloatChannel}
          Shear.Y {type=MovieSceneScriptingFloatChannel}
         ]
        }]
      }
      {
       type=MovieSceneFloatTrack
       display_name=Render Opacity
       sections [{type=MovieSceneFloatSection, channels [None {type=MovieSceneScriptingFloatChannel, thinned_from=5, keys [[177,0],[191,1],[290,1],[299,0]]}]}]
      }
     ]
    }
    IconBox {
     id=#8
     possessed_class=SizeBox
     tracks [
      {
       type=MovieSceneFloatTrack
       display_name=Render Opacity
       sections [{type=MovieSceneFloatSection, channels [None {type=MovieSceneScriptingFloatChannel, thinned_from=5, keys [[33,0],[36,1],[317,1],[319,0]]}]}]
      }
      {
       type=MovieScene2DTransformTrack
       sections [{
         type=MovieScene2DTransformSection
         channels [
          Translation.X {type=MovieSceneScriptingFloatChannel, keys [[204,0.75],[324,-0.75]]}
          Translation.Y {type=MovieSceneScriptingFloatChannel, keys [[163,0],[193,-128],[293,-128],[319,-80]]}
          Angle {type=MovieSceneScriptingFloatChannel}
          Scale.X {type=MovieSceneScriptingFloatChannel, keys [[33,0],[63,1.25],[293,1.25],[319,3.75]]}
          Scale.Y {type=MovieSceneScriptingFloatChannel, keys [[33,0],[63,1.25],[293,1.25],[319,3.75]]}
          Shear.X {type=MovieSceneScriptingFloatChannel}
          Shear.Y {type=MovieSceneScriptingFloatChannel}
         ]
        }]
      }
     ]
    }
    SmallTriangles {
     id=#9
     possessed_class=HorizontalBox
     tracks [
      {
       type=MovieSceneFloatTrack
       display_name=Render Opacity
       sections [{
         type=MovieSceneFloatSection
         end_frame=318
         start_frame=12
         channels [None {type=MovieSceneScriptingFloatChannel, keys [[179,0],[214,1],[276,1],[306,0]]}]
        }]
      }
      {
       type=MovieScene2DTransformTrack
       sections [{
         type=MovieScene2DTransformSection
         end_frame=321
         start_frame=12
         channels [
          Translation.X {type=MovieSceneScriptingFloatChannel, const [0,1]}
          Translation.Y {type=MovieSceneScriptingFloatChannel, const [0,1]}
          Angle {type=MovieSceneScriptingFloatChannel, const [0,1]}
          Scale.X {type=MovieSceneScriptingFloatChannel, keys [[149,0.29],[192,1],[309,1.2078]]}
          Scale.Y {type=MovieSceneScriptingFloatChannel, keys [[149,0.58],[192,1]]}
          Shear.X {type=MovieSceneScriptingFloatChannel, const [0,1]}
          Shear.Y {type=MovieSceneScriptingFloatChannel, const [0,1]}
         ]
        }]
      }
     ]
    }
   ]
   display_rate {numerator=60}
  }
 }]
graphs [
 EventGraph {nodes [
   PlayIntro() {is=title K2Node_CustomEvent}
   n1 = Get UnlockSequence() -> (UnlockSequence:Widget Animation Object Reference)
   > PlayAnimation(InAnimation=@n1)
   > PlaySound2D(Sound=Get SoundOnPlayIntro() -> (SoundOnPlayIntro:Sound Base Object Reference))
   > BindToAnimationFinished(Animation=@n1, Delegate=Create Event())
   OnIntroAnimationFinished() {is=title K2Node_CustomEvent}
   > UnbindFromAnimationFinished!(
    execute
    self:User Widget Object Reference
    Animation:Widget Animation Object Reference=Get UnlockSequence()
    Delegate:Delegate=Create Event()
   ) -> (then)
   > DeactivateWidget()
   Event On Activated()
   > PlayIntro() {is=title K2Node_CallFunction}
  ]}
 OnUnlockAnimationFinished {nodes [OnUnlockAnimationFinished() {is=title K2Node_FunctionEntry}]}
 SetPass {nodes [
   n1 = SetPass!() -> (then, Pass:Season Pass VM Object Reference) {is=title K2Node_FunctionEntry}
   > Is Valid(InputObject=@n1)
   > Set Text(
    Text=GetDisplayName!(self:Season Pass VM Object Reference=@n1) -> (ReturnValue:Text)
    self=Get BattlePass() -> (BattlePass:Common Text Object Reference)
   )
   > SetBrushFromLazyDisplayAsset!(
    execute
    self:Common Lazy Image Object Reference=Get Icon() -> (Icon:Fort Lazy Image Object Reference)
    LazyObject:"Soft Object Reference (by ref)"=GetPurchasePreMessageMaterial!(self:Season Pass VM Object Reference=@n1) -> (ReturnValue:Soft Object Reference)
    bMatchTextureSize:bool=false
   ) -> (then)
  ]}
]
variables [
 OnUnlockAnimationFinished {}
 SoundOnPlayIntro {type=SoundBase, default=/SeasonPass/audio/MSS_UI_BP_All_Sets_Complete}
]
widgets {
 info {inheritedWidgetCount=0, parentClass=&CommonActivatableWidget, rootWidgetClass=&Overlay, widgetCount=24}
 widgets [
  Overlay_1 {widgetClassPath=&Overlay, widget=Overlay_1, properties {visibility=Visible}}
  Overlay {
   widgetClassPath=&Overlay
   parent=Overlay_1
   widget=Overlay
   properties {clipping=ClipToBounds}
   slot {class=OverlaySlot, properties {horizontalAlignment=HAlign_Fill, verticalAlignment=VAlign_Fill}}
  }
  Backing {
   widgetClassPath=&Image
   bIsVariable=true
   parent=Overlay
   widget=Backing
   properties {brush {
     bIsDynamicallyLoaded=false
     drawAs=Image
     imageType=FullColor
     mirroring=NoMirror
     resourceName=None
     resourceObject=&/M_UI_BattlePassPurchase_Screen_Transition_Inst
     tiling=NoTile
     imageSize=<32 32>
     margin=<ltrb 0 0 0 0>
     outlineSettings {
      bUseBrushTransparency=false
      roundingType=HalfHeightRadius
      width=0
      color {colorUseRule=UseColor_Specified, specifiedColor=<rgba 0 0 0 0>}
      cornerRadii=<0 0 0 0>
     }
     tintColor {colorUseRule=UseColor_Specified, specifiedColor=<rgba 1 1 1 1>}
     uVRegion {bIsValid=false, max=<0 0>, min=<0 0>}
    }}
   slot {class=OverlaySlot, properties {horizontalAlignment=HAlign_Fill, verticalAlignment=VAlign_Fill}}
  }
  Overlay_3 {
   widgetClassPath=&Overlay
   parent=Overlay
   widget=Overlay_3
   slot {class=OverlaySlot, properties {horizontalAlignment=HAlign_Center, verticalAlignment=VAlign_Center, padding=<ltrb 0 0 0 120>}}
  }
  Glow {
   widgetClassPath=&Image
   bIsVariable=true
   parent=Overlay_3
   widget=Glow
   properties {
    renderOpacity=0
    brush {
     bIsDynamicallyLoaded=false
     drawAs=Image
     imageType=FullColor
     mirroring=NoMirror
     resourceName=None
     resourceObject=&/M_UI_BattlePassPurchase_Screen_Glow
     tiling=NoTile
     imageSize=<1600 2000>
     margin=<ltrb 0 0 0 0>
     outlineSettings {
      bUseBrushTransparency=false
      roundingType=HalfHeightRadius
      width=0
      color {colorUseRule=UseColor_Specified, specifiedColor=<rgba 0 0 0 0>}
      cornerRadii=<0 0 0 0>
     }
     tintColor {colorUseRule=UseColor_Specified, specifiedColor=<rgba 1 1 1 1>}
     uVRegion {bIsValid=false, max=<0 0>, min=<0 0>}
    }
   }
   slot {class=OverlaySlot, properties {horizontalAlignment=HAlign_Center, verticalAlignment=VAlign_Center}}
  }
  Flare {
   widgetClassPath=&Image
   bIsVariable=true
   parent=Overlay_3
   widget=Flare
   properties {
    renderOpacity=0
    brush {
     bIsDynamicallyLoaded=false
     drawAs=Image
     imageType=FullColor
     mirroring=NoMirror
     resourceName=None
     resourceObject=&M_UI_BattlePassPurchase_Screen_Glow
     tiling=NoTile
     imageSize=<18000 317.8351>
     margin=<ltrb 0 0 0 0>
     outlineSettings {
      bUseBrushTransparency=false
      roundingType=HalfHeightRadius
      width=0
      color {colorUseRule=UseColor_Specified, specifiedColor=<rgba 0 0 0 0>}
      cornerRadii=<0 0 0 0>
     }
     tintColor {colorUseRule=UseColor_Specified, specifiedColor=<rgba 1 1 1 1>}
     uVRegion {bIsValid=false, max=<0 0>, min=<0 0>}
    }
   }
   slot {class=OverlaySlot, properties {horizontalAlignment=HAlign_Center, verticalAlignment=VAlign_Center}}
  }
  SmallTriangles {
   widgetClassPath=&HorizontalBox
   parent=Overlay_3
   widget=SmallTriangles
   properties {renderOpacity=0, renderTransform {angle=0, scale=<1 -1>, shear=<0 0>, translation=<0 0>}}
   slot {class=OverlaySlot, properties {horizontalAlignment=HAlign_Center, verticalAlignment=VAlign_Center}}
  }
  Image_TrianglesLeft_Small {
   widgetClassPath=&Image
   parent=SmallTriangles
   widget=Image_TrianglesLeft_Small
   properties {
    visibility=SelfHitTestInvisible
    brush {
     bIsDynamicallyLoaded=false
     drawAs=Image
     imageType=FullColor
     mirroring=NoMirror
     resourceName=None
     resourceObject=/FortniteCoreUI/HUD/EndGame/PostGameFlowScreens/DefaultScreens/XPScreen/Materials/MI_PostMatch_Triangles_Small
     tiling=NoTile
     imageSize=<320 320>
     margin=<ltrb 0 0 0 0>
     outlineSettings {
      bUseBrushTransparency=false
      roundingType=HalfHeightRadius
      width=0
      color {colorUseRule=UseColor_Specified, specifiedColor=<rgba 0 0 0 0>}
      cornerRadii=<0 0 0 0>
     }
     tintColor {colorUseRule=UseColor_Specified, specifiedColor=<rgba 0.0209 1 0.147 1>}
     uVRegion {bIsValid=false, max=<0 0>, min=<0 0>}
    }
   }
   slot {class=HorizontalBoxSlot}
  }
  Spacer_1 {widgetClassPath=&Spacer, parent=SmallTriangles, widget=Spacer_1, properties {size=<320 1>}, slot {class=HorizontalBoxSlot}}
  Image_TrianglesRight_Small {
   widgetClassPath=&Image
   parent=SmallTriangles
   widget=Image_TrianglesRight_Small
   properties {
    visibility=SelfHitTestInvisible
    brush {
     bIsDynamicallyLoaded=false
     drawAs=Image
     imageType=FullColor
     mirroring=NoMirror
     resourceName=None
     resourceObject=&MI_PostMatch_Triangles_Small
     tiling=NoTile
     imageSize=<320 320>
     margin=<ltrb 0 0 0 0>
     outlineSettings {
      bUseBrushTransparency=false
      roundingType=HalfHeightRadius
      width=0
      color {colorUseRule=UseColor_Specified, specifiedColor=<rgba 0 0 0 0>}
      cornerRadii=<0 0 0 0>
     }
     tintColor {colorUseRule=UseColor_Specified, specifiedColor=<rgba 0.0209 1 0.147 1>}
     uVRegion {bIsValid=false, max=<0 0>, min=<0 0>}
    }
    renderTransform {angle=180, scale=<1 1>, shear=<0 0>, translation=<0 0>}
   }
   slot {class=HorizontalBoxSlot, properties {horizontalAlignment=HAlign_Center, verticalAlignment=VAlign_Center}}
  }
  LargeTriangles {
   widgetClassPath=&HorizontalBox
   parent=Overlay_3
   widget=LargeTriangles
   properties {renderOpacity=0}
   slot {class=OverlaySlot, properties {horizontalAlignment=HAlign_Center, verticalAlignment=VAlign_Center}}
  }
  Image_TrianglesLeft_Large {
   widgetClassPath=&Image
   parent=LargeTriangles
   widget=Image_TrianglesLeft_Large
   properties {
    visibility=SelfHitTestInvisible
    brush {
     bIsDynamicallyLoaded=false
     drawAs=Image
     imageType=FullColor
     mirroring=NoMirror
     resourceName=None
     resourceObject=/FortniteCoreUI/HUD/EndGame/PostGameFlowScreens/DefaultScreens/XPScreen/Materials/MI_PostMatch_Triangles
     tiling=NoTile
     imageSize=<512 320>
     margin=<ltrb 0 0 0 0>
     outlineSettings {
      bUseBrushTransparency=false
      roundingType=HalfHeightRadius
      width=0
      color {colorUseRule=UseColor_Specified, specifiedColor=<rgba 0 0 0 0>}
      cornerRadii=<0 0 0 0>
     }
     tintColor {colorUseRule=UseColor_Specified, specifiedColor=<rgba 0.1645 0.3594 0.0299 1>}
     uVRegion {bIsValid=false, max=<0 0>, min=<0 0>}
    }
   }
   slot {class=HorizontalBoxSlot}
  }
  Spacer_50 {widgetClassPath=&Spacer, parent=LargeTriangles, widget=Spacer_50, properties {size=<320 1>}, slot {class=HorizontalBoxSlot}}
  Image_TrianglesRight_Large {
   widgetClassPath=&Image
   parent=LargeTriangles
   widget=Image_TrianglesRight_Large
   properties {
    visibility=SelfHitTestInvisible
    brush {
     bIsDynamicallyLoaded=false
     drawAs=Image
     imageType=FullColor
     mirroring=NoMirror
     resourceName=None
     resourceObject=/FortniteCoreUI/HUD/EndGame/PostGameFlowScreens/DefaultScreens/XPScreen/Materials/MI_PostMatch_Triangles_Alternate
     tiling=NoTile
     imageSize=<512 320>
     margin=<ltrb 0 0 0 0>
     outlineSettings {
      bUseBrushTransparency=false
      roundingType=HalfHeightRadius
      width=0
      color {colorUseRule=UseColor_Specified, specifiedColor=<rgba 0 0 0 0>}
      cornerRadii=<0 0 0 0>
     }
     tintColor {colorUseRule=UseColor_Specified, specifiedColor=<rgba 0.1645 0.3594 0.0299 1>}
     uVRegion {bIsValid=false, max=<0 0>, min=<0 0>}
    }
    renderTransform {angle=180, scale=<1 1>, shear=<0 0>, translation=<0 0>}
   }
   slot {class=HorizontalBoxSlot, properties {horizontalAlignment=HAlign_Center, verticalAlignment=VAlign_Center}}
  }
  IconBox {
   widgetClassPath=&SizeBox
   parent=Overlay
   widget=IconBox
   properties {bOverride_HeightOverride=true, bOverride_WidthOverride=true, heightOverride=512, widthOverride=512}
   slot {class=OverlaySlot, properties {horizontalAlignment=HAlign_Center, verticalAlignment=VAlign_Center}}
  }
  AnimBoundIconTint {
   widgetClassPath=&Border
   parent=IconBox
   widget=AnimBoundIconTint
   properties {
    background {
     bIsDynamicallyLoaded=false
     drawAs=NoDrawType
     imageType=NoImage
     mirroring=NoMirror
     resourceName=None
     resourceObject=null
     tiling=NoTile
     imageSize=<32 32>
     margin=<ltrb 0 0 0 0>
     outlineSettings {
      bUseBrushTransparency=false
      roundingType=HalfHeightRadius
      width=0
      color {colorUseRule=UseColor_Specified, specifiedColor=<rgba 0 0 0 0>}
      cornerRadii=<0 0 0 0>
     }
     tintColor {colorUseRule=UseColor_Specified, specifiedColor=<rgba 1 1 1 1>}
     uVRegion {bIsValid=false, max=<0 0>, min=<0 0>}
    }
    padding=<ltrb 0 0 0 0>
   }
   slot {class=SizeBoxSlot}
  }
  Icon {
   widgetClassPath=&FortLazyImage
   bIsVariable=true
   parent=AnimBoundIconTint
   widget=Icon
   properties {brush {
     bIsDynamicallyLoaded=false
     drawAs=Image
     imageType=FullColor
     mirroring=NoMirror
     resourceName=None
     resourceObject=&/M_UI_BattlePassComplete_Animated_Transition
     tiling=NoTile
     imageSize=<512 512>
     margin=<ltrb 0 0 0 0>
     outlineSettings {
      bUseBrushTransparency=false
      roundingType=HalfHeightRadius
      width=0
      color {colorUseRule=UseColor_Specified, specifiedColor=<rgba 0 0 0 0>}
      cornerRadii=<0 0 0 0>
     }
     tintColor {colorUseRule=UseColor_Specified, specifiedColor=<rgba 1 1 1 1>}
     uVRegion {bIsValid=false, max=<0 0>, min=<0 0>}
    }}
   slot {class=BorderSlot, properties {padding=<ltrb 0 0 0 0>}}
  }
  VerticalBox {
   widgetClassPath=&VerticalBox
   parent=Overlay
   widget=VerticalBox
   slot {class=OverlaySlot, properties {horizontalAlignment=HAlign_Center, verticalAlignment=VAlign_Fill}}
  }
  Spacer_93 {
   widgetClassPath=&Spacer
   parent=VerticalBox
   widget=Spacer_93
   slot {class=VerticalBoxSlot, properties {size {sizeRule=Fill, value=1}}}
  }
  VerticalBox_2 {
   widgetClassPath=&VerticalBox
   parent=VerticalBox
   widget=VerticalBox_2
   slot {class=VerticalBoxSlot, properties {padding=<ltrb 0 256 0 0>, size {sizeRule=Fill, value=1}}}
  }
  AnimBoundBattlepassText {
   widgetClassPath=&Border
   parent=VerticalBox_2
   widget=AnimBoundBattlepassText
   properties {
    background {
     bIsDynamicallyLoaded=false
     drawAs=NoDrawType
     imageType=NoImage
     mirroring=NoMirror
     resourceName=None
     resourceObject=null
     tiling=NoTile
     imageSize=<32 32>
     margin=<ltrb 0 0 0 0>
     outlineSettings {
      bUseBrushTransparency=false
      roundingType=HalfHeightRadius
      width=0
      color {colorUseRule=UseColor_Specified, specifiedColor=<rgba 0 0 0 0>}
      cornerRadii=<0 0 0 0>
     }
     tintColor {colorUseRule=UseColor_Specified, specifiedColor=<rgba 1 1 1 1>}
     uVRegion {bIsValid=false, max=<0 0>, min=<0 0>}
    }
    padding=<ltrb 0 0 0 0>
   }
   slot {class=VerticalBoxSlot, properties {horizontalAlignment=HAlign_Center}}
  }
  BattlePass {
   widgetClassPath=&CommonTextBlock
   bIsVariable=true
   parent=AnimBoundBattlepassText
   widget=BattlePass
   properties {
    justification=Center
    text=Battle Pass
    textTransformPolicy=ToUpper
    colorAndOpacity {colorUseRule=UseColor_Specified, specifiedColor=<rgba 0.0263 0.9167 0.2668 1>}
    font {
     bForceMonospaced=false
     bMaterialIsStencil=false
     fontMaterial=null
     fontObject=&Burbank
     letterSpacing=0
     monospacedWidth=1
     size=40
     skewAmount=0
     typefaceFontName=BurbankRegular-Black
     outlineSettings {
      bApplyOutlineToDropShadows=false
      bMiteredCorners=false
      bSeparateFillAlpha=false
      outlineMaterial=null
      outlineSize=2
      outlineColor=<rgba 0.0908 0.0185 0.0033 0>
     }
    }
    shadowOffset=<0 0>
   }
   slot {class=BorderSlot, properties {padding=<ltrb 0 0 0 0>}}
  }
  AnimBoundUnlockText {
   widgetClassPath=&Border
   parent=VerticalBox_2
   widget=AnimBoundUnlockText
   properties {
    background {
     bIsDynamicallyLoaded=false
     drawAs=NoDrawType
     imageType=NoImage
     mirroring=NoMirror
     resourceName=None
     resourceObject=null
     tiling=NoTile
     imageSize=<32 32>
     margin=<ltrb 0 0 0 0>
     outlineSettings {
      bUseBrushTransparency=false
      roundingType=HalfHeightRadius
      width=0
      color {colorUseRule=UseColor_Specified, specifiedColor=<rgba 0 0 0 0>}
      cornerRadii=<0 0 0 0>
     }
     tintColor {colorUseRule=UseColor_Specified, specifiedColor=<rgba 1 1 1 1>}
     uVRegion {bIsValid=false, max=<0 0>, min=<0 0>}
    }
    padding=<ltrb 0 0 0 0>
   }
   slot {class=VerticalBoxSlot, properties {horizontalAlignment=HAlign_Center, verticalAlignment=VAlign_Center, padding=<ltrb 0 -16 0 0>}}
  }
  Completed {
   widgetClassPath=&CommonTextBlock
   parent=AnimBoundUnlockText
   widget=Completed
   properties {
    justification=Center
    text="Completed!"
    textTransformPolicy=ToUpper
    font {
     bForceMonospaced=false
     bMaterialIsStencil=false
     fontMaterial=null
     fontObject=&Burbank
     letterSpacing=0
     monospacedWidth=1
     size=88
     skewAmount=0
     typefaceFontName=BurbankRegular-Black
     outlineSettings {
      bApplyOutlineToDropShadows=false
      bMiteredCorners=false
      bSeparateFillAlpha=false
      outlineMaterial=null
      outlineSize=0
      outlineColor=<rgba 0 0 0 1>
     }
    }
    margin=<ltrb 16 0 16 0>
    shadowOffset=<0 0>
   }
   slot {class=BorderSlot, properties {padding=<ltrb 0 0 0 0>}}
  }
 ]
}
