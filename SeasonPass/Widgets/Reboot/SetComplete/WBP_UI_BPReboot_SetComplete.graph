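# Full-screen activatable celebration overlay shown when one or more Season Pass sets are completed.
# On activation it plays the Celebration animation with a sound, then deactivates itself when the animation finishes.
# SetSingleOrMulti switches between set art and a count with "Sets Complete!" text based on NumSetsCompleted; SetCharacterImage loads the theme's LargePreview image.
# The animation drives material parameters, and a sequencer event broadcasts AnimationFinishing for fade-out.
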WidgetBlueprint /SeasonPass/Widgets/Reboot/SetComplete/WBP_UI_BPReboot_SetComplete
animations [Celebration {
  binding_count=10
  length_seconds=5.0501
  object=&self:Celebration
  sequence {
   bindings [
    SetsNumberSwitcher {
     id=#1
     possessed_class=CommonVisibilitySwitcher
     tracks [
      {
       type=MovieSceneFloatTrack
       display_name=Render Opacity
       sections [{
         type=MovieSceneFloatSection
         end_frame=97
         start_frame=0
         channels [None {type=MovieSceneScriptingFloatChannel, keys [[23,0],[27,-0.0033],[32,1.01],[89.7727,1],[95,0]]}]
        }]
      }
      {
       type=MovieScene2DTransformTrack
       sections [{
         type=MovieScene2DTransformSection
         end_frame=95
         start_frame=0
         channels [Translation.X {type=MovieSceneScriptingFloatChannel, keys [[27,150.9989],[95,-148.4272]]}]
        }]
      }
     ]
    }
    Modal {
     id=#2
     possessed_class=Image
     tracks [
      {
       type=MovieSceneWidgetMaterialTrack
       sections [{
         type=MovieSceneParameterSection
         end_frame=101
         start_frame=0
         channels [
          Modal Top Fade {type=MovieSceneScriptingFloatChannel, keys [[0,0.4],[21,0.4],[28,0.36],[91,0.36],[96,0.6]]}
          GreenAmount {type=MovieSceneScriptingFloatChannel, keys [[-0.0007,1],[0,0],[28,0],[40,2],[50,2],[96,0]]}
          Line Reveal {type=MovieSceneScriptingFloatChannel, keys [[25,1],[31,0.1]]}
          UnderGlow_Opacity {type=MovieSceneScriptingFloatChannel, keys [[25,0],[40,1]]}
          Underlight_Opacity {type=MovieSceneScriptingFloatChannel, keys [[27,0],[30,1]]}
          Color 1R {type=MovieSceneScriptingFloatChannel, keys [[0,0.039],[40,0.039],[60,0.5333],[85,0.5255],[101,1]]}
          Color 1G {type=MovieSceneScriptingFloatChannel, thinned_from=6, keys [[-0.0007,1],[40,1],[60,0.1981],[85,0.1881],[101,1]]}
          Color 1B {type=MovieSceneScriptingFloatChannel, thinned_from=5, keys [[0,0.23],[40,0.23],[60,1],[101,1]]}
          Color 1A {type=MovieSceneScriptingFloatChannel, const [1,1]}
          Color 2R {type=MovieSceneScriptingFloatChannel, keys [[0,0.039],[40,0.039],[60,0],[85,0],[101,1]]}
          Color 2G {type=MovieSceneScriptingFloatChannel, keys [[4,1],[40,1],[60,0.189],[85,0.189],[101,1]]}
          Color 2B {type=MovieSceneScriptingFloatChannel, thinned_from=5, keys [[0,0.23],[40,0.23],[60,1],[101,1]]}
          Color 2A {type=MovieSceneScriptingFloatChannel, const [1,1]}
         ]
        }]
      }
      {
       type=MovieSceneFloatTrack
       display_name=Render Opacity
       sections [{
         type=MovieSceneFloatSection
         end_frame=97
         start_frame=0
         channels [None {type=MovieSceneScriptingFloatChannel, keys [[0,0],[20,0],[34,1],[91,1],[91.119,0.75],[96,0]]}]
        }]
      }
      {
       type=MovieScene2DTransformTrack
       sections [{
         type=MovieScene2DTransformSection
         end_frame=98
         start_frame=0
         channels [
          Translation.X {type=MovieSceneScriptingFloatChannel, default=0}
          Translation.Y {type=MovieSceneScriptingFloatChannel, keys [[24,50],[29,0],[88,0],[96,70]]}
          Scale.Y {type=MovieSceneScriptingFloatChannel, keys [[24.011,0],[30,1.0078],[88,1],[96,0]]}
         ]
        }]
      }
     ]
    }
    SetName_Background {
     id=#3
     possessed_class=UEFN_TextBlock_C
     tracks [
      {
       type=MovieScene2DTransformTrack
       sections [{
         type=MovieScene2DTransformSection
         end_frame=96
         start_frame=2
         channels [Translation.X {type=MovieSceneScriptingFloatChannel, keys [[25,-1250],[95,1250]]}]
        }]
      }
      {
       type=MovieSceneFloatTrack
       display_name=Render Opacity
       sections [{
         type=MovieSceneFloatSection
         end_frame=94
         start_frame=0
         channels [None {type=MovieSceneScriptingFloatChannel, keys [[0,0],[25,0],[31,1],[88,1],[94,0.1]]}]
        }]
      }
     ]
    }
    BackgroundBlur_51 {
     id=#4
     possessed_class=BackgroundBlur
     tracks [{
       type=MovieSceneFloatTrack
       display_name=Blur Strength
       sections [{
         type=MovieSceneFloatSection
         end_frame=101
         start_frame=0
         channels [None {type=MovieSceneScriptingFloatChannel, keys [[0,0],[0.0007,0],[3.0007,30],[95.0007,30],[100.0007,0]]}]
        }]
      }]
    }
    VerticalBoxSlot_3 {id=#5, parent_id=#2, possessed_class=VerticalBoxSlot, tracks []}
    SetName {
     id=#6
     possessed_class=UEFN_TextBlock_C
     tracks [
      {
       type=MovieScene2DTransformTrack
       sections [{
         type=MovieScene2DTransformSection
         end_frame=100
         start_frame=0
         channels [Translation.X {type=MovieSceneScriptingFloatChannel, keys [[25,-312.7227],[81,28.5291],[94,485.6409]]}]
        }]
      }
      {
       type=MovieSceneFloatTrack
       display_name=Render Opacity
       sections [{
         type=MovieSceneFloatSection
         end_frame=99
         start_frame=0
         channels [None {type=MovieSceneScriptingFloatChannel, keys [[0,0],[25,0],[31,1],[87.7493,1],[93.8747,0]]}]
        }]
      }
     ]
    }
    Icon {
     id=#7
     possessed_class=Image
     tracks [
      {
       type=MovieSceneWidgetMaterialTrack
       sections [{
         type=MovieSceneParameterSection
         end_frame=97
         start_frame=0
         channels [
          ArcEndAngle {type=MovieSceneScriptingFloatChannel, keys [[0,0],[2,0],[14,0.75]]}
          ArcStartAngle {type=MovieSceneScriptingFloatChannel, keys [[0,0],[2,0],[14,1.1]]}
          FinalOpacity {type=MovieSceneScriptingFloatChannel, keys [[0,0],[15,0],[20,1],[86,1],[93,0]]}
          CheckedOutlineOpacity {type=MovieSceneScriptingFloatChannel, keys [[0,0],[15,0],[16,1],[18,1],[20,0]]}
          CircleOutlineOpacity {type=MovieSceneScriptingFloatChannel, thinned_from=4, keys [[-0.0007,1],[18,1],[20,0]]}
          Gradient Amount {type=MovieSceneScriptingFloatChannel, keys [[0,1],[18,1],[20,0]]}
         ]
        }]
      }
      {
       type=MovieScene2DTransformTrack
       sections [{
         type=MovieScene2DTransformSection
         end_frame=88
         start_frame=0
         channels [
          Translation.Y {type=MovieSceneScriptingFloatChannel, keys [[0,-180],[24,-180],[30,0]]}
          Scale.X {type=MovieSceneScriptingFloatChannel, keys [[0,12],[14,10],[20,1]]}
          Scale.Y {type=MovieSceneScriptingFloatChannel, keys [[0,12],[14,10],[20,1]]}
         ]
        }]
      }
     ]
    }
    MultipleSets {
     id=#8
     possessed_class=UEFN_TextBlock_C
     tracks [{
       type=MovieSceneWidgetMaterialTrack
       display_name=Font.FontMaterial
       sections [{
         type=MovieSceneParameterSection
         end_frame=100
         start_frame=0
         channels [
          Number Appear {type=MovieSceneScriptingFloatChannel, keys [[20,0],[45,0.5912],[90,1]]}
          Fill Alpha {type=MovieSceneScriptingFloatChannel, default=0, keys [[23,-0.0048],[58,2.2],[84,2.2],[95,0]]}
          Line Alpha Opacity {type=MovieSceneScriptingFloatChannel, keys [[20,0],[27,1],[54,2],[90,1]]}
          Color 1R {type=MovieSceneScriptingFloatChannel, keys [[46,0.0395],[66.055,0.1395]]}
          Color 1G {type=MovieSceneScriptingFloatChannel, keys [[46,1],[66.055,0.5885]]}
          Color 1B {type=MovieSceneScriptingFloatChannel, keys [[46,0.2307],[66.055,1.0207]]}
          Color 1A {type=MovieSceneScriptingFloatChannel, const [1,1]}
          Color 2R {type=MovieSceneScriptingFloatChannel, keys [[46,0.2582],[66.055,0.7482]]}
          Color 2G {type=MovieSceneScriptingFloatChannel, keys [[46,0.5711],[66.055,0.2585]]}
          Color 2B {type=MovieSceneScriptingFloatChannel, const [1,1]}
          Color 2A {type=MovieSceneScriptingFloatChannel, const [1,1]}
         ]
        }]
      }]
    }
    Background {
     id=#9
     possessed_class=Image
     tracks [{
       type=MovieSceneWidgetMaterialTrack
       sections [{
         type=MovieSceneParameterSection
         end_frame=101
         start_frame=-1
         channels [
          Blobs Opacity {type=MovieSceneScriptingFloatChannel, keys [[0,0],[5,0],[8,0.0298],[40,0.1],[62,0.6],[100,0.1]]}
          UnderGlow_Opacity {type=MovieSceneScriptingFloatChannel, keys [[0,0],[5,0],[8,0.3562],[40,1],[62,1],[100,0]]}
          Back Opacity {type=MovieSceneScriptingFloatChannel, keys [[0,0],[5,0],[8,1],[87,1],[100,0]]}
          Center Blob Opacity {type=MovieSceneScriptingFloatChannel, default=0.1, keys [[19,0],[21,1],[40,0],[100,0]]}
          Inverted Center Blob Opacity {type=MovieSceneScriptingFloatChannel, default=0, keys [[0,0],[30,0],[35,0.15],[62,0],[100,0]]}
          Centre Blob Position {type=MovieSceneScriptingFloatChannel, keys [[26,0],[33,-0.25]]}
          Centre Blob Scale {type=MovieSceneScriptingFloatChannel, keys [[19,0],[24,1],[56,4],[57,4]]}
          Color 4R {type=MovieSceneScriptingFloatChannel, keys [[0,0.001],[25,0.0091]]}
          Color 4G {type=MovieSceneScriptingFloatChannel, keys [[0,0.002],[25,0.0122]]}
          Color 4B {type=MovieSceneScriptingFloatChannel, keys [[0,0.005],[25,0.05]]}
          Color 4A {type=MovieSceneScriptingFloatChannel, const [1,2]}
         ]
        }]
      }]
    }
    SetUnlockedTag {
     id=#10
     possessed_class=UEFN_TextBlock_C
     tracks [
      {
       type=MovieScene2DTransformTrack
       sections [{
         type=MovieScene2DTransformSection
         end_frame=101
         start_frame=3
         channels [Translation.X {type=MovieSceneScriptingFloatChannel, keys [[25,157.9111],[84,-22.9844],[95,-413.0516]]}]
        }]
      }
      {
       type=MovieSceneFloatTrack
       display_name=Render Opacity
       sections [{
         type=MovieSceneFloatSection
         end_frame=94
         start_frame=0
         channels [None {type=MovieSceneScriptingFloatChannel, keys [[0,0],[25,-0.0079],[30,0.9888],[88,1.0066],[94,0]]}]
        }]
      }
      {
       type=MovieSceneWidgetMaterialTrack
       display_name=Font.FontMaterial
       sections [{
         type=MovieSceneParameterSection
         end_frame=100
         start_frame=0
         channels [
          WhiteGradient_Opacity {type=MovieSceneScriptingFloatChannel, keys [[5.7283,0.4],[26,0.4],[94,0.2]]}
          Color 1R {type=MovieSceneScriptingFloatChannel, keys [[49,0.1927],[59,0.8298]]}
          Color 1G {type=MovieSceneScriptingFloatChannel, keys [[49,1],[59,0.6617]]}
          Color 1B {type=MovieSceneScriptingFloatChannel, keys [[49,0.3534],[59,0.8391]]}
          Color 1A {type=MovieSceneScriptingFloatChannel, const [1,2]}
          Color 2R {type=MovieSceneScriptingFloatChannel, const [0.2865,1]}
          Color 2G {type=MovieSceneScriptingFloatChannel, const [0.5875,1]}
          Color 2B {type=MovieSceneScriptingFloatChannel, const [1,1]}
          Color 2A {type=MovieSceneScriptingFloatChannel, const [1,1]}
          Color 3R {type=MovieSceneScriptingFloatChannel, const [1,1]}
          Color 3G {type=MovieSceneScriptingFloatChannel, const [1,1]}
          Color 3B {type=MovieSceneScriptingFloatChannel, const [1,1]}
          Color 3A {type=MovieSceneScriptingFloatChannel, const [1,1]}
         ]
        }]
      }
     ]
    }
   ]
   tracks [{
     type=MovieSceneEventTrack
     sections [{
       type=MovieSceneEventTriggerSection
       channels [None {type=MovieSceneScriptingEventChannel, const ["<Struct 'MovieSceneEvent' (0x46916fc97e0) {}>", 1]}]
      }]
    }]
  }
 }]
graphs [
 AnimationFinishing {nodes [AnimationFinishing() {is=title K2Node_FunctionEntry}]}
 EventGraph {nodes [
   Event On Activated()
   > n1 = Play Animation with Finished event!(
    execute:Exec
    Widget:User Widget Object Reference=Self-Reference()
    InAnimation:Widget Animation Object Reference=Get Celebration() -> (Celebration:Widget Animation Object Reference)
    StartAtTime:float=0
    NumLoopsToPlay:int=1
    PlayMode:EUMGSequencePlayMode Enum
    PlaybackSpeed:float=1
   ) -> (then:Exec, Result:Widget Animation Handle Structure, Finished:Exec) {is=title K2Node_PlayAnimation2}
   > PlaySound2D(Sound=/SeasonPass/audio/MSS_UI_BP_Set_Complete)
   DeactivateWidget(execute=@n1.Finished)
  ]}
 Sequencer Events {nodes [
   FadingOutStarted() {is=title K2Node_CustomEvent}
   > Call Animation Finishing!(execute:Exec, self:WBP UI BPReboot Set Complete Object Reference) -> (then:Exec) {is=title K2Node_CallDelegate}
  ]}
 SetCharacterImage {nodes [
   n1 = SetCharacterImage!() -> (then, ThemeData:Season Pass Category Theme Data Structure) {is=title K2Node_FunctionEntry}
   > SetBrushFromImageReference!(
    execute
    self:Fort Lazy Image Object Reference=Get SetImage() -> (SetImage:Fort Lazy Image Object Reference)
    ImageReference:"Fort Image Reference Structure (by ref)"=LookupImage!(
     Table:"Fort Image Reference Table Structure (by ref)"=Break Season Pass Category Theme Data!(SeasonPassCategoryThemeData:"Season Pass Category Theme Data Structure (by ref)"=@n1) -> (
      ImageReferenceTable:Fort Image Reference Table Structure
      ColorReferenceTable:Fort Color Reference Table Structure
      TextReferenceTable:Fort Text Reference Table Structure
     ) {is=title K2Node_BreakStruct}
     Name:String=LargePreview
    ) -> (ReturnValue:"Fort Image Reference Structure (by ref)")
    bMatchSize:bool=false
    DownloadPriority:ECMSDownloadPriority Enum=Normal
   ) -> (then)
  ]}
 SetSingleOrMulti {nodes [
   SetSingleOrMulti!() -> (then, NumSetsCompleted:int) {is=title K2Node_FunctionEntry}
   n1 = Get NumSetsCompleted() -> (NumSetsCompleted:int=0)
   > SetActiveWidgetIndex(
    self:Common Visibility Switcher Object Reference=Get SetsNumberSwitcher() -> (SetsNumberSwitcher:Common Visibility Switcher Object Reference)
    Index=Select(Option 0:int=0, Option 1:int=1, Index="integer > integer"(A=@n1, B=1)) -> (ReturnValue:int)
   )
   > "SetText (Text)"(
    self=Get MultipleSets() -> (MultipleSets:Text Block Object Reference)
    InText="To Text (Integer)"(Value=Get NumSetsCompleted() -> (NumSetsCompleted=0))
   )
   n2 = Get NumSetsCompleted() -> (NumSetsCompleted=0)
   > SetVisibility(
    self=[
     Get SetName_Background() -> (SetName_Background:Text Block Object Reference)
     Get SetUnlockedTag() -> (SetUnlockedTag:Text Block Object Reference)
    ]
    InVisibility=Select(Option 0:ESlateVisibility Enum=Visible, Option 1:ESlateVisibility Enum=Collapsed, Index="integer > integer"(A=@n2, B=1)) -> (ReturnValue:ESlateVisibility Enum)
   )
   > Branch(Condition="integer > integer"(A=Get NumSetsCompleted() -> (NumSetsCompleted=0), B=1))
   > "SetText (Text)"(self=Get SetName() -> (SetName:Text Block Object Reference), InText="LOCTEXT(\"Sets Complete!\")")
  ]}
]
variables [
 AnimationFinishing {}
 NumSetsCompleted {type=integer}
]
widgets {
 info {inheritedWidgetCount=0, parentClass=&CommonActivatableWidget, rootWidgetClass=&Overlay, widgetCount=13}
 widgets [
  Overlay_0 {widgetClassPath=&Overlay, widget=Overlay_0}
  BackgroundBlur_51 {
   widgetClassPath=&BackgroundBlur
   parent=Overlay_0
   widget=BackgroundBlur_51
   properties {blurStrength=30}
   slot {class=OverlaySlot, properties {horizontalAlignment=HAlign_Fill, verticalAlignment=VAlign_Fill}}
  }
  Background {
   widgetClassPath=&Image
   bIsVariable=true
   parent=Overlay_0
   widget=Background
   properties {brush {
     bIsDynamicallyLoaded=false
     drawAs=Image
     imageType=FullColor
     mirroring=NoMirror
     resourceName=None
     resourceObject=&/M_UI_BPReboot_SetComplete_BG
     tiling=NoTile
     imageSize=<32 32>
     margin=<ltrb 0 0 0 0>
     outlineSettings {
      bUseBrushTransparency=false
      roundingType=HalfHeightRadius
      width=0
      color {colorUseRule=UseColor_Specified, specifiedColor=<rgba 0 0 0 0>}
      cornerRadii=<0 0 0 0>
     }
     tintColor {colorUseRule=UseColor_Specified, specifiedColor=<rgba 1 1 1 1>}
     uVRegion {bIsValid=false, max=<0 0>, min=<0 0>}
    }}
   slot {class=OverlaySlot, properties {horizontalAlignment=HAlign_Fill, verticalAlignment=VAlign_Fill}}
  }
  Overlay_1 {
   widgetClassPath=&Overlay
   parent=Overlay_0
   widget=Overlay_1
   slot {class=OverlaySlot, properties {horizontalAlignment=HAlign_Fill, verticalAlignment=VAlign_Fill}}
  }
  VerticalBox_0 {
   widgetClassPath=&VerticalBox
   parent=Overlay_1
   widget=VerticalBox_0
   slot {class=OverlaySlot, properties {horizontalAlignment=HAlign_Fill, verticalAlignment=VAlign_Center}}
  }
  Modal {
   widgetClassPath=&Image
   bIsVariable=true
   parent=VerticalBox_0
   widget=Modal
   properties {brush {
     bIsDynamicallyLoaded=false
     drawAs=Image
     imageType=FullColor
     mirroring=NoMirror
     resourceName=None
     resourceObject=&/M_UI_BPReboot_SetComplete_Modal
     tiling=NoTile
     imageSize=<1920 560>
     margin=<ltrb 0 0 0 0>
     outlineSettings {
      bUseBrushTransparency=false
      roundingType=HalfHeightRadius
      width=0
      color {colorUseRule=UseColor_Specified, specifiedColor=<rgba 0 0 0 0>}
      cornerRadii=<0 0 0 0>
     }
     tintColor {colorUseRule=UseColor_Specified, specifiedColor=<rgba 1 1 1 1>}
     uVRegion {bIsValid=false, max=<0 0>, min=<0 0>}
    }}
   slot {class=VerticalBoxSlot, properties {padding=<ltrb 0 0 0 -92>}}
  }
  Icon {
   widgetClassPath=&Image
   bIsVariable=true
   parent=VerticalBox_0
   widget=Icon
   properties {brush {
     bIsDynamicallyLoaded=false
     drawAs=Image
     imageType=FullColor
     mirroring=NoMirror
     resourceName=None
     resourceObject=&/M_UI_BPReboot_SetComplete_Check
     tiling=NoTile
     imageSize=<64 64>
     margin=<ltrb 0 0 0 0>
     outlineSettings {
      bUseBrushTransparency=false
      roundingType=HalfHeightRadius
      width=0
      color {colorUseRule=UseColor_Specified, specifiedColor=<rgba 0 0 0 0>}
      cornerRadii=<0 0 0 0>
     }
     tintColor {colorUseRule=UseColor_Specified, specifiedColor=<rgba 1 1 1 1>}
     uVRegion {bIsValid=false, max=<0 0>, min=<0 0>}
    }}
   slot {class=VerticalBoxSlot, properties {horizontalAlignment=HAlign_Center, verticalAlignment=VAlign_Center}}
  }
  SetName {
   widgetClassPath=&UEFN_TextBlock
   bIsVariable=true
   parent=VerticalBox_0
   widget=SetName
   properties {justification=Center, text=Voyager's Cube}
   slot {class=VerticalBoxSlot}
  }
  SetUnlockedTag {
   widgetClassPath=&UEFN_TextBlock
   bIsVariable=true
   parent=VerticalBox_0
   widget=SetUnlockedTag
   properties {
    justification=Center
    text="Set complete!"
    font {
     bForceMonospaced=false
     bMaterialIsStencil=false
     fontMaterial=&/MI_UI_BPReboot_BPComplete_Font
     fontObject=&HeadingNow
     letterSpacing=10
     monospacedWidth=1
     size=20.25
     skewAmount=0
     typefaceFontName=86 Bold
     outlineSettings {
      bApplyOutlineToDropShadows=false
      bMiteredCorners=false
      bSeparateFillAlpha=false
      outlineMaterial=null
      outlineSize=0
      outlineColor=<rgba 0 0 0 1>
     }
    }
   }
   slot {class=VerticalBoxSlot}
  }
  SetName_Background {
   widgetClassPath=&UEFN_TextBlock
   bIsVariable=true
   parent=Overlay_1
   widget=SetName_Background
   properties {
    justification=Center
    text=Voyager's Cube
    font {
     bForceMonospaced=false
     bMaterialIsStencil=true
     fontMaterial=/SeasonPass/Widgets/Reboot/SetUnlock/M_UI_BPReboot_BPUnlocking_Font_Outline
     fontObject=&HeadingNow
     letterSpacing=-40
     monospacedWidth=1
     size=72.75
     skewAmount=0
     typefaceFontName=86 Bold Italic
     outlineSettings {
      bApplyOutlineToDropShadows=false
      bMiteredCorners=false
      bSeparateFillAlpha=false
      outlineMaterial=null
      outlineSize=0
      outlineColor=<rgba 0 0 0 1>
     }
    }
    renderTransform {angle=0, scale=<3 3>, shear=<0 0>, translation=<0 0>}
   }
   slot {class=OverlaySlot, properties {horizontalAlignment=HAlign_Center, verticalAlignment=VAlign_Center, padding=<ltrb 0 0 0 60>}}
  }
  SetsNumberSwitcher {
   widgetClassPath=&CommonVisibilitySwitcher
   bIsVariable=true
   parent=Overlay_1
   widget=SetsNumberSwitcher
   slot {class=OverlaySlot, properties {horizontalAlignment=HAlign_Center, verticalAlignment=VAlign_Center, padding=<ltrb 0 0 0 100>}}
  }
  SetImage {
   widgetClassPath=&FortLazyImage
   bIsVariable=true
   parent=SetsNumberSwitcher
   widget=SetImage
   properties {
    materialTextureParamName=SetImage
    brush {
     bIsDynamicallyLoaded=false
     drawAs=Image
     imageType=FullColor
     mirroring=NoMirror
     resourceName=None
     resourceObject=/SeasonPass/Widgets/Reboot/SetUnlock/M_UI_BPReboot_SetUnlocking_Character
     tiling=NoTile
     imageSize=<720 720>
     margin=<ltrb 0 0 0 0>
     outlineSettings {
      bUseBrushTransparency=false
      roundingType=HalfHeightRadius
      width=0
      color {colorUseRule=UseColor_Specified, specifiedColor=<rgba 0 0 0 0>}
      cornerRadii=<0 0 0 0>
     }
     tintColor {colorUseRule=UseColor_Specified, specifiedColor=<rgba 1 1 1 1>}
     uVRegion {bIsValid=false, max=<0 0>, min=<0 0>}
    }
   }
   slot {class=CommonVisibilitySwitcherSlot}
  }
  MultipleSets {
   widgetClassPath=&UEFN_TextBlock
   bIsVariable=true
   parent=SetsNumberSwitcher
   widget=MultipleSets
   properties {
    justification=Center
    text=9
    font {
     bForceMonospaced=false
     bMaterialIsStencil=false
     fontMaterial=&/M_UI_BPReboot_BPComplete_Font_Number
     fontObject=&HeadingNow
     letterSpacing=0
     monospacedWidth=1
     size=351
     skewAmount=0
     typefaceFontName=75 Medium Italic
     outlineSettings {
      bApplyOutlineToDropShadows=false
      bMiteredCorners=false
      bSeparateFillAlpha=false
      outlineMaterial=null
      outlineSize=0
      outlineColor=<rgba 0 0 0 1>
     }
    }
    strikeBrush {
     bIsDynamicallyLoaded=false
     drawAs=Image
     imageType=FullColor
     mirroring=NoMirror
     resourceName=None
     resourceObject=null
     tiling=NoTile
     imageSize=<32 32>
     margin=<ltrb 0 0 0 0>
     outlineSettings {
      bUseBrushTransparency=false
      roundingType=HalfHeightRadius
      width=0
      color {colorUseRule=UseColor_Specified, specifiedColor=<rgba 0 0 0 0>}
      cornerRadii=<0 0 0 0>
     }
     tintColor {colorUseRule=UseColor_Specified, specifiedColor=<rgba 1 1 1 1>}
     uVRegion {bIsValid=false, max=<0 0>, min=<0 0>}
    }
   }
   slot {class=CommonVisibilitySwitcherSlot, properties {padding=<ltrb 0 80 0 0>}}
  }
 ]
}
