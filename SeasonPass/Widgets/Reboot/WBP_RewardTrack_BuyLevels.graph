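# Season pass dialog for buying levels with V-Bucks, with ±1/±10 steppers and a discounted bundle toggle.
# SetLevelsToBuy clamps the count and refreshes price, rewards/sets disclaimer and confirm button, which switches between "Buy Levels" and "Get V-Bucks".
# Confirm purchases a level/bundle offer map, after a CPC dialog when right-of-withdrawal applies; failure broadcasts OnGoingToBuyVBucksPage.
# A clamped counter pitches stepper sounds by press direction and streak.
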WidgetBlueprint /SeasonPass/Widgets/Reboot/WBP_RewardTrack_BuyLevels
graphs [
 AdjustVisualsToBundleButton {nodes [
   AdjustVisualsToBundleButton() {is=title K2Node_FunctionEntry}
   n1 = Get Block_Text!(self:WBP UIKit Button Generic Base Object Reference=Get Button_BundleToggle() -> (Button_BundleToggle:WBP UIKit Button With Toggle Object Reference)) -> (Block_Text:WBP UIKit Block Text Object Reference)
   > SetWrapping!(execute, self:WBP UIKit Block Text Object Reference=@n1, IsAutoWrap:bool=true, WrapTextAt:int=0) -> (then)
   > n2 = Cast To GridSlot!(execute, Object=Get Slot!(self:Widget Object Reference=@n1) -> (Slot:Panel Slot Object Reference)) -> (then, CastFailed, AsGrid Slot:Grid Slot Object Reference) {is=title K2Node_DynamicCast}
   > SetVerticalAlignment(self:Grid Slot Object Reference=@n2, InVerticalAlignment=VAlign_Center)
  ]}
 AudioIncrementDecrementCounter {nodes [
   n1 = AudioIncrementDecrementCounter!() -> (then, bIsIncrement:bool=~) {is=title K2Node_FunctionEntry}
   > n2 = Branch(Condition=@n1)
   > n3 = Increment Int(Value=Get AudioRewardCounter() -> (AudioRewardCounter:int=0))
   n4 = SelectInt(A=1, B=@n3, bPickA="Equal (Integer)"(A=@n3))
   n5 = Decrement Int(" "=@n2.else, Value=Get AudioRewardCounter() -> (AudioRewardCounter=0))
   n6 = SelectInt(A=-1, B=@n5, bPickA="Equal (Integer)"(A=@n5))
   n7 = Get bIsIncrement() -> (bIsIncrement:bool)
   Set AudioRewardCounter(
    execute=[@n3."  ", @n5."  "]
    AudioRewardCounter:int="Clamp (Integer)"(Value=SelectInt(A=@n4, B=@n6, bPickA=@n7), Min=-5, Max=5) -> (ReturnValue=0)
   ) -> (Output_Get:int=0)
   > n8 = SpawnSound2D!(
    execute
    Sound:Sound Base Object Reference=/Game/Sounds/Fort_UI_Sounds/FrontEnd/DefaultSounds/PressedButtons/MSS_UI_Button_Pressed_RewardAddSubtract
    VolumeMultiplier:float=1
    PitchMultiplier:float=1
    StartTime:float=0
    ConcurrencySettings:Sound Concurrency Object Reference
    bPersistAcrossLevelTransition:bool=false
    bAutoDestroy:bool=true
   ) -> (then, ReturnValue:Audio Component Object Reference)
   > Set Integer Parameter!(
    execute
    self:Audio Parameter Controller Interface Interface=@n8
    InName:Name=Pitch
    InInt:int=Get AudioRewardCounter() -> (AudioRewardCounter=0)
   ) -> (then) {is=title K2Node_CallFunction}
   > Execute Trigger Parameter!(execute, self:Audio Parameter Controller Interface Interface=@n8, InName:Name=Play) -> (then)
  ]}
 CanBuyRewards {nodes [
   n1 = CanBuyRewards!() -> (then, Preload Complete:bool=~, RewardsToBuy:int) {is=title K2Node_FunctionEntry}
   > Return Node!(execute, Result:bool=AND Boolean(A=@n1, B="integer > integer"(A=@n1.RewardsToBuy)))
  ]}
 CheckIsBuyingBundle {nodes [
   CheckIsBuyingBundle() {is=title K2Node_FunctionEntry}
   n1 = Get HB_PurchaseBundle() -> (HB_PurchaseBundle:Horizontal Box Object Reference)
   n2 = GetLevelsToGrant!(self:Season Pass Bundle Offer VM Object Reference=Get LevelBundleOffer!(self:Season Pass VM Object Reference=Get SeasonPassVM() -> (SeasonPassVM:Season Pass VM Object Reference)) -> (LevelBundleOffer:Season Pass Bundle Offer VM Object Reference)) -> (ReturnValue:int=0)
   n3 = Get RewardsToBuy() -> (RewardsToBuy:int=0)
   > n4 = Set IsBuyingBundle(IsBuyingBundle:bool=AND Boolean(A="integer >= integer"(A=@n3, B=@n2), B=IsVisible(self=@n1))) -> (Output_Get:bool)
   n5 = Get Block_Toggle!(self:WBP UIKit Button With Toggle Object Reference=Get Button_BundleToggle()) -> (Block_Toggle:WBP UIKit Block Toggle Object Reference)
   > SetIsSelected!(execute, self:WBP UIKit Block Base Object Reference=@n5, IsSelected:bool=@n4) -> (then)
   > SetSelected!(execute, self:WBP UIKit Block Toggle Object Reference=@n5, Selected:bool=@n4) -> (then)
  ]}
 EventGraph {nodes [
   n1 = "On Clicked (Button_Add_1)"!() -> (OutputDelegate, then, Button:Common Button Base Object Reference) {is=title K2Node_ComponentBoundEvent}
   > Fire Event Element Interacted(Element=@n1.Button)
   > SetLevelsToBuy!(execute, self, Amount:int=1, IsAdditive:bool=true) -> (then)
   > AudioIncrementDecrementCounter!(execute, self, bIsIncrement:bool=true) -> (then) {is=title K2Node_CallFunction}
   n2 = "On Clicked (Button_BundleToggle)"!() -> (OutputDelegate, then, Button:Common Button Base Object Reference) {is=title K2Node_ComponentBoundEvent}
   > n3 = Fire Event Element Interacted(Element=@n2.Button)
   n4 = Branch(execute=[@n12.then, @n3.then], Condition=Get IsBuyingBundle() -> (IsBuyingBundle:bool))
   n5 = GetLevelsToGrant(self=Get LevelBundleOffer(self=Get SeasonPassVM())) -> (ReturnValue=0)
   > n6 = Set with Broadcast RewardsToBuy!(execute, RewardsToBuy:int=int - int(A=@n5)) -> (then, Output_Get:int=0)
   SetLevelsToBuy(execute=[@n6.then, @n7.then], Amount=0, IsAdditive=true)
   n7 = Set with Broadcast RewardsToBuy(execute=@n4.else, RewardsToBuy=@n5) -> (Output_Get=0)
   n8 = "On Clicked (Button_Add_10)"!() -> (OutputDelegate, then, Button:Common Button Base Object Reference) {is=title K2Node_ComponentBoundEvent}
   > Fire Event Element Interacted(Element=@n8.Button)
   > SetLevelsToBuy(Amount=10, IsAdditive=true)
   > AudioIncrementDecrementCounter(bIsIncrement=true)
   n9 = "On Clicked (Button_Subtract_1)"!() -> (OutputDelegate, then, Button:Common Button Base Object Reference) {is=title K2Node_ComponentBoundEvent}
   > Fire Event Element Interacted(Element=@n9.Button)
   > SetLevelsToBuy(Amount=-1, IsAdditive=true)
   > AudioIncrementDecrementCounter(bIsIncrement=false)
   n10 = "On Clicked (Button_Subtract_10)"!() -> (OutputDelegate, then, Button:Common Button Base Object Reference) {is=title K2Node_ComponentBoundEvent}
   > Fire Event Element Interacted(Element=@n10.Button)
   > SetLevelsToBuy(Amount=-10, IsAdditive=true)
   > AudioIncrementDecrementCounter(bIsIncrement=false)
   n11 = "On Clicked (Button_BundleTogglePrompt)"!() -> (OutputDelegate, then, Button:Common Button Base Object Reference) {is=title K2Node_ComponentBoundEvent}
   > n12 = Fire Event Element Interacted(Element=@n11.Button)
   n13 = OnDialogButtonClicked!() -> (OutputDelegate, then, ButtonIndex:int) {is=title K2Node_CustomEvent}
   > n14 = Switch on Int!(execute:Exec, Selection:int=@n13.ButtonIndex) -> (Default:Exec, 0:Exec, 1:Exec)
   FireEventElementInteractedByName!(
    execute=@n14.1
    Element Name:String=Button_Cancel
    WidgetCallingFrom:Widget Object Reference=Get WBP_UIKit_Dialog_Internal() -> (WBP_UIKit_Dialog_Internal:WBP UIKit Dialog Internal Object Reference)
   ) -> (then, Success:bool)
   > DeactivateWidget()
   FireEventElementInteractedByName(execute=@n14.0, Element Name=Button_Buy_Levels, WidgetCallingFrom=Get WBP_UIKit_Dialog_Internal())
   > n15 = ShouldShowRightOfWithdrawalModal!(execute, Context:Widget Object Reference) -> (then, ReturnValue:bool)
   > n16 = Branch(Condition=@n15)
   > n17 = DeactivateWidget()
   n18 = OnCpcDismissed!() -> (OutputDelegate, then, Accepted:bool=~) {is=title K2Node_CustomEvent}
   n19 = Self-Reference()
   OnCpcDialogRequested!(
    execute=@n17.then
    Outer:Object Reference=@n19
    BodyText:Text="LOCTABLE(\"/SeasonPass/Data/ST_Passes\", \"CPC_BattlePassLevels\")"
    OnCpcDialogDismissed:"Delegate (by ref)"=@n18
    Target:Widget Object Reference=@n19
    Context:User Widget Object Reference=@n19
   ) -> (then)
   n20 = PurchaseMultipleSeasonPassOffers!(
    execute:Exec=[@n26.then, @n16.else]
    SeasonPass:Season Pass VM Object Reference=Get SeasonPassVM()
    OffersWithQuantities:Map of Fort Simple Offer VM Object References to Integers=Make Offer Map!() -> (Map:Map of Fort Simple Offer VM Object References to Integers)
   ) -> (then:Exec, OnOperationCompleted:Exec, Success:bool=~) {is=title K2Node_AsyncAction}
   n21 = Branch(execute=@n20.OnOperationCompleted, Condition=@n20)
   n22 = Branch(execute=[@n23.then, @n21.then], Condition=@n15)
   DeactivateWidget(execute=@n22.else)
   n23 = Call On Going to Buy VBucks Page!(execute:Exec=@n21.else, self:WBP Reward Track Buy Levels Object Reference) -> (then:Exec) {is=title K2Node_CallDelegate}
   n24 = Event PreConstruct()
   > SetupDialog()
   > n25 = Branch(Condition=@n24.IsDesignTime)
   SetGridSlotColumnAndRow!(
    execute=@n25.else
    Widget:Widget Object Reference=Get Prompt!(self:WBP UIKit Button CTA Base Object Reference=Get Button_Add_1() -> (Button_Add_1:WBP UIKit Button Regular Object Reference)) -> (Prompt:Grid Panel Object Reference)
    Row:int=1
    Row Span:int=0
    Column:int=4
    Column Span:int=0
   ) -> (then)
   > SetGridSlotColumnAndRow(
    Widget=Get Prompt(self=Get Button_Add_10() -> (Button_Add_10:WBP UIKit Button Regular Object Reference))
    Row=1
    Row Span=0
    Column=4
    Column Span=0
   )
   > SetCurrency!(
    execute
    self:WBP Price Switcher Base Object Reference=Get WBP_PriceSwitcher_Base() -> (WBP_PriceSwitcher_Base:WBP Price Switcher Base Object Reference)
    Currency:EFortPriceCurrency Enum=Mtx
   ) -> (then)
   > AdjustVisualsToBundleButton() {is=title K2Node_CallFunction}
   > SetCompactUIStyle() {is=title K2Node_CallFunction}
   > Branch(Condition=IsNintendoSwitchOrNintendoSwitchPreview!() -> (ReturnValue:bool))
   > SetBrightness!(
    execute
    self:WBP UIKit Backplate Base Object Reference=Get ContentBackplate!(self:WBP UIKit Dialog Internal Object Reference=Get WBP_UIKit_Dialog_Internal()) -> (ContentBackplate:WBP UIKit Backplate Base Object Reference)
    Brightness:E_UI_BackplateBrightness Enum=NewEnumerator3
   ) -> (then)
   Event On Activated()
   > Fire Event Sub View Opened!(execute, View:Widget Object Reference=Self-Reference(), AdditionalInfo:Flow Analytics Additional Info Structure, bIsOverlay:bool=true) -> (then)
   > Bind Event to Dialog Button Clicked Dispatcher(self:WBP UIKit Dialog Internal Object Reference=Get WBP_UIKit_Dialog_Internal(), Delegate=@n13) {is=title K2Node_AddDelegate}
   > Set with Broadcast RewardsToBuy(RewardsToBuy=InitAdditionalRewards!(self) -> (ReturnValue:int=0)) -> (Output_Get=0)
   > SetLevelsToBuy(Amount=0, IsAdditive=true)
   > SetIsToggleable!(execute, self:Common Button Base Object Reference=Get Button_BundleToggle(), bInIsToggleable:bool=false) -> (then)
   Event On Deactivated()
   > Unbind Event from Dialog Button Clicked Dispatcher(self:WBP UIKit Dialog Internal Object Reference=Get WBP_UIKit_Dialog_Internal(), Delegate=@n13) {is=title K2Node_RemoveDelegate}
   > Set with Broadcast RewardsToBuy(RewardsToBuy=1) -> (Output_Get=0)
   n26 = Branch(execute=@n18.then, Condition=@n18.Accepted)
  ]}
 GetBundleText {nodes [
   n1 = GetBundleText!() -> (then, FinalPrice:int, RegularPrice:Text, BundleSize:int, LoadingState:EFortMtxStoreLoadingState Enum=Uninitialized) {is=title K2Node_FunctionEntry}
   n2 = Get FortMtxStoreVM() -> (FortMtxStoreVM:Fort Mtx Store VM Object Reference)
   n3 = GetRealPriceText!(
    FortMTXStoreVM:Fort Mtx Store VM Object Reference=@n2
    LoadingState:EFortMtxStoreLoadingState Enum=Get LoadingState!(self:Fort Mtx Store VM Object Reference=@n2) -> (LoadingState:EFortMtxStoreLoadingState Enum=Uninitialized)
    MtxPrice:int=Get FinalPrice() -> (FinalPrice:int=0)
    HasRealPrice:bool=true
    Currency:EFortPriceCurrency Enum=Mtx
    IsPriceShowing:bool=true
    IsAdditionalText:bool=false
   ) -> (Text:Text)
   n4 = AND Boolean(
    A=NOT Boolean(A=TextIsEmpty(InText=@n3))
    B="Equal (Enum)"(A:EFortMtxStoreLoadingState Enum=@n1.LoadingState, B:EFortMtxStoreLoadingState Enum=Ready)
   )
   n5 = Format Text!(Format="LOCTEXT(\"(normally {RegularPrice})\")", RegularPrice:Text=@n1.RegularPrice) -> (Result)
   > Return Node(Result:Text=Format Text!(
     Format="LOCTEXT(\"Buy a bundle of {LevelsAmountInBundle} Levels at a reduced price of {FinalPrice} V-Bucks {ExtraText}\")"
     LevelsAmountInBundle:int=Get BundleSize() -> (BundleSize:int=0)
     FinalPrice:int=Get FinalPrice() -> (FinalPrice=0)
     ExtraText:Text=Select(Option 0:Text=@n5, Option 1:Text=@n3, Index=@n4) -> (ReturnValue:Text)
    ) -> (Result))
  ]}
 GetDisclaimerText {nodes [
   n1 = GetDisclaimerText!() -> (then, Rewards:int, Sets:int) {is=title K2Node_FunctionEntry}
   n2 = "integer > integer"(A=@n1.Sets)
   > Return Node(Result=Format Text!(
     Format=Select(
      Option 0="LOCTEXT(\"You\\'ll be able to claim <text color=\\\"000000B3\\\" fontface=\\\"black\\\">{Rewards} {Rewards}|plural(one=REWARD,other=REWARDS)</>\")"
      Option 1="LOCTEXT(\"You\\'ll be able to claim <text color=\\\"000000B3\\\" fontface=\\\"black\\\">{Rewards} {Rewards}|plural(one=REWARD,other=REWARDS)</> and unlock <text color=\\\"000000B3\\\" fontface=\\\"black\\\">{Sets} {Sets}|plural(one=SET,other=SETS)</>\")"
      Index=@n2
     )
     Rewards:int=Get Rewards() -> (Rewards:int=0)
     Sets:int=Get Sets() -> (Sets:int=0)
    ) -> (Result))
  ]}
 GetRealPriceText {nodes [
   n1 = GetRealPriceText!() -> (then, LoadingState:EFortMtxStoreLoadingState Enum=Uninitialized, MtxPrice:int) {is=title K2Node_FunctionEntry}
   n2 = Get MtxPrice() -> (MtxPrice:int=0)
   n3 = Get FortMtxStoreVM()
   > Return Node!(
    execute
    Text:Text=Select(
     Option 1=GetRealPriceText!(self:Fort Mtx Store VM Object Reference=@n3, MtxPrice=@n2, bIsAdditionalText:bool=false) -> (ReturnValue:Text) {is=title K2Node_CallFunction}
     Index="Equal (Enum)"(A=@n1, B=Ready)
    )
   )
  ]}
 GetRewardsForPreloading {nodes [
   n1 = GetRewardsForPreloading!() -> (then, Rewards:Array of Season Pass Reward VM Object References, RewardsToBuy:int, AutoClaimOn:bool=~) {is=title K2Node_FunctionEntry}
   > n2 = Branch(Condition=@n1.AutoClaimOn)
   n3 = Get Rewards() -> (Rewards:Array of Season Pass Reward VM Object References)
   > n4 = For Loop(LastIndex="Min (Integer)"(A=Length(TargetArray:Array of Season Pass Reward VM Object References=@n3), B=Get RewardsToBuy() -> (RewardsToBuy=0)) -> (ReturnValue=0))
   Return Node!(
    execute=[@n2.else, @n4.Completed]
    AssetsToPreload:Array of Soft Object References=Get Results() -> (Results:Array of Soft Object References)
   )
   n5 = Get Rewards()
   > Add(
    TargetArray:Array of Soft Object References=Get Results()
    NewItem:"Soft Object Reference (by ref)"=GetButtonIcon!(self:Season Pass Reward VM Object Reference="Get (a copy)"(Array:Array of Season Pass Reward VM Object References=@n5, Dimension 1=@n4) -> (Output:Season Pass Reward VM Object Reference)) -> (ReturnValue:Soft Object Reference)
   ) {is=title K2Node_CallArrayFunction}
  ]}
 InitAdditionalRewards {nodes [
   InitAdditionalRewards() {is=title K2Node_FunctionEntry}
   > n1 = Is Valid(InputObject=Get ItemToPurchase() -> (ItemToPurchase:Season Pass Reward VM Object Reference))
   Return Node!(execute=@n1.Is Not Valid, ReturnValue:int=1)
   > Return Node(ReturnValue=GetRewardsRequiredToUnlock!(self:Season Pass Reward VM Object Reference=Get ItemToPurchase()) -> (ReturnValue:int=0))
  ]}
 Make Offer Map {nodes [
   Make Offer Map() {is=title K2Node_FunctionEntry}
   > n1 = Branch(Condition=Get IsBuyingBundle())
   n2 = GetLevelsToGrant(self=Get LevelBundleOffer(self=Get SeasonPassVM())) -> (ReturnValue=0)
   n3 = Get LevelsToBuy() -> (LevelsToBuy:int=0)
   n4 = Get SeasonPassVM()
   > Return Node!(
    execute
    Map:Map of Fort Simple Offer VM Object References to Integers=Make Map!(
     Key 0:Fort Simple Offer VM Object Reference=Get LevelOffer!(self:Season Pass VM Object Reference=@n4) -> (LevelOffer:Fort Simple Offer VM Object Reference)
     Value 0:int=int - int(A=@n3, B=@n2)
     Key 1:Fort Simple Offer VM Object Reference=Get LevelBundleOffer(self=@n4)
     Value 1:int=1
    ) -> (Map:Map of Fort Simple Offer VM Object References to Integers) {is=title K2Node_MakeMap}
   )
   n5 = Get SeasonPassVM()
   Return Node(execute=@n1.else, Map=Make Map!(Key 0=Get LevelOffer(self=@n5), Value 0=Get LevelsToBuy() -> (LevelsToBuy=0)) -> (Map))
  ]}
 OnGoingToBuyVBucksPage {nodes [OnGoingToBuyVBucksPage() {is=title K2Node_FunctionEntry}]}
 Set Available Rewards To Buy {nodes [
   n1 = Set Available Rewards To Buy!() -> (then, AvailableRewardsToBuy:int) {is=title K2Node_FunctionEntry}
   > Set AvailableRewardsToBuy(AvailableRewardsToBuy:int=@n1) -> (Output_Get:int=0)
   > SetLevelsToBuy(Amount=0, IsAdditive=true)
  ]}
 SetCompactUIStyle {nodes [
   SetCompactUIStyle() {is=title K2Node_FunctionEntry}
   > Branch(Condition=IsUsingCompactUI())
   n1 = Get WBP_UIKit_Dialog_Internal()
   > SetSizeBox!(
    execute
    self:WBP UIKit Dialog Internal Object Reference=@n1
    Data:"S UI Size Box Structure (by ref)"=Make S UI Size Box!(
     WidthOverride_107_#1:S UI Bool Float Structure=(Enable_3_#2=False,Value_4_#3=0)
     HeightOverride_108_#4:S UI Bool Float Structure=(Enable_3_#2=False,Value_4_#3=0)
     MinDesiredWidth_109_#5:S UI Bool Float Structure=(Enable_3_#2=False,Value_4_#3=0)
     MinDesiredHeight_110_#6:S UI Bool Float Structure=(Enable_3_#2=False,Value_4_#3=0)
     MaxDesiredWidth_111_#7:S UI Bool Float Structure=(Enable_3_#2=False,Value_4_#3=0)
     MaxDesiredHeight_112_#8:S UI Bool Float Structure=Make S UI Bool Float!(Enable_3_#2:bool=true, Value_4_#3:double=872) -> (S_UI_BoolFloat:S UI Bool Float Structure) {is=title K2Node_MakeStruct}
     MinAspectRatio_113_#9:S UI Bool Float Structure=(Enable_3_#2=False,Value_4_#3=0)
     MaxAspectRatio_114_#10:S UI Bool Float Structure=(Enable_3_#2=False,Value_4_#3=0)
    ) -> (S_UI_SizeBox:S UI Size Box Structure) {is=title K2Node_MakeStruct}
   ) -> (then)
   > SetPlacement!(
    execute
    self:WBP UIKit Dialog Internal Object Reference=@n1
    Widget:Widget Object Reference=Get SizeBox_Dialog!(self:WBP UIKit Dialog Internal Object Reference=@n1) -> (SizeBox_Dialog:Size Box Object Reference)
    Placement:S UI Placement Structure=Make S UI Placement!(
     HorizontalAlignment_4_#11:EHorizontalAlignment Enum=HAlign_Center
     VerticalAlignment_5_#12:EVerticalAlignment Enum=VAlign_Center
     Padding_8_#13:Margin Structure=Make Margin(Left=0, Top=0, Right=0, Bottom=24)
     Grid_20_#14:S UI Grid Cell Structure=(Row_15_#11=0,RowSpan_17_#15=0,Column_19_#16=0,ColumnSpan_21_#17=0,Layer_23_#18=0,Nudge_26_#19=(X=0,Y=0))
    ) -> (S_UI_Placement:S UI Placement Structure) {is=title K2Node_MakeStruct}
   ) -> (then)
   > SetPadding!(
    execute
    self:Vertical Box Slot Object Reference=SlotAsVerticalBoxSlot(Widget=Get HB_PurchaseBundle())
    InPadding_Left:float=120
    InPadding_Top:float=24
    InPadding_Right:float=120
    InPadding_Bottom:float=24
   ) -> (then)
   > SetPadding(
    self=[
     SlotAsVerticalBoxSlot(Widget=Get LegalText_RealPrice_PriceSwitcher() -> (LegalText_RealPrice_PriceSwitcher:WBP Real Price Legal Text Object Reference))
     SlotAsVerticalBoxSlot(Widget=Get LegalText_RealPrice_ReducedPrice() -> (LegalText_RealPrice_ReducedPrice:WBP Real Price Legal Text Object Reference))
    ]
    InPadding_Left=0
    InPadding_Top=4
    InPadding_Right=0
    InPadding_Bottom=0
   )
   > SetPadding(
    self:Overlay Slot Object Reference=SlotAsOverlaySlot(Widget=Get SZ_Legal() -> (SZ_Legal:Safe Zone Object Reference))
    InPadding_Left=48
    InPadding_Top=24
    InPadding_Right=48
    InPadding_Bottom=24
   )
  ]}
 SetLevelsToBuy {nodes [
   SetLevelsToBuy!() -> (then, Amount:int, IsAdditive:bool=~) {is=title K2Node_FunctionEntry}
   > n1 = Sequence()
   n2 = "integer < integer"(A=Get RewardsToBuy() -> (RewardsToBuy=0), B=Get AvailableRewardsToBuy() -> (AvailableRewardsToBuy:int=0))
   SetIsEnabled(execute=@n1.then_1, self=Get Button_Add_1(), bInIsEnabled=@n2)
   > SetIsEnabled(self=Get Button_Add_10(), bInIsEnabled=@n2)
   n3 = "integer >= integer"(A=int - int(A=Get RewardsToBuy() -> (RewardsToBuy=0)), B=1)
   > SetIsEnabled(self=Get Button_Subtract_1() -> (Button_Subtract_1:WBP UIKit Button Regular Object Reference), bInIsEnabled=@n3)
   > SetIsEnabled(self=Get Button_Subtract_10() -> (Button_Subtract_10:WBP UIKit Button Regular Object Reference), bInIsEnabled=@n3)
   > UpdateConfirmButtonAction()
   > UpdateDisclaimerText()
   n4 = Get Amount() -> (Amount:int=0)
   n5 = int + int(A=@n4, B=Get RewardsToBuy() -> (RewardsToBuy=0))
   n6 = Get IsAdditive() -> (IsAdditive:bool)
   n7 = Set with Broadcast RewardsToBuy(
    execute=@n1.then_0
    RewardsToBuy="Clamp (Integer)"(
     Value=Select(Option 0:int=@n4, Option 1:int=@n5, Index=@n6) -> (ReturnValue:int)
     Min=1
     Max=Get AvailableRewardsToBuy() -> (AvailableRewardsToBuy=0)
    ) -> (ReturnValue=0)
   ) -> (Output_Get=0)
   > "SetText (Text)"(
    self=Get Text_Levels() -> (Text_Levels:Common Text Object Reference)
    InText=Format Text!(Format="LOCTEXT(\"{RewardsAmount} Levels\")", RewardsAmount:int=@n7) -> (Result)
   )
   n8 = Get AvailableRewardsToBuy() -> (AvailableRewardsToBuy=0)
   n9 = Get RewardsToBuy() -> (RewardsToBuy=0)
   > "SetText (Text)"(
    self=Get Text_LevelsLeft() -> (Text_LevelsLeft:Common Text Object Reference)
    InText=Format Text(Format="LOCTEXT(\"{RewardsAmount} Left\")", RewardsAmount=int - int(A=@n8, B=@n9))
   )
   > CheckIsBuyingBundle() {is=title K2Node_CallFunction}
   > UpdateFinalPrice()
  ]}
 SetupDialog {nodes [
   SetupDialog() {is=title K2Node_FunctionEntry}
   > n1 = Construct Object from Class(Class=/SeasonPass/Widgets/Reboot/Data/DialogVM_BuyLevels.DialogVM_BuyLevels_C) -> (ReturnValue:Dialog VM Buy Levels Object Reference)
   > SetDialogData!(execute, self:WBP UIKit Dialog Internal Object Reference=Get WBP_UIKit_Dialog_Internal(), Data:UI Kit Dialog Viewmodel Object Reference=@n1) -> (then)
   > n2 = Sequence()
   n3 = Get ButtonWidgets!(self:WBP UIKit Dialog Internal Object Reference=Get WBP_UIKit_Dialog_Internal()) -> (ButtonWidgets:Array of WBP UIKit Button CTA Base Object References)
   SetIsFocusable!(
    execute=@n2.then_1
    self:Common Button Base Object Reference="Get (a copy)"(Array:Array of WBP UIKit Button CTA Base Object References=@n3, Dimension 1=1) -> (Output:WBP UIKit Button CTA Base Object Reference)
    bInIsFocusable:bool=false
   ) -> (then)
   n4 = Set ButtonConfirm(execute=@n2.then_0, ButtonConfirm:WBP UIKit Button CTA Base Object Reference="Get (a copy)"(Array=@n3)) -> (Output_Get:WBP UIKit Button CTA Base Object Reference)
   > SetIsFocusable(self=@n4, bInIsFocusable=false)
  ]}
 UpdateConfirmButtonAction {nodes [
   UpdateConfirmButtonAction() {is=title K2Node_FunctionEntry}
   n1 = "integer >= integer"(
    A=GetMtxBalance!(self:Fort Local Player VM Object Reference=Get FortLocalPlayerVM() -> (FortLocalPlayerVM:Fort Local Player VM Object Reference)) -> (ReturnValue:int=0)
    B=Get TotalCost() -> (TotalCost:int=0)
   )
   > SetText!(
    execute
    self:Fort CTAButton Object Reference=Get ButtonConfirm() -> (ButtonConfirm:WBP UIKit Button CTA Base Object Reference)
    ButtonText:"Text (by ref)"=SelectText(A="LOCTEXT(\"Buy Levels\")", B="LOCTEXT(\"Get V-Bucks\")", bPickA=@n1)
   ) -> (then)
   > n2 = ShouldShowRightOfWithdrawalModal()
   n3 = AND Boolean(A=NOT Boolean(A=@n2), B=@n1)
   > SetRequiresHold!(execute, self:Common Button Base Object Reference=Get ButtonConfirm(), bInRequiresHold:bool=@n3) -> (then)
   > SetTriggeringInputAction!(
    execute
    self:Common Button Base Object Reference=Get ButtonConfirm()
    InputActionRow_DataTable:Data Table Object Reference=&InputActionDataTable
    InputActionRow_RowName:Name=SelectName!(A:Name=AcceptHold, B:Name=AcceptQuantity, bPickA:bool=@n3) -> (ReturnValue:Name=None)
   ) -> (then)
   n4 = Get ButtonConfirm()
   > n5 = Set bTriggerHapticsOnActionCommited!(
    execute
    bTriggerHapticsOnActionCommited:bool=GetRequiresHold!(self:Common Button Base Object Reference=@n4) -> (ReturnValue:bool)
    self:UIKit Modular Button Object Reference=@n4
   ) -> (then, Output_Get:bool)
   > Set bTriggerHapticsOnPressed!(execute, bTriggerHapticsOnPressed:bool=@n5, self:UIKit Modular Button Object Reference=@n4) -> (then, Output_Get:bool)
  ]}
 UpdateDisclaimerText {nodes [
   UpdateDisclaimerText() {is=title K2Node_FunctionEntry}
   > Set RewardsUnlocked(RewardsUnlocked:int=GetNumberRewardsUnlockedFromLevelPurchase!(self:Season Pass VM Object Reference=Get SeasonPassVM(), LevelsToBuy:int=Get LevelsToBuy() -> (LevelsToBuy=0)) -> (ReturnValue:int=0)) -> (Output_Get:int=0)
   > Set SetsUnlocked(SetsUnlocked:int=GetNumberOfSetsUnlockedFromLevelPurchase!(self:Season Pass VM Object Reference=Get SeasonPassVM(), LevelsToBuy:int=Get LevelsToBuy() -> (LevelsToBuy=0)) -> (ReturnValue:int=0)) -> (Output_Get:int=0)
   n1 = "integer > integer"(A=Get SetsUnlocked() -> (SetsUnlocked:int=0))
   n2 = "integer > integer"(A=Get RewardsUnlocked() -> (RewardsUnlocked:int=0))
   > SetVisibility(
    self=Get Tag_Disclaimer() -> (Tag_Disclaimer:Border Object Reference)
    InVisibility=Select(Option 0:ESlateVisibility Enum=Collapsed, Option 1:ESlateVisibility Enum=SelfHitTestInvisible, Index=OR Boolean(A=@n2, B=@n1)) -> (ReturnValue:ESlateVisibility Enum)
   )
   > SetText!(
    execute
    self:Rich Text Block Object Reference=Get Text_Disclaimer() -> (Text_Disclaimer:Common Rich Text Block Object Reference)
    InText:"Text (by ref)"=GetDisclaimerText!(Rewards:int=Get RewardsUnlocked() -> (RewardsUnlocked=0), Sets:int=Get SetsUnlocked() -> (SetsUnlocked=0)) -> (Result:Text) {is=title K2Node_CallFunction}
   ) -> (then)
  ]}
 UpdateFinalPrice {nodes [
   UpdateFinalPrice() {is=title K2Node_FunctionEntry}
   n1 = Get SeasonPassVM()
   n2 = Get RewardsToBuy() -> (RewardsToBuy=0)
   > Set LevelsToBuy(LevelsToBuy:int="Min (Integer)"(A=GetNumberOfCurrentlyPurchaseableLevels!(self:Season Pass VM Object Reference=@n1) -> (ReturnValue:int=0), B=@n2) -> (ReturnValue=0)) -> (Output_Get:int=0)
   n3 = Get SeasonPassVM()
   > Set Regular Price(Regular Price:int=GetRegularPriceIntForQuantity!(self:Fort Simple Offer VM Object Reference=Get LevelOffer(self=@n3), Quantity:int=@n2) -> (ReturnValue:int=0)) -> (Output_Get:int=0)
   n4 = Get LevelBundleOffer(self=Get SeasonPassVM())
   > n5 = Branch(Condition=AND Boolean(A=Get IsBuyingBundle(), B=IsValid(Object=@n4)))
   n6 = Get LevelBundleOffer(self=Get SeasonPassVM())
   > Set Bundle Price(Bundle Price:int="To Integer (Integer64)"!(InInt:Integer64=GetPrice!(self:Fort Base Offer VM Object Reference=@n6) -> (ReturnValue:Integer64=0)) -> (ReturnValue:int=0)) -> (Output_Get:int=0)
   > n7 = Set Levels To Count For Final Price(Levels To Count For Final Price:int=int - int(A=Get LevelsToBuy() -> (LevelsToBuy=0), B=GetLevelsToGrant(self=@n6) -> (ReturnValue=0))) -> (Output_Get:int=0)
   n8 = Get Levels To Count For Final Price() -> (Levels To Count For Final Price:int=0)
   n9 = Get LevelOffer(self=Get SeasonPassVM())
   n10 = Set with Broadcast TotalCost!(
    execute=[@n7.then, @n13.then]
    TotalCost:int=int + int(
     A=Get Bundle Price() -> (Bundle Price:int=0)
     B=GetPriceIntForQuantity!(self:Fort Simple Offer VM Object Reference=@n9, Quantity:int=@n8) -> (ReturnValue:int=0)
    )
   ) -> (then, Output_Get:int=0)
   > SetPrice!(execute, self:WBP Price Switcher Base Object Reference=Get WBP_PriceSwitcher_Base(), Price:Text="To Text (Integer)"(Value=@n10)) -> (then)
   > SetRegularPrice!(
    execute
    self:WBP Price Switcher Base Object Reference=Get WBP_PriceSwitcher_Base()
    Price:Text="To Text (Integer)"(Value=Get Regular Price() -> (Regular Price:int=0))
   ) -> (then)
   n11 = Get Regular Price() -> (Regular Price=0)
   n12 = Get TotalCost() -> (TotalCost=0)
   > SetStrikethroughPriceVisibility!(
    execute
    self:WBP Price Switcher Base Object Reference=Get WBP_PriceSwitcher_Base()
    Visibility:ESlateVisibility Enum=Select(Option 0=SelfHitTestInvisible, Option 1=Collapsed, Index="Equal (Integer)"(A=@n11, B=@n12))
   ) -> (then)
   n13 = Set Levels To Count For Final Price(execute=@n5.else, Levels To Count For Final Price=Get LevelsToBuy() -> (LevelsToBuy=0)) -> (Output_Get=0)
  ]}
]
variables [
 AudioRewardCounter {type=integer}
 AvailableRewardsToBuy {type=integer, default=100}
 ButtonConfirm {type=WBP_UIKit_ButtonCTA_Base_C, default=null}
 IsBuyingBundle {type=boolean}
 ItemToPurchase {type=SeasonPassRewardVM, default=null}
 LevelsToBuy {type=integer}
 OnGoingToBuyVBucksPage {}
 PurchasableRewardsVM {type="array<SeasonPassRewardVM>", default []}
 RewardsToBuy {type=integer}
 TotalCost {type=integer, default=100}
]
widgets {
 info {
  inheritedWidgetCount=0
  namedSlotCount=2
  parentClass=&CommonActivatableWidget
  rootWidgetClass=/FortUILibrary/Systems/Dialog/WBP_UIKit_Dialog_Internal.WBP_UIKit_Dialog_Internal_C
  widgetCount=30
 }
 widgets [
  WBP_UIKit_Dialog_Internal {
   widgetClassPath=/FortUILibrary/Systems/Dialog/WBP_UIKit_Dialog_Internal
   bIsVariable=true
   widget=WBP_UIKit_Dialog_Internal
   properties {
    defaultHoldData=/FortUILibrary/Input/PurchaseInteraction_HoldData.PurchaseInteraction_HoldData_C
    buttons {
     each Button Horizontal Fill=false
     each Button Min Width=0
     each Button Size=Medium
     focuse Button=0
     space Between Buttons=30
     cTA Buttons Types [Loud,Regular]
     group Placement {
      horizontal Alignment=HAlign_Center
      vertical Alignment=VAlign_Bottom
      grid {column=1, column Span=0, layer=0, row=2, row Span=0, nudge=<0 0>}
      padding=<ltrb 80 32 80 48>
     }
    }
    dialogSIze {
     height Override {enable=false, value=800}
     max Aspect Ratio {enable=false, value=0}
     max Desired Height {enable=true, value=800}
     max Desired Width {enable=false, value=0}
     min Aspect Ratio {enable=false, value=0}
     min Desired Height {enable=false, value=0}
     min Desired Width {enable=false, value=0}
     width Override {enable=true, value=1300}
    }
    titleProperties {
     auto Wrap Text=true
     justification=Center
     line Height Percentage=1
     mobile Font Size Multiplier=1.3
     transform Policy=ToUpper
     wrap Text At=0
     color {colorUseRule=UseColor_Specified, specifiedColor=<rgba 1 1 1 1>}
     font {
      bForceMonospaced=false
      bMaterialIsStencil=false
      fontMaterial=null
      fontObject=&HeadingNow
      letterSpacing=0
      monospacedWidth=1
      size=32.25
      skewAmount=0
      typefaceFontName=86 Bold Italic
      outlineSettings {
       bApplyOutlineToDropShadows=false
       bMiteredCorners=false
       bSeparateFillAlpha=false
       outlineMaterial=null
       outlineSize=0
       outlineColor=<rgba 0 0 0 1>
      }
     }
     placement {
      horizontal Alignment=HAlign_Fill
      vertical Alignment=VAlign_Top
      grid {column=1, column Span=0, layer=0, row=0, row Span=0, nudge=<0 0>}
      padding=<ltrb 80 80 80 32>
     }
    }
   }
  }
  Overlay_4 {widgetClassPath=&Overlay, namedSlotHost=WBP_UIKit_Dialog_Internal, widget=Overlay_4}
  VB_All {
   widgetClassPath=&VerticalBox
   parent=Overlay_4
   widget=VB_All
   slot {class=OverlaySlot, properties {horizontalAlignment=HAlign_Fill, verticalAlignment=VAlign_Center, padding=<ltrb 60 0 60 0>}}
  }
  Overlay_3 {widgetClassPath=&Overlay, parent=VB_All, widget=Overlay_3, slot {class=VerticalBoxSlot}}
  ContentOverlay {
   widgetClassPath=&Image
   bIsVariable=true
   parent=Overlay_3
   widget=ContentOverlay
   properties {
    brush {
     bIsDynamicallyLoaded=false
     drawAs=Image
     imageType=FullColor
     mirroring=NoMirror
     resourceName=None
     resourceObject=/SeasonPass/Materials/RewardTrack/MI_UI_ModalOverlay
     tiling=NoTile
     imageSize=<32 32>
     margin=<ltrb 0 0 0 0>
     outlineSettings {
      bUseBrushTransparency=false
      roundingType=HalfHeightRadius
      width=0
      color {colorUseRule=UseColor_Specified, specifiedColor=<rgba 0 0 0 0>}
      cornerRadii=<0 0 0 0>
     }
     tintColor {colorUseRule=UseColor_Specified, specifiedColor=<rgba 1 1 1 1>}
     uVRegion {bIsValid=false, max=<0 0>, min=<0 0>}
    }
    colorAndOpacity=<rgba 1 1 1 0.2>
   }
   slot {class=OverlaySlot, properties {horizontalAlignment=HAlign_Fill, verticalAlignment=VAlign_Fill}}
  }
  VerticalBox_72 {
   widgetClassPath=&VerticalBox
   parent=Overlay_3
   widget=VerticalBox_72
   slot {class=OverlaySlot, properties {horizontalAlignment=HAlign_Fill}}
  }
  Controllers {
   widgetClassPath=&HorizontalBox
   parent=VerticalBox_72
   widget=Controllers
   slot {class=VerticalBoxSlot, properties {padding=<ltrb 24 12 24 12>}}
  }
  Button_Subtract_10 {
   widgetClassPath=&WBP_UIKit_Button_Regular
   bIsVariable=true
   parent=Controllers
   widget=Button_Subtract_10
   properties {
    bIsFocusable=false
    minHeight=48
    minWidth=112
    text=-10
    triggeringInputAction {dataTable=&InputActionDataTable, rowName=GenericLeftTrigger}
   }
   slot {class=HorizontalBoxSlot, properties {horizontalAlignment=HAlign_Center}}
  }
  Button_Subtract_1 {
   widgetClassPath=&WBP_UIKit_Button_Regular
   bIsVariable=true
   parent=Controllers
   widget=Button_Subtract_1
   properties {
    bIsFocusable=false
    minHeight=48
    minWidth=112
    text=-1
    triggeringInputAction {dataTable=&InputActionDataTable, rowName=GenericLeftShoulder}
   }
   slot {class=HorizontalBoxSlot, properties {horizontalAlignment=HAlign_Center, padding=<ltrb 16 0 0 0>}}
  }
  VerticalBox_68 {
   widgetClassPath=&VerticalBox
   parent=Controllers
   widget=VerticalBox_68
   slot {class=HorizontalBoxSlot, properties {verticalAlignment=VAlign_Center, padding=<ltrb 16 0 16 0>, size {sizeRule=Fill, value=1}}}
  }
  Text_Levels {
   widgetClassPath=&CommonTextBlock
   bIsVariable=true
   parent=VerticalBox_68
   widget=Text_Levels
   properties {
    autoWrapText=true
    justification=Center
    text=XX Levels
    textTransformPolicy=ToUpper
    colorAndOpacity {colorUseRule=UseColor_Specified, specifiedColor=<rgba 0.5776 0.5776 0.5776 1>}
    font {
     bForceMonospaced=false
     bMaterialIsStencil=false
     fontMaterial=null
     fontObject=&HeadingNow
     letterSpacing=0
     monospacedWidth=1
     size=25.5
     skewAmount=0
     typefaceFontName=86 Bold Italic
     outlineSettings {
      bApplyOutlineToDropShadows=false
      bMiteredCorners=false
      bSeparateFillAlpha=false
      outlineMaterial=null
      outlineSize=0
      outlineColor=<rgba 0 0 0 1>
     }
    }
    shadowOffset=<0 0>
   }
   slot {class=VerticalBoxSlot, properties {verticalAlignment=VAlign_Center, size {sizeRule=Fill, value=1}}}
  }
  Text_LevelsLeft {
   widgetClassPath=&CommonTextBlock
   bIsVariable=true
   parent=VerticalBox_68
   widget=Text_LevelsLeft
   properties {
    autoWrapText=true
    justification=Center
    text=XX LEFT
    textTransformPolicy=ToUpper
    colorAndOpacity {colorUseRule=UseColor_Specified, specifiedColor=<rgba 0.5776 0.5776 0.5776 0.5>}
    font {
     bForceMonospaced=false
     bMaterialIsStencil=false
     fontMaterial=null
     fontObject=&HeadingNow
     letterSpacing=0
     monospacedWidth=1
     size=25.5
     skewAmount=0
     typefaceFontName=86 Bold Italic
     outlineSettings {
      bApplyOutlineToDropShadows=false
      bMiteredCorners=false
      bSeparateFillAlpha=false
      outlineMaterial=null
      outlineSize=0
      outlineColor=<rgba 0 0 0 1>
     }
    }
    shadowOffset=<0 0>
   }
   slot {class=VerticalBoxSlot, properties {verticalAlignment=VAlign_Center, size {sizeRule=Fill, value=1}}}
  }
  Button_Add_1 {
   widgetClassPath=&WBP_UIKit_Button_Regular
   bIsVariable=true
   parent=Controllers
   widget=Button_Add_1
   properties {
    bIsFocusable=false
    minHeight=48
    minWidth=112
    text=+1
    triggeringInputAction {dataTable=&InputActionDataTable, rowName=GenericRightShoulder}
   }
   slot {class=HorizontalBoxSlot, properties {horizontalAlignment=HAlign_Center, padding=<ltrb 0 0 16 0>}}
  }
  Button_Add_10 {
   widgetClassPath=&WBP_UIKit_Button_Regular
   bIsVariable=true
   parent=Controllers
   widget=Button_Add_10
   properties {
    bIsFocusable=false
    minHeight=48
    minWidth=112
    text=+10
    triggeringInputAction {dataTable=&InputActionDataTable, rowName=GenericRightTrigger}
   }
   slot {class=HorizontalBoxSlot, properties {horizontalAlignment=HAlign_Center}}
  }
  Tag_Disclaimer {
   widgetClassPath=&Border
   bIsVariable=true
   parent=VerticalBox_72
   widget=Tag_Disclaimer
   properties {
    horizontalAlignment=HAlign_Center
    verticalAlignment=VAlign_Center
    visibility=Collapsed
    background {
     bIsDynamicallyLoaded=false
     drawAs=Image
     imageType=FullColor
     mirroring=NoMirror
     resourceName=None
     resourceObject=/FortUILibrary/Systems/Badge/Materials/M_UIKit_Badge_Background
     tiling=NoTile
     imageSize=<32 32>
     margin=<ltrb 0 0 0 0>
     outlineSettings {
      bUseBrushTransparency=false
      roundingType=HalfHeightRadius
      width=0
      color {colorUseRule=UseColor_Specified, specifiedColor=<rgba 0 0 0 0>}
      cornerRadii=<0 0 0 0>
     }
     tintColor {colorUseRule=UseColor_Specified, specifiedColor=<rgba 0.9301 1 0.0103 1>}
     uVRegion {bIsValid=false, max=<0 0>, min=<0 0>}
    }
    padding=<ltrb 0 2 0 2>
   }
   slot {class=VerticalBoxSlot}
  }
  Text_Disclaimer {
   widgetClassPath=&CommonRichTextBlock
   bIsVariable=true
   parent=Tag_Disclaimer
   widget=Text_Disclaimer
   properties {
    bOverrideDefaultStyle=true
    mobileTextBlockScale=1.25
    scrollStyle=&TextScrollStyle-Base_C
    text=Disclaimer Text
    defaultTextStyleOverride {
     overflowPolicy=Clip
     transformPolicy=None
     colorAndOpacity {colorUseRule=UseColor_Specified, specifiedColor=<rgba 0 0 0 0.7>}
     font {
      bForceMonospaced=false
      bMaterialIsStencil=false
      fontMaterial=null
      fontObject=&HeadingNow
      letterSpacing=10
      monospacedWidth=1
      size=18
      skewAmount=0
      typefaceFontName=74 Regular
      outlineSettings {
       bApplyOutlineToDropShadows=false
       bMiteredCorners=false
       bSeparateFillAlpha=false
       outlineMaterial=null
       outlineSize=0
       outlineColor=<rgba 0 0 0 1>
      }
     }
     highlightColor {colorUseRule=UseColor_Specified, specifiedColor=<rgba 0 0 0 1>}
     highlightShape {
      bIsDynamicallyLoaded=false
      drawAs=Image
      imageType=NoImage
      mirroring=NoMirror
      resourceName=None
      resourceObject=null
      tiling=NoTile
      imageSize=<32 32>
      margin=<ltrb 0 0 0 0>
      outlineSettings {
       bUseBrushTransparency=false
       roundingType=HalfHeightRadius
       width=0
       color {colorUseRule=UseColor_Specified, specifiedColor=<rgba 0 0 0 0>}
       cornerRadii=<0 0 0 0>
      }
      tintColor {colorUseRule=UseColor_Specified, specifiedColor=<rgba 1 1 1 1>}
      uVRegion {bIsValid=false, max=<0 0>, min=<0 0>}
     }
     selectedBackgroundColor {colorUseRule=UseColor_Foreground, specifiedColor=<rgba 1 0 1 1>}
     shadowColorAndOpacity=<rgba 0 0 0 1>
     shadowOffset=<0 0>
     strikeBrush {
      bIsDynamicallyLoaded=false
      drawAs=Image
      imageType=NoImage
      mirroring=NoMirror
      resourceName=None
      resourceObject=null
      tiling=NoTile
      imageSize=<32 32>
      margin=<ltrb 0 0 0 0>
      outlineSettings {
       bUseBrushTransparency=false
       roundingType=HalfHeightRadius
       width=0
       color {colorUseRule=UseColor_Specified, specifiedColor=<rgba 0 0 0 0>}
       cornerRadii=<0 0 0 0>
      }
      tintColor {colorUseRule=UseColor_Specified, specifiedColor=<rgba 1 1 1 1>}
      uVRegion {bIsValid=false, max=<0 0>, min=<0 0>}
     }
     underlineBrush {
      bIsDynamicallyLoaded=false
      drawAs=Image
      imageType=NoImage
      mirroring=NoMirror
      resourceName=None
      resourceObject=null
      tiling=NoTile
      imageSize=<32 32>
      margin=<ltrb 0 0 0 0>
      outlineSettings {
       bUseBrushTransparency=false
       roundingType=HalfHeightRadius
       width=0
       color {colorUseRule=UseColor_Specified, specifiedColor=<rgba 0 0 0 0>}
       cornerRadii=<0 0 0 0>
      }
      tintColor {colorUseRule=UseColor_Specified, specifiedColor=<rgba 1 1 1 1>}
      uVRegion {bIsValid=false, max=<0 0>, min=<0 0>}
     }
    }
   }
   slot {class=BorderSlot, properties {horizontalAlignment=HAlign_Center, verticalAlignment=VAlign_Center, padding=<ltrb 0 2 0 2>}}
  }
  HB_PurchaseBundle {
   widgetClassPath=&HorizontalBox
   bIsVariable=true
   parent=VB_All
   widget=HB_PurchaseBundle
   properties {visibility=Collapsed}
   slot {class=VerticalBoxSlot, properties {horizontalAlignment=HAlign_Center, verticalAlignment=VAlign_Center, padding=<ltrb 160 16 160 8>}}
  }
  CommonHardwareVisibilityBorder_0 {
   widgetClassPath=&CommonHardwareVisibilityBorder
   parent=HB_PurchaseBundle
   widget=CommonHardwareVisibilityBorder_0
   properties {
    horizontalAlignment=HAlign_Center
    verticalAlignment=VAlign_Center
    visibility=Collapsed
    visibilityQuery {
     autoDescription=" ANY( Input.Gamepad )"
     tokenStreamVersion=0
     userDescription=""
     queryTokenStream [0,1,1,1,0]
     tagDictionary [{tagName=Input.Gamepad}]
    }
   }
   slot {class=HorizontalBoxSlot, properties {horizontalAlignment=HAlign_Center, verticalAlignment=VAlign_Center}}
  }
  Button_BundleTogglePrompt {
   widgetClassPath=&WBP_UIKit_Button_Generic
   bIsVariable=true
   parent=CommonHardwareVisibilityBorder_0
   widget=Button_BundleTogglePrompt
   properties {
    bHideInputActionWithKeyboard=true
    bIsFocusable=false
    minHeight=40
    promptHorizontalAlignment=HAlign_Center
    useBackgroundBlock=false
    useIconBlock=false
    useTextBlock=false
    buttonPadding=<ltrb 0 0 0 0>
    promptNudge=<0 0>
    triggeringInputAction {dataTable=&InputActionDataTable, rowName=GenericFaceButtonLeft}
   }
   slot {class=BorderSlot, properties {horizontalAlignment=HAlign_Center, verticalAlignment=VAlign_Center, padding=<ltrb 0 0 0 0>}}
  }
  Button_BundleToggle {
   widgetClassPath=/FortUILibrary/UIKit/Buttons/WBP_UIKit_ButtonWithToggle
   bIsVariable=true
   parent=HB_PurchaseBundle
   widget=Button_BundleToggle
   properties {
    bIsFocusable=false
    contentHorizontalAlignment=HAlign_Center
    statusIndicatorHorizontalAlignment=HAlign_Center
    statusIndicatorVerticalAlignment=VAlign_Center
    text="Buy a bundle of XX Levels at a reduced price of XX V-Bucks (normally XX)"
    textJustification=Left
    textSecondJustification=Left
    textTransformPolicy=None
    toggleType=CheckBox
    font {
     bForceMonospaced=false
     bMaterialIsStencil=false
     fontMaterial=/FortUILibrary/UIKit/Buttons/MaterialInstances/GenericButtons/MI_UIKit_ButtonWithToggle_Text
     fontObject=&HeadingNow
     letterSpacing=48
     monospacedWidth=1
     size=18
     skewAmount=0
     typefaceFontName=75 Medium
     outlineSettings {
      bApplyOutlineToDropShadows=false
      bMiteredCorners=false
      bSeparateFillAlpha=false
      outlineMaterial=null
      outlineSize=0
      outlineColor=<rgba 0 0 0 1>
     }
    }
    promptNudge=<0 0>
    secondTextPadding=<ltrb 0 0 0 0>
    statusIndicatorPadding=<ltrb 0 0 0 0>
    textPadding=<ltrb 14 0 14 0>
   }
   slot {class=HorizontalBoxSlot, properties {horizontalAlignment=HAlign_Center, verticalAlignment=VAlign_Center, size {sizeRule=Fill, value=1}}}
  }
  WBP_PriceSwitcher_Base {
   widgetClassPath=/FortUILibrary/Systems/ItemShopTile/WBP_PriceSwitcher_Base
   bIsVariable=true
   parent=VB_All
   widget=WBP_PriceSwitcher_Base
   properties {
    hasRealPrice=true
    isAdditionalRealPrice=true
    previewRealPriceText=(equivalent to $8.99**)
    priceAndOwnershipHorizontalAlignment=HAlign_Center
    realPriceHorizontalAlignment=HAlign_Center
    widgetHeight=60
    widgetHeightWithRealPrice=88
    currencyImageSize=<36 36>
    ownedTextFont {
     bForceMonospaced=false
     bMaterialIsStencil=false
     fontMaterial=/FortUILibrary/Systems/ItemShopTile/Materials/PriceSwitcher/MI_UIKit_PriceSwitcher_Text
     fontObject=&HeadingNow
     letterSpacing=0
     monospacedWidth=1
     size=25.5
     skewAmount=0
     typefaceFontName=86 Bold Italic
     outlineSettings {
      bApplyOutlineToDropShadows=false
      bMiteredCorners=false
      bSeparateFillAlpha=false
      outlineMaterial=null
      outlineSize=0
      outlineColor=<rgba 0 0 0 1>
     }
    }
    priceTextFont {
     bForceMonospaced=false
     bMaterialIsStencil=false
     fontMaterial=&MI_UIKit_PriceSwitcher_Text
     fontObject=&HeadingNow
     letterSpacing=0
     monospacedWidth=1
     size=25.5
     skewAmount=0
     typefaceFontName=86 Bold Italic
     outlineSettings {
      bApplyOutlineToDropShadows=false
      bMiteredCorners=false
      bSeparateFillAlpha=false
      outlineMaterial=null
      outlineSize=0
      outlineColor=<rgba 0 0 0 1>
     }
    }
    realPriceTextFont {
     bForceMonospaced=false
     bMaterialIsStencil=false
     fontMaterial=&MI_UIKit_PriceSwitcher_Text
     fontObject=&HeadingNow
     letterSpacing=0
     monospacedWidth=1
     size=19.5
     skewAmount=0
     typefaceFontName=74 Regular
     outlineSettings {
      bApplyOutlineToDropShadows=false
      bMiteredCorners=false
      bSeparateFillAlpha=false
      outlineMaterial=null
      outlineSize=0
      outlineColor=<rgba 0 0 0 1>
     }
    }
    regularPriceTextFont {
     bForceMonospaced=false
     bMaterialIsStencil=false
     fontMaterial=&MI_UIKit_PriceSwitcher_Text
     fontObject=&HeadingNow
     letterSpacing=0
     monospacedWidth=1
     size=25.5
     skewAmount=0
     typefaceFontName=75 Medium
     outlineSettings {
      bApplyOutlineToDropShadows=false
      bMiteredCorners=false
      bSeparateFillAlpha=false
      outlineMaterial=null
      outlineSize=0
      outlineColor=<rgba 0 0 0 1>
     }
    }
   }
   slot {class=VerticalBoxSlot, properties {horizontalAlignment=HAlign_Center, verticalAlignment=VAlign_Center}}
  }
  Overlay_0 {widgetClassPath=&Overlay, namedSlotHost=WBP_UIKit_Dialog_Internal, widget=Overlay_0}
  DarkBottomFade {
   widgetClassPath=&Image
   bIsVariable=true
   parent=Overlay_0
   widget=DarkBottomFade
   properties {
    visibility=HitTestInvisible
    brush {
     bIsDynamicallyLoaded=false
     drawAs=Image
     imageType=FullColor
     mirroring=NoMirror
     resourceName=None
     resourceObject=/SeasonPass/Materials/LegalText/MI_UI_LegalText_Fade_Bottom
     tiling=NoTile
     imageSize=<32 250>
     margin=<ltrb 0 0 0 0>
     outlineSettings {
      bUseBrushTransparency=false
      roundingType=HalfHeightRadius
      width=0
      color {colorUseRule=UseColor_Specified, specifiedColor=<rgba 0 0 0 0>}
      cornerRadii=<0 0 0 0>
     }
     tintColor {colorUseRule=UseColor_Specified, specifiedColor=<rgba 1 1 1 1>}
     uVRegion {bIsValid=false, max=<0 0>, min=<0 0>}
    }
   }
   slot {class=OverlaySlot, properties {horizontalAlignment=HAlign_Fill, verticalAlignment=VAlign_Bottom}}
  }
  SZ_Legal {
   widgetClassPath=&SafeZone
   bIsVariable=true
   parent=Overlay_0
   widget=SZ_Legal
   slot {class=OverlaySlot, properties {horizontalAlignment=HAlign_Fill, verticalAlignment=VAlign_Fill, padding=<ltrb 40 40 40 40>}}
  }
  LegalandPlayerInfo {widgetClassPath=&Overlay, bIsVariable=true, parent=SZ_Legal, widget=LegalandPlayerInfo, slot {class=SafeZoneSlot}}
  VerticalBox_LegalTexts {
   widgetClassPath=&VerticalBox
   parent=LegalandPlayerInfo
   widget=VerticalBox_LegalTexts
   slot {class=OverlaySlot, properties {horizontalAlignment=HAlign_Fill, verticalAlignment=VAlign_Bottom}}
  }
  LegalText {
   widgetClassPath=&CommonTextBlock
   bIsVariable=true
   parent=VerticalBox_LegalTexts
   widget=LegalText
   properties {
    autoWrapText=true
    bIsScrollingEnabled=false
    mobileFontSizeMultiplier=1.4
    text="Rewards purchases are not eligible for returns through Fortnite Cancel Purchase or Return Tickets. Season X through Month 00 2:00am ET 1 Discount bundle per season. Rewards may be purchased, max 100 per season.Unclaimed Battle Pass rewards are automatically claimed at the end of season.Some rewards require quests to be completed in Battle Royale / Zero Build mode in order to be earned.\r\nNot all cosmetics are usable in all experiences. For more details on item variations and limitations, see fn.gg/cosmetics"
    font {
     bForceMonospaced=false
     bMaterialIsStencil=false
     fontMaterial=null
     fontObject=&HeadingNow
     letterSpacing=10
     monospacedWidth=1
     size=11.25
     skewAmount=0
     typefaceFontName=74 Regular
     outlineSettings {
      bApplyOutlineToDropShadows=false
      bMiteredCorners=false
      bSeparateFillAlpha=false
      outlineMaterial=null
      outlineSize=0
      outlineColor=<rgba 0 0 0 1>
     }
    }
    shadowOffset=<0 0>
   }
   slot {class=VerticalBoxSlot}
  }
  LegalText_RealPrice_ReducedPrice {
   widgetClassPath=/FortUILibrary/UIKit/Price/WBP_RealPrice_LegalText
   bIsVariable=true
   parent=VerticalBox_LegalTexts
   widget=LegalText_RealPrice_ReducedPrice
   properties {autoWrap=true, wrapTextAt=0, textColor {colorUseRule=UseColor_Specified, specifiedColor=<rgba 1 1 1 1>}}
   slot {class=VerticalBoxSlot, properties {padding=<ltrb 0 8 0 0>}}
  }
  LegalText_RealPrice_PriceSwitcher {
   widgetClassPath=&WBP_RealPrice_LegalText
   bIsVariable=true
   parent=VerticalBox_LegalTexts
   widget=LegalText_RealPrice_PriceSwitcher
   properties {
    autoWrap=true
    isAdditionalText=true
    previewText="**To buy exactly 1,000 V-Bucks, you’d pay €8.99."
    wrapTextAt=0
    textColor {colorUseRule=UseColor_Specified, specifiedColor=<rgba 1 1 1 1>}
   }
   slot {class=VerticalBoxSlot, properties {padding=<ltrb 0 8 0 0>}}
  }
  WBP_PlayerInfo {
   widgetClassPath=/Game/UI/Frontend/ItemShop/MultiProduct/WBP_PlayerInfo
   bIsVariable=true
   parent=LegalandPlayerInfo
   widget=WBP_PlayerInfo
   slot {class=OverlaySlot, properties {horizontalAlignment=HAlign_Right}}
  }
 ]
}
