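# UIKit widget that shows which option is selected, as either a pip image or a "current / total" text counter.
# HandleDisplayChanged switches to rich text when NumOptions exceeds MaxPipOptions, otherwise drives the pip dynamic material's Options and SelectedIndex parameters.
# HandleMaxPipsChanged only switches views, using the reverse comparison: text when the incoming max exceeds NumOptions.
# PreConstruct refreshes the display at design time only.
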WidgetBlueprint /FortUILibrary/UIKit/PipList/WBP_UIKit_Pip_Indicator
graphs [
 EventGraph {nodes [
   n1 = Event PreConstruct()
   > Branch(Condition=@n1.IsDesignTime)
   > HandleDisplayChanged!(execute, self, OptionChanged:int=0) -> (then)
   Event Construct()
   Event Tick()
  ]}
 HandleDisplayChanged {nodes [
   HandleDisplayChanged!() -> (then, OptionChanged:int) {is=title K2Node_FunctionEntry}
   > n1 = Set IsShowingText(IsShowingText:bool="integer > integer"(A=Get NumOptions() -> (NumOptions:int=0), B=Get MaxPipOptions() -> (MaxPipOptions:int=0))) -> (Output_Get:bool)
   > n2 = Branch(Condition=@n1)
   > SetActiveWidgetIndex(self:Common Visibility Switcher Object Reference=Get Switcher_Root() -> (Switcher_Root:Common Visibility Switcher Object Reference), Index=1)
   n3 = Get SelectedIndex() -> (SelectedIndex:int=0)
   > SetText(
    self=Get RText_PipTextIndicator() -> (RText_PipTextIndicator:Rich Text Block Object Reference)
    InText=Format Text!(
     Format="LOCTEXT(\"{CurrentOption} <weak>/ {TotalOptions}</>\")"
     CurrentOption:int=int + int(A=@n3)
     TotalOptions:int=Get NumOptions() -> (NumOptions=0)
    ) -> (Result)
   )
   SetActiveWidgetIndex(execute=@n2.else, self=Get Switcher_Root(), Index=0)
   > n4 = GetDynamicMaterial(self=Get Img_PipIndicator() -> (Img_PipIndicator:Image Object Reference))
   > SetScalarParameterValue(self=@n4, ParameterName=Options, Value="To Float (Integer)"(InInt=Get NumOptions() -> (NumOptions=0)))
   > SetScalarParameterValue(self=@n4, ParameterName=SelectedIndex, Value="To Float (Integer)"(InInt=Get SelectedIndex() -> (SelectedIndex=0)))
  ]}
 HandleMaxPipsChanged {nodes [
   n1 = HandleMaxPipsChanged!() -> (then, MaxPipsIn:int) {is=title K2Node_FunctionEntry}
   n2 = Get NumOptions() -> (NumOptions=0)
   > SetActiveWidgetIndex(self=Get Switcher_Root(), Index=Select(Option 0:int=0, Option 1:int=1, Index="integer > integer"(A=@n1, B=@n2)) -> (ReturnValue:int))
  ]}
]
variables [
 IsShowingText {type=boolean}
 MaxPipOptions {type=integer, default=3}
 NumOptions {type=integer, default=2}
 SelectedIndex {type=integer}
]
widgets {
 info {inheritedWidgetCount=0, parentClass=&UserWidget, rootWidgetClass=&CommonVisibilitySwitcher, widgetCount=3}
 widgets [
  Switcher_Root {widgetClassPath=&CommonVisibilitySwitcher, bIsVariable=true, widget=Switcher_Root}
  Img_PipIndicator {
   widgetClassPath=&Image
   bIsVariable=true
   parent=Switcher_Root
   widget=Img_PipIndicator
   properties {brush {
     bIsDynamicallyLoaded=false
     drawAs=Image
     imageType=FullColor
     mirroring=NoMirror
     resourceName=None
     resourceObject=&/MI_UI_RotatorPips
     tiling=NoTile
     imageSize=<50 32>
     margin=<ltrb 0 0 0 0>
     outlineSettings {
      bUseBrushTransparency=false
      roundingType=HalfHeightRadius
      width=0
      color {colorUseRule=UseColor_Specified, specifiedColor=<rgba 0 0 0 0>}
      cornerRadii=<0 0 0 0>
     }
     tintColor {colorUseRule=UseColor_Specified, specifiedColor=<rgba 1 1 1 1>}
     uVRegion {bIsValid=false, max=<0 0>, min=<0 0>}
    }}
   slot {class=CommonVisibilitySwitcherSlot, properties {horizontalAlignment=HAlign_Left, verticalAlignment=VAlign_Top}}
  }
  RText_PipTextIndicator {
   widgetClassPath=/Script/UMG.RichTextBlock
   bIsVariable=true
   parent=Switcher_Root
   widget=RText_PipTextIndicator
   properties {text="1 <weak>/ 8</>", textStyleSet=/FortUILibrary/Styles/DT_UI_RichTextStyles_HeadingNow}
   slot {class=CommonVisibilitySwitcherSlot}
  }
 ]
}
