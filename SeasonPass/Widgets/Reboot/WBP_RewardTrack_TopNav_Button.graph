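# Season pass reward track top-nav button showing a category's character portrait with equipped outline, lock indicator and page pips.
# Theme data supplies a CharacterColor pushed into the portrait's dynamic material when the lazy icon loads or the color updates.
# Lock changes and scrubbed progress drive the SetUnlock animation; selection plays the Selected animation and fires analytics.
# View-model handlers show pips only for the selected category outside the frontend, tracking page count and index.
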WidgetBlueprint /SeasonPass/Widgets/Reboot/WBP_RewardTrack_TopNav_Button
animations [
 Selected {
  binding_count=3
  length_seconds=0.25
  object=&self:Selected
  sequence {bindings [
    MainContent {
     id=#1
     possessed_class=Overlay
     tracks [{
       type=MovieScene2DTransformTrack
       sections [{
         type=MovieScene2DTransformSection
         end_frame=4
         start_frame=0
         channels [
          Translation.X {type=MovieSceneScriptingFloatChannel, const [0,1]}
          Translation.Y {type=MovieSceneScriptingFloatChannel, const [0,1]}
          Angle {type=MovieSceneScriptingFloatChannel, const [0,1]}
          Scale.X {type=MovieSceneScriptingFloatChannel, keys [[0,1],[4,1.33]]}
          Scale.Y {type=MovieSceneScriptingFloatChannel, keys [[0,1],[4,1.33]]}
          Shear.X {type=MovieSceneScriptingFloatChannel, const [0,1]}
          Shear.Y {type=MovieSceneScriptingFloatChannel, const [0,1]}
         ]
        }]
      }]
    }
    Portrait {
     id=#2
     possessed_class=FortLazyImage
     tracks [{
       type=MovieSceneWidgetMaterialTrack
       sections [{
         type=MovieSceneParameterSection
         end_frame=4
         start_frame=0
         channels [
          Selected {type=MovieSceneScriptingFloatChannel, keys [[0,0],[4,1]]}
          Scale {type=MovieSceneScriptingFloatChannel, keys [[0,2],[4,2.4]]}
         ]
        }]
      }]
    }
    StatusIndicator_Locked {
     id=#3
     possessed_class=WBP_UIKit_StatusIndicator_C
     tracks [{
       type=MovieScene2DTransformTrack
       sections [{
         type=MovieScene2DTransformSection
         end_frame=5
         start_frame=0
         channels [
          Translation.X {type=MovieSceneScriptingFloatChannel, const [0,1]}
          Translation.Y {type=MovieSceneScriptingFloatChannel, keys [[2,0],[5,30]]}
          Angle {type=MovieSceneScriptingFloatChannel, const [0,1]}
          Scale.X {type=MovieSceneScriptingFloatChannel, const [1,1]}
          Scale.Y {type=MovieSceneScriptingFloatChannel, const [1,1]}
          Shear.X {type=MovieSceneScriptingFloatChannel, const [0,1]}
          Shear.Y {type=MovieSceneScriptingFloatChannel, const [0,1]}
         ]
        }]
      }]
    }
   ]}
 }
 SetUnlock {
  binding_count=3
  length_seconds=1.2
  object=&self:SetUnlock
  sequence {bindings [
    RootContainer {
     id=#4
     possessed_class=Overlay
     tracks [{
       type=MovieScene2DTransformTrack
       sections [{
         type=MovieScene2DTransformSection
         end_frame=24
         start_frame=0
         channels [
          Translation.X {type=MovieSceneScriptingFloatChannel, const [0,1]}
          Translation.Y {type=MovieSceneScriptingFloatChannel, const [0,1]}
          Angle {type=MovieSceneScriptingFloatChannel, const [0,1]}
          Scale.X {type=MovieSceneScriptingFloatChannel, keys [[0,1],[6,1.2],[14,1.33],[24,1]]}
          Scale.Y {type=MovieSceneScriptingFloatChannel, keys [[0,1],[6,1.2],[14,1.33],[24,1]]}
          Shear.X {type=MovieSceneScriptingFloatChannel, const [0,1]}
          Shear.Y {type=MovieSceneScriptingFloatChannel, const [0,1]}
         ]
        }]
      }]
    }
    StatusIndicator_Locked {
     id=#5
     possessed_class=WBP_UIKit_StatusIndicator_C
     tracks [{
       type=MovieScene2DTransformTrack
       sections [{
         type=MovieScene2DTransformSection
         end_frame=24
         start_frame=0
         channels [
          Angle {type=MovieSceneScriptingFloatChannel, const [0,1]}
          Scale.X {type=MovieSceneScriptingFloatChannel, keys [[0,1],[24,0]]}
          Scale.Y {type=MovieSceneScriptingFloatChannel, keys [[0,1],[24,0]]}
          Shear.X {type=MovieSceneScriptingFloatChannel, const [0,1]}
          Shear.Y {type=MovieSceneScriptingFloatChannel, const [0,1]}
         ]
        }]
      }]
    }
    Glow {
     id=#6
     possessed_class=Image
     tracks [
      {
       type=MovieSceneWidgetMaterialTrack
       sections [{type=MovieSceneParameterSection, end_frame=24, start_frame=0, channels [Reveal {type=MovieSceneScriptingFloatChannel, keys [[0,0],[24,1]]}]}]
      }
      {
       type=MovieSceneFloatTrack
       display_name=Render Opacity
       sections [{type=MovieSceneFloatSection, end_frame=24, start_frame=0, channels [None {type=MovieSceneScriptingFloatChannel, keys [[23,1],[24,0]]}]}]
      }
     ]
    }
   ]}
 }
]
graphs [
 DEBUG_DisplayDebugData {nodes [
   DEBUG_DisplayDebugData() {is=title K2Node_FunctionEntry}
   > n1 = GetDynamicMaterial(self=Get Portrait() -> (Portrait:Fort Lazy Image Object Reference))
   > n2 = Set CharacterTheme(CharacterTheme:Material Instance Dynamic Object Reference=@n1) -> (Output_Get:Material Instance Dynamic Object Reference)
   n3 = Get RewardName() -> (RewardName:Byte=0)
   > SetTextureParameterValue(
    self=@n2
    ParameterName=RewardTexture
    Value=Select!(
     NewEnumerator0:Texture Object Reference=/BRCosmetics/UI/Foundation/Textures/Icons/Heroes/Athena/Soldier/S36/T_Soldier_SquidGlistenLift
     NewEnumerator1:Texture Object Reference=/BRCosmetics/UI/Foundation/Textures/Icons/Heroes/Athena/Soldier/S36/T_Soldier_BentBaton_Flip
     NewEnumerator2:Texture Object Reference=/BRCosmetics/UI/Foundation/Textures/Icons/Heroes/Athena/Soldier/S36/T_Soldier_LazyLizzChip
     NewEnumerator3:Texture Object Reference=/BRCosmetics/UI/Foundation/Textures/Icons/Heroes/Athena/Soldier/S36/T_Soldier_SpatialTravelWatt
     NewEnumerator4:Texture Object Reference=/BRCosmetics/UI/Foundation/Textures/Icons/Heroes/Athena/Soldier/S36/T_Soldier_AshenMagnusSlant
     NewEnumerator5:Texture Object Reference=/BRCosmetics/UI/Foundation/Textures/Icons/Heroes/Athena/Soldier/S36/T_Soldier_PolarityWinnField
     NewEnumerator6:Texture Object Reference=/BRCosmetics/UI/Foundation/Textures/Icons/Heroes/Athena/Soldier/S36/T_Soldier_MistRaven_Peck
     NewEnumerator7:Texture Object Reference=/BRCosmetics/UI/Foundation/Textures/Icons/Heroes/Athena/Soldier/S36/T_Soldier_WealthLamb
     NewEnumerator8:Texture Object Reference
     NewEnumerator9:Texture Object Reference
     NewEnumerator10:Texture Object Reference
     NewEnumerator11:Texture Object Reference
     NewEnumerator12:Texture Object Reference
     NewEnumerator13:Texture Object Reference
     NewEnumerator14:Texture Object Reference
     NewEnumerator15:Texture Object Reference
     Index:E_BPReboot_Characters Enum=Byte to Enum E_BPReboot_Characters!(Byte:Byte=@n3) -> (ReturnValue:E_BPReboot_Characters Enum) {is=title K2Node_CastByteToEnum}
    ) -> (ReturnValue:Texture Object Reference)
   )
   n4 = Get RewardName() -> (RewardName=0)
   > SetVectorParameterValue(
    self=@n2
    ParameterName=CharacterColor
    Value=Select(
     NewEnumerator0:Linear Color Structure=(R=0.162,G=0,B=0.0024,A=1)
     NewEnumerator1:Linear Color Structure=(R=0.305,G=0.016,B=0.3663,A=1)
     NewEnumerator2:Linear Color Structure=(R=1,G=0.7157,B=0.063,A=1)
     NewEnumerator3:Linear Color Structure=(R=0.0048,G=0.0723,B=0.1779,A=1)
     NewEnumerator4:Linear Color Structure=(R=0.0262,G=0.063,B=0,A=1)
     NewEnumerator5:Linear Color Structure=(R=0.1714,G=0.0561,B=0.521,A=1)
     NewEnumerator6:Linear Color Structure=(R=0.0999,G=0,B=0.0144,A=1)
     NewEnumerator7:Linear Color Structure=(R=0.0284,G=0.0802,B=0.1812,A=1)
     NewEnumerator8:Linear Color Structure
     NewEnumerator9:Linear Color Structure
     NewEnumerator10:Linear Color Structure
     NewEnumerator11:Linear Color Structure
     NewEnumerator12:Linear Color Structure
     NewEnumerator13:Linear Color Structure
     NewEnumerator14:Linear Color Structure
     NewEnumerator15:Linear Color Structure
     Index=Byte to Enum E_BPReboot_Characters(Byte=@n4)
    ) -> (ReturnValue:Linear Color Structure)
   )
  ]}
 EventGraph {nodes [
   n1 = OnLazyIconUpdated!() -> (OutputDelegate, then, LazyObject:"Soft Object Reference (by ref)") {is=title K2Node_CustomEvent}
   > Branch(Condition=IsValidSoftObjectReference(SoftObjectReference=@n1.LazyObject))
   > n2 = Async Load Asset(Asset=@n1.LazyObject)
   n3 = Cast To MaterialInstance!(execute=@n2.Completed, Object=@n2) -> (then, CastFailed, AsMaterial Instance:Material Instance Object Reference) {is=title K2Node_DynamicCast}
   n4 = Get Portrait()
   > n5 = SetBrushFromMaterial(self=@n4, Material=@n3)
   n6 = GetDynamicMaterial(execute=[@n5.then, @n7.then], self=@n4)
   > SetVectorParameterValue(self=@n6, ParameterName=CharacterColor, Value=Get CachedCharacterColor() -> (CachedCharacterColor:Linear Color Structure))
   n7 = OnCachedColorUpdated() {is=title K2Node_CustomEvent}
   n8 = Event PreConstruct()
   > Parent: PreConstruct(IsDesignTime=@n8.IsDesignTime)
   > SetEquipCornerRadius()
   > Branch(Condition=@n8.IsDesignTime)
   > DEBUG_DisplayDebugData() {is=title K2Node_CallFunction}
   Event Construct()
   > Parent: Construct()
   > UpdateIsInFrontend()
   > SetEquipped!(execute, self, IsEquipped:bool=false) -> (then)
   > n9 = Branch(Condition=Get IsInFrontend() -> (IsInFrontend:bool))
   > RegisterModularBlock!(
    execute
    self:UIKit Modular Button Object Reference
    Block:UIKit Block Object Reference=Get BlockOutline() -> (BlockOutline:WBP UIKit Block Outline Object Reference)
   ) -> (then)
   SetVisibility(execute=@n9.else, self=Get BlockOutline(), InVisibility=Collapsed)
   Event On Selected()
   > Fire Event Element Interacted(
    Element=Self-Reference()
    AdditionalInfo=Get Analytics Info Pass Category Name!(
     MakeFlowAnalyticsAdditionalInfo:"Flow Analytics Additional Info Structure (by ref)"=Make Flow Analytics Additional Info!() -> (FlowAnalyticsAdditionalInfo:Flow Analytics Additional Info Structure) {is=title K2Node_MakeStruct}
     Key:String=SelectedCategory
     SeasonPassSelectedVM:Season Pass Selection VM Object Reference=Get SeasonPassSelectionVM() -> (SeasonPassSelectionVM:Season Pass Selection VM Object Reference)
    ) -> (PassName:Flow Analytics Additional Info Structure) {is=title K2Node_MacroInstance}
   )
   > UpdateSelected!(execute, self, IsSelected:bool=true) -> (then)
   Event On Deselected()
   > UpdateSelected(IsSelected=false)
   Event On Hovered()
   n10 = GetSelected!(self:Common Button Base Object Reference) -> (ReturnValue:bool)
   > n11 = Branch(Condition=@n10)
   RequestTransition!(execute=@n11.else, self:UIKit Block Object Reference=Get BlockOutline(), State:EUIKitBlockVisualState Enum=Focused) -> (then)
   Event On Unhovered()
   > n12 = Branch(Condition=@n10)
   RequestTransition(execute=@n12.else, self=Get BlockOutline(), State=Unfocused)
  ]}
 GetIsLocked {nodes [
   n1 = GetIsLocked!() -> (then, ViewmodelIsLocked:bool=~, IsOverridingLocked:bool=~) {is=title K2Node_FunctionEntry}
   > n2 = Branch(Condition=@n1.IsOverridingLocked)
   > Return Node!(execute, IsLocked:bool=Get CachedLockedState() -> (CachedLockedState:bool))
   Return Node(execute=@n2.else, IsLocked=@n1)
  ]}
 SetEquipCornerRadius {nodes [
   SetEquipCornerRadius() {is=title K2Node_FunctionEntry}
   > n1 = GetDynamicMaterial(self:Border Object Reference=Get EquippedOutline() -> (EquippedOutline:Common Border Object Reference))
   > SetScalarParameterValue(self=@n1, ParameterName=UseValueInPixels, Value=0)
   > SetScalarParameterValue(self=@n1, ParameterName=TopLeftCornerRadius, Value=Get TopLeftCornerRadius() -> (TopLeftCornerRadius:float))
   > SetScalarParameterValue(self=@n1, ParameterName=TopRightCornerRadius, Value=Get TopRightCornerRadius() -> (TopRightCornerRadius:float))
   > SetScalarParameterValue(self=@n1, ParameterName=BottomLeftCornerRadius, Value=Get BottomLeftCornerRadius() -> (BottomLeftCornerRadius:float))
   > SetScalarParameterValue(self=@n1, ParameterName=BottomRightCornerRadius, Value=Get BottomRightCornerRadius() -> (BottomRightCornerRadius:float))
  ]}
 SetEquipped {nodes [
   n1 = SetEquipped!() -> (then, IsEquipped:bool=~) {is=title K2Node_FunctionEntry}
   > Set Is Equipped(Is Equipped:bool=@n1) -> (Output_Get:bool)
   > Set!(
    execute
    BrushColor_R:float=1
    BrushColor_G:float=1
    BrushColor_B:float=1
    BrushColor_A:float=Select!(Option 0:float=0, Option 1:float=1, Index:bool=@n1) -> (ReturnValue:float)
    self:Border Object Reference=Get EquippedOutline()
   ) -> (then, Output_Get:Linear Color Structure)
  ]}
 SetLocked {nodes [
   n1 = SetLocked!() -> (then, Locked:bool=~) {is=title K2Node_FunctionEntry}
   > Set CachedLockedState(CachedLockedState:bool=@n1) -> (Output_Get:bool)
   > SetVisibility(
    self=Get LockedOverlay() -> (LockedOverlay:Overlay Object Reference)
    InVisibility=Select(Option 0:ESlateVisibility Enum=Collapsed, Option 1:ESlateVisibility Enum=SelfHitTestInvisible, Index=@n1) -> (ReturnValue:ESlateVisibility Enum)
   )
   > n2 = GetDynamicMaterial(self:Image Object Reference=Get Portrait())
   > SetScalarParameterValue(self=@n2, ParameterName=Locked, Value=SelectFloat(A=1, B=0, bPickA=@n1))
   n3 = Get SetUnlock() -> (SetUnlock:Widget Animation Object Reference)
   > PlayAnimation(
    InAnimation=@n3
    StartAtTime=GetEndTime(self=@n3)
    PlayMode=Select(Option 0:EUMGSequencePlayMode Enum, Option 1:EUMGSequencePlayMode Enum=Reverse, Index=@n1) -> (ReturnValue:EUMGSequencePlayMode Enum)
   )
  ]}
 SetUnlockAnimationProgress {nodes [
   n1 = SetUnlockAnimationProgress!() -> (then, animProgress:double=~) {is=title K2Node_FunctionEntry}
   > Branch(Condition=GetSelected())
   n2 = Get SetUnlock()
   > PlayAnimation(InAnimation=@n2, StartAtTime=@n1)
   > PauseAnimation!(execute, self:User Widget Object Reference, InAnimation:Widget Animation Object Reference=@n2) -> (then, ReturnValue:float)
  ]}
 ShouldDisplayPagePips {nodes [
   ShouldDisplayPagePips() {is=title K2Node_FunctionEntry}
   n1 = Get SeasonPassSelectionVM()
   > n2 = Is Valid(InputObject=@n1)
   Return Node!(execute=@n2.Is Not Valid, DisplayPagePips:bool=false)
   n3 = Get IsInFrontend()
   n4 = Get SeasonPassCategoryVM() -> (SeasonPassCategoryVM:Season Pass Category VM Object Reference)
   n5 = Get SelectedCategory!(self:Season Pass Selection VM Object Reference=@n1) -> (SelectedCategory:Season Pass Category VM Object Reference)
   > Return Node(DisplayPagePips=AND Boolean(A="Equal (Object)"(A=@n5, B=@n4), B=NOT Boolean(A=@n3)))
  ]}
 UpdateIsInFrontend {nodes [
   UpdateIsInFrontend() {is=title K2Node_FunctionEntry}
   > n1 = Cast To FortPlayerControllerFrontEnd!(execute, Object=GetOwningPlayer()) -> (then, CastFailed, AsFort Player Controller Front End:Fort Player Controller Front End Object Reference) {is=title K2Node_DynamicCast}
   > Set IsInFrontend(IsInFrontend:bool=true) -> (Output_Get:bool)
   Set IsInFrontend(execute=@n1.CastFailed, IsInFrontend=false)
  ]}
 UpdateSelected {nodes [
   n1 = UpdateSelected!() -> (then, IsSelected:bool=~) {is=title K2Node_FunctionEntry}
   > Set Is Selected(Is Selected:bool=@n1) -> (Output_Get:bool)
   > PlayAnimation(InAnimation=Get Selected() -> (Selected:Widget Animation Object Reference), PlayMode=Select(Option 0=Reverse, Index=@n1))
   > RequestTransition(
    self=Get BlockOutline()
    State=Select(Option 0:EUIKitBlockVisualState Enum=Unfocused, Option 1:EUIKitBlockVisualState Enum=Focused, Index=@n1) -> (ReturnValue:EUIKitBlockVisualState Enum)
   )
  ]}
 VM_HandleCategoryPagesSet {nodes [
   VM_HandleCategoryPagesSet!() -> (then, SeasonPassPages:Array of Season Pass Page VM Object References) {is=title K2Node_FunctionEntry}
   > SetVisibility(
    self=Get Att_PipIndicator() -> (Att_PipIndicator:Fort Visual Attachment Object Reference)
    InVisibility=Select(
     Option 0:ESlateVisibility Enum=Collapsed
     Option 1:ESlateVisibility Enum=SelfHitTestInvisible
     Index=ShouldDisplayPagePips!() -> (DisplayPagePips:bool) {is=title K2Node_CallFunction}
    ) -> (ReturnValue:ESlateVisibility Enum)
   )
   > Set with Broadcast NumPages!(
    execute
    NumPages:int=Length(TargetArray:Array of Season Pass Page VM Object References=Get SeasonPassPages() -> (SeasonPassPages:Array of Season Pass Page VM Object References))
   ) -> (then, Output_Get:int=0)
   > n1 = Sequence()
   > Branch(Condition=Get IsWaitingCategoriesDataToUpdateSelectedPage() -> (IsWaitingCategoriesDataToUpdateSelectedPage:bool))
   > VM_HandleSelectedPageUpdated!(
    execute
    self
    SelectedPage:Season Pass Page VM Object Reference=Get SelectedPage!(self:Season Pass Selection VM Object Reference=Get SeasonPassSelectionVM()) -> (SelectedPage:Season Pass Page VM Object Reference)
   ) -> (then)
   Branch(
    execute=@n1.then_1
    Condition=Get IsWaitingCategoriesDataToUpdateSelectedCategory() -> (IsWaitingCategoriesDataToUpdateSelectedCategory:bool)
   )
   > VM_HandleSelectedCategoryUpdated!(execute, self, SelectedCategory:Season Pass Category VM Object Reference=Get SelectedCategory(self=Get SeasonPassSelectionVM())) -> (then)
  ]}
 VM_HandleSelectedCategoryUpdated {nodes [
   VM_HandleSelectedCategoryUpdated!() -> (then, SelectedCategory:Season Pass Category VM Object Reference) {is=title K2Node_FunctionEntry}
   > SetVisibility(self=Get Att_PipIndicator(), InVisibility=Select(Option 0=Collapsed, Option 1=SelfHitTestInvisible, Index=ShouldDisplayPagePips()))
   > Branch(Condition=NOT Boolean(A=Get IsInFrontend()))
   > Set IsWaitingCategoriesDataToUpdateSelectedCategory(IsWaitingCategoriesDataToUpdateSelectedCategory:bool=true) -> (Output_Get:bool)
   > Is Valid(InputObject=Get SeasonPassCategoryVM())
   > Set IsWaitingCategoriesDataToUpdateSelectedCategory(IsWaitingCategoriesDataToUpdateSelectedCategory=false)
  ]}
 VM_HandleSelectedPageUpdated {nodes [
   n1 = VM_HandleSelectedPageUpdated!() -> (then, SelectedPage:Season Pass Page VM Object Reference) {is=title K2Node_FunctionEntry}
   > SetVisibility(self=Get Att_PipIndicator(), InVisibility=Select(Option 0=Collapsed, Option 1=SelfHitTestInvisible, Index=ShouldDisplayPagePips()))
   > Set IsWaitingCategoriesDataToUpdateSelectedPage(IsWaitingCategoriesDataToUpdateSelectedPage:bool=true) -> (Output_Get:bool)
   n2 = Get SeasonPassCategoryVM()
   > Is Valid(InputObject=@n2)
   > Set IsWaitingCategoriesDataToUpdateSelectedPage(IsWaitingCategoriesDataToUpdateSelectedPage=false)
   n3 = Find Item!(
    TargetArray:Array of Season Pass Page VM Object References=Get Pages!(self:Season Pass Category VM Object Reference=@n2) -> (Pages:Array of Season Pass Page VM Object References)
    ItemToFind:"Season Pass Page VM Object Reference (by ref)"=@n1
   ) -> (ReturnValue)
   > Branch(Condition=AND Boolean(A="integer >= integer"(A=@n3), B=ShouldDisplayPagePips()))
   > Set with Broadcast SelectedPageIndex!(execute, SelectedPageIndex:int=@n3) -> (then, Output_Get:int=0)
  ]}
 VM_HandleThemeData {nodes [
   n1 = VM_HandleThemeData!() -> (then, Theme:Season Pass Category Theme Data Structure) {is=title K2Node_FunctionEntry}
   n2 = TryLookupColor!(
    Table:"Fort Color Reference Table Structure (by ref)"=Break Season Pass Category Theme Data!(SeasonPassCategoryThemeData:"Season Pass Category Theme Data Structure (by ref)"=@n1) -> (
     ImageReferenceTable:Fort Image Reference Table Structure
     ColorReferenceTable:Fort Color Reference Table Structure
     TextReferenceTable:Fort Text Reference Table Structure
    ) {is=title K2Node_BreakStruct}.ColorReferenceTable
    Name:String=CharacterColor
   ) -> (OutColor:Linear Color Structure, ReturnValue:bool)
   > Set CachedCharacterColor(CachedCharacterColor:Linear Color Structure=Select(Option 0:Linear Color Structure=(R=0.0802,G=1,B=0,A=1), Option 1:Linear Color Structure=@n2, Index=@n2.ReturnValue) -> (ReturnValue:Linear Color Structure)) -> (Output_Get:Linear Color Structure)
   > OnCachedColorUpdated() {is=title K2Node_CallFunction}
  ]}
]
variables [
 BottomLeftCornerRadius {type=number, default=1}
 BottomRightCornerRadius {type=number, default=1}
 CachedCharacterColor {type=LinearColor, default=<rgba 0.0802 1 0 1>}
 CachedLockedState {type=boolean}
 CharacterTheme {type=MaterialInstanceDynamic}
 Is Equipped {type=boolean}
 Is Selected {type=boolean}
 IsInFrontend {type=boolean}
 IsOverridingLockedState {type=boolean}
 IsWaitingCategoriesDataToUpdateSelectedCategory {type=boolean}
 IsWaitingCategoriesDataToUpdateSelectedPage {type=boolean}
 NumPages {type=integer}
 RewardName {type=integer}
 SelectedPageIndex {type=integer}
 TopLeftCornerRadius {type=number, default=1}
 TopRightCornerRadius {type=number, default=1}
]
widgets {
 info {
  inheritedWidgetCount=0
  namedSlotCount=1
  parentClass=/FortUILibrary/Systems/ModularButton/WBP_UIKit_ButtonGeneric_Base.WBP_UIKit_ButtonGeneric_Base_C
  rootWidgetClass=null
  widgetCount=12
 }
 widgets [
  RootContainer {widgetClassPath=&Overlay, bIsVariable=true, widget=RootContainer}
  MainContent {
   widgetClassPath=&Overlay
   bIsVariable=true
   parent=RootContainer
   widget=MainContent
   slot {class=OverlaySlot, properties {horizontalAlignment=HAlign_Center, padding=<ltrb 4 4 4 4>}}
  }
  EquippedOutline {
   widgetClassPath=&CommonBorder
   bIsVariable=true
   parent=MainContent
   widget=EquippedOutline
   properties {
    background {
     bIsDynamicallyLoaded=false
     drawAs=Image
     imageType=FullColor
     mirroring=NoMirror
     resourceName=None
     resourceObject=/FortUILibrary/Systems/Items/MaterialInstances/MI_UIKit_EquippedStateOutline
     tiling=NoTile
     imageSize=<32 32>
     margin=<ltrb 0 0 0 0>
     outlineSettings {
      bUseBrushTransparency=false
      roundingType=HalfHeightRadius
      width=0
      color {colorUseRule=UseColor_Specified, specifiedColor=<rgba 0 0 0 0>}
      cornerRadii=<0 0 0 0>
     }
     tintColor {colorUseRule=UseColor_Specified, specifiedColor=<rgba 1 1 1 1>}
     uVRegion {bIsValid=false, max=<0 0>, min=<0 0>}
    }
    padding=<ltrb 4 4 4 4>
   }
   slot {class=OverlaySlot, properties {horizontalAlignment=HAlign_Fill, verticalAlignment=VAlign_Fill}}
  }
  Portrait {
   widgetClassPath=&FortLazyImage
   bIsVariable=true
   parent=EquippedOutline
   widget=Portrait
   properties {
    materialTextureParamName=RewardTexture
    brush {
     bIsDynamicallyLoaded=false
     drawAs=Image
     imageType=FullColor
     mirroring=NoMirror
     resourceName=None
     resourceObject=/SeasonPass/Widgets/Reboot/Materials/M_UI_BPReboot_RewardTrack_TopNav_Portrait
     tiling=NoTile
     imageSize=<48 48>
     margin=<ltrb 0 0 0 0>
     outlineSettings {
      bUseBrushTransparency=false
      roundingType=HalfHeightRadius
      width=0
      color {colorUseRule=UseColor_Specified, specifiedColor=<rgba 0 0 0 0>}
      cornerRadii=<0 0 0 0>
     }
     tintColor {colorUseRule=UseColor_Specified, specifiedColor=<rgba 1 1 1 1>}
     uVRegion {bIsValid=false, max=<0 0>, min=<0 0>}
    }
   }
   slot {class=BorderSlot, properties {padding=<ltrb 4 4 4 4>}}
  }
  BlockOutline {
   widgetClassPath=/FortUILibrary/UIKit/Buttons/Blocks/WBP_UIKit_Block_Outline
   bIsVariable=true
   parent=MainContent
   widget=BlockOutline
   properties {bottomLeftCornerRadius=1, bottomRightCornerRadius=1, topLeftCornerRadius=1, topRightCornerRadius=1, visibility=SelfHitTestInvisible}
   slot {class=OverlaySlot, properties {horizontalAlignment=HAlign_Fill, verticalAlignment=VAlign_Fill, padding=<ltrb 5 5 5 5>}}
  }
  LockedOverlay {
   widgetClassPath=&Overlay
   bIsVariable=true
   parent=RootContainer
   widget=LockedOverlay
   slot {class=OverlaySlot, properties {horizontalAlignment=HAlign_Fill, verticalAlignment=VAlign_Fill}}
  }
  StatusIndicator_Locked {
   widgetClassPath=/FortUILibrary/UIKit/StatusIndicator/WBP_UIKit_StatusIndicator
   bIsVariable=true
   parent=LockedOverlay
   widget=StatusIndicator_Locked
   properties {
    type=Utility
    icon=/FortUILibrary/Textures/MaterialInstances/MI_UIKit_Icon_Locked
    iconSize=Compact
    overrideStatusIndicatorState=true
    overrideStatusIndicatorVisibility=true
    visibility=HitTestInvisible
   }
   slot {class=OverlaySlot, properties {horizontalAlignment=HAlign_Center, verticalAlignment=VAlign_Center}}
  }
  SetUnlockAninmationContainer {
   widgetClassPath=&CommonVisualAttachment
   bIsVariable=true
   parent=RootContainer
   widget=SetUnlockAninmationContainer
   properties {contentAnchor=<0.5 0.575>}
   slot {class=OverlaySlot, properties {horizontalAlignment=HAlign_Center, verticalAlignment=VAlign_Center}}
  }
  Overlay_2 {widgetClassPath=&Overlay, parent=SetUnlockAninmationContainer, widget=Overlay_2, slot {class=SizeBoxSlot}}
  Glow {
   widgetClassPath=&Image
   bIsVariable=true
   parent=Overlay_2
   widget=Glow
   properties {
    renderOpacity=0
    visibility=SelfHitTestInvisible
    brush {
     bIsDynamicallyLoaded=false
     drawAs=Image
     imageType=FullColor
     mirroring=NoMirror
     resourceName=None
     resourceObject=/SeasonPass/Materials/RewardTrack/M_UI_BPReboot_SetUnlocking
     tiling=NoTile
     imageSize=<600 600>
     margin=<ltrb 0 0 0 0>
     outlineSettings {
      bUseBrushTransparency=false
      roundingType=HalfHeightRadius
      width=0
      color {colorUseRule=UseColor_Specified, specifiedColor=<rgba 0 0 0 0>}
      cornerRadii=<0 0 0 0>
     }
     tintColor {colorUseRule=UseColor_Specified, specifiedColor=<rgba 1 1 1 1>}
     uVRegion {bIsValid=false, max=<0 0>, min=<0 0>}
    }
   }
   slot {class=OverlaySlot, properties {horizontalAlignment=HAlign_Center, verticalAlignment=VAlign_Center, padding=<ltrb 0 48 0 0>}}
  }
  Att_PipIndicator {
   widgetClassPath=&FortVisualAttachment
   bIsVariable=true
   parent=RootContainer
   widget=Att_PipIndicator
   properties {contentAnchor=<0.5 0>}
   slot {class=OverlaySlot, properties {horizontalAlignment=HAlign_Center, verticalAlignment=VAlign_Bottom}}
  }
  WBP_UIKit_Pip_Indicator {
   widgetClassPath=/FortUILibrary/UIKit/PipList/WBP_UIKit_Pip_Indicator
   bIsVariable=true
   parent=Att_PipIndicator
   widget=WBP_UIKit_Pip_Indicator
   slot {class=SizeBoxSlot, properties {horizontalAlignment=HAlign_Left, verticalAlignment=VAlign_Bottom, padding=<ltrb 0 0 0 -6>}}
  }
 ]
}
