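# Season pass Common UI widget that shows how many sets remain to unlock, as a number over a background image beside a label.
# SetText takes the count, collapses the widget's main overlay at zero, otherwise shows it and fills the number and singular or plural "Set(s) to Unlock" text.
# On Construct, SetCompactUIStyle enlarges the background brush in compact UI, copying its other brush settings.
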WidgetBlueprint /SeasonPass/Widgets/Reboot/WBP_SeasonPass_RemainingSetsToUnlock
graphs [
 EventGraph {nodes [
   Event Construct()
   > SetCompactUIStyle() {is=title K2Node_CallFunction}
  ]}
 SetCompactUIStyle {nodes [
   SetCompactUIStyle() {is=title K2Node_FunctionEntry}
   > Branch(Condition=IsUsingCompactUI())
   n1 = Get NumberOfSetsBackgroundImage() -> (NumberOfSetsBackgroundImage:Image Object Reference)
   n2 = Get!(self:Image Object Reference=@n1) -> (
    Brush_TintColor:Slate Color Structure
    Brush_DrawAs:ESlateBrushDrawType Enum
    Brush_Tiling:ESlateBrushTileType Enum
    Brush_Mirroring:ESlateBrushMirrorType Enum
    Brush_ImageSize:Vector 2D Structure
    Brush_Margin:Margin Structure
    Brush_ResourceObject:Object Reference
    Brush_OutlineSettings:Slate Brush Outline Settings Structure
   )
   > SetBrush!(
    execute
    self=@n1
    InBrush_TintColor:Slate Color Structure=@n2
    InBrush_DrawAs:ESlateBrushDrawType Enum=@n2.Brush_DrawAs
    InBrush_Tiling:ESlateBrushTileType Enum=@n2.Brush_Tiling
    InBrush_Mirroring:ESlateBrushMirrorType Enum=@n2.Brush_Mirroring
    InBrush_ImageSize_X:double=40
    InBrush_ImageSize_Y:double=40
    InBrush_Margin:Margin Structure=@n2.Brush_Margin
    InBrush_ResourceObject:Object Reference=@n2.Brush_ResourceObject
    InBrush_OutlineSettings:Slate Brush Outline Settings Structure=@n2.Brush_OutlineSettings
   ) -> (then)
  ]}
 SetText {nodes [
   n1 = SetText!() -> (then, SetsToUnlock:int) {is=title K2Node_FunctionEntry}
   > n2 = Branch(Condition="Equal (Integer)"(A=@n1))
   n3 = Get MainOverlay() -> (MainOverlay:Overlay Object Reference)
   > SetVisibility(self=@n3, InVisibility=Collapsed)
   SetVisibility(execute=@n2.else, self=@n3, InVisibility=SelfHitTestInvisible)
   > Set Text(Text="To Text (Integer)"(Value=@n1), self=Get NumberOfSets() -> (NumberOfSets:Common Text Object Reference))
   > Set Text(
    Text=SelectText(
     A=Format Text(Format="LOCTEXT(\"Set to Unlock\")")
     B=Format Text(Format="LOCTEXT(\"Sets to Unlock\")")
     bPickA="Equal (Integer)"(A=@n1, B=1)
    )
    self=Get Description() -> (Description:Common Text Object Reference)
   )
  ]}
]
variables []
widgets {
 info {inheritedWidgetCount=0, parentClass=&CommonUserWidget, rootWidgetClass=&Overlay, widgetCount=6}
 widgets [
  MainOverlay {widgetClassPath=&Overlay, bIsVariable=true, widget=MainOverlay}
  HorizontalBox_0 {widgetClassPath=&HorizontalBox, parent=MainOverlay, widget=HorizontalBox_0, slot {class=OverlaySlot}}
  Overlay_0 {
   widgetClassPath=&Overlay
   parent=HorizontalBox_0
   widget=Overlay_0
   slot {class=HorizontalBoxSlot, properties {horizontalAlignment=HAlign_Left, verticalAlignment=VAlign_Center, padding=<ltrb 0 0 8 0>}}
  }
  NumberOfSetsBackgroundImage {
   widgetClassPath=&Image
   bIsVariable=true
   parent=Overlay_0
   widget=NumberOfSetsBackgroundImage
   properties {brush {
     bIsDynamicallyLoaded=false
     drawAs=Image
     imageType=FullColor
     mirroring=NoMirror
     resourceName=None
     resourceObject=/SeasonPass/Widgets/Reboot/Materials/M_SeasonPass_Background_Unlock
     tiling=NoTile
     imageSize=<32 32>
     margin=<ltrb 0 0 0 0>
     outlineSettings {
      bUseBrushTransparency=false
      roundingType=HalfHeightRadius
      width=0
      color {colorUseRule=UseColor_Specified, specifiedColor=<rgba 0 0 0 0>}
      cornerRadii=<0 0 0 0>
     }
     tintColor {colorUseRule=UseColor_Specified, specifiedColor=<rgba 1 1 1 1>}
     uVRegion {bIsValid=false, max=<0 0>, min=<0 0>}
    }}
   slot {class=OverlaySlot, properties {horizontalAlignment=HAlign_Fill, verticalAlignment=VAlign_Fill}}
  }
  NumberOfSets {
   widgetClassPath=&CommonTextBlock
   bIsVariable=true
   parent=Overlay_0
   widget=NumberOfSets
   properties {
    justification=Center
    mobileFontSizeMultiplier=1.15
    scrollStyle=&TextScrollStyle-Base_C
    text=0
    colorAndOpacity {colorUseRule=UseColor_Specified, specifiedColor=<rgba 0.0048 0.004 0.0116 1>}
    font {
     bForceMonospaced=false
     bMaterialIsStencil=false
     fontMaterial=null
     fontObject=&HeadingNow
     letterSpacing=0
     monospacedWidth=1
     size=15.75
     skewAmount=0
     typefaceFontName=86 Bold Italic
     outlineSettings {
      bApplyOutlineToDropShadows=false
      bMiteredCorners=false
      bSeparateFillAlpha=false
      outlineMaterial=null
      outlineSize=0
      outlineColor=<rgba 0 0 0 1>
     }
    }
    shadowOffset=<0 0>
   }
   slot {class=OverlaySlot, properties {horizontalAlignment=HAlign_Center, verticalAlignment=VAlign_Center, padding=<ltrb 12 0 12 0>}}
  }
  Description {
   widgetClassPath=&CommonTextBlock
   bIsVariable=true
   parent=HorizontalBox_0
   widget=Description
   properties {
    mobileFontSizeMultiplier=1.15
    style=&TextStyle-Base_C
    text=Sets to Unlock
    font {
     bForceMonospaced=false
     bMaterialIsStencil=false
     fontMaterial=null
     fontObject=&NotoSans
     letterSpacing=0
     monospacedWidth=1
     size=24
     skewAmount=0
     typefaceFontName=NotoSans-Regular
     outlineSettings {
      bApplyOutlineToDropShadows=false
      bMiteredCorners=false
      bSeparateFillAlpha=false
      outlineMaterial=null
      outlineSize=0
      outlineColor=<rgba 0 0 0 1>
     }
    }
    shadowOffset=<0 0>
   }
   slot {class=HorizontalBoxSlot, properties {horizontalAlignment=HAlign_Left, verticalAlignment=VAlign_Center}}
  }
 ]
}
