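# Season pass reward track panel for claiming several reward sets at once.
# Tabs list the group's unclaimed categories; add/remove builds SetsToClaim, which fills the reward tile list and drives the claim-or-buy-levels state.
# The final set tab can stay locked until almost all others are added.
# Claiming runs async PurchaseItems, rebuilds tabs and pushes set-complete and pass-complete screens; buying levels opens a modal.
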WidgetBlueprint /SeasonPass/Widgets/Reboot/WBP_RewardTrack_ClaimSets
animations [Opening {
  binding_count=1
  length_seconds=1
  object=&self:Opening
  sequence {bindings [MainSizeBox {
     id=#1
     possessed_class=SizeBox
     tracks [
      {
       type=MovieSceneFloatTrack
       display_name=Render Opacity
       sections [{type=MovieSceneFloatSection, end_frame=20, start_frame=0, channels [None {type=MovieSceneScriptingFloatChannel, keys [[0,0],[20,1]]}]}]
      }
      {
       type=MovieScene2DTransformTrack
       sections [{
         type=MovieScene2DTransformSection
         end_frame=20
         start_frame=0
         channels [
          Translation.X {type=MovieSceneScriptingFloatChannel, keys [[0,100],[20,0]]}
          Translation.Y {type=MovieSceneScriptingFloatChannel, const [0,2]}
          Angle {type=MovieSceneScriptingFloatChannel}
          Scale.X {type=MovieSceneScriptingFloatChannel}
          Scale.Y {type=MovieSceneScriptingFloatChannel}
          Shear.X {type=MovieSceneScriptingFloatChannel}
          Shear.Y {type=MovieSceneScriptingFloatChannel}
         ]
        }]
      }
     ]
    }]}
 }]
graphs [
 EventGraph {nodes [
   n1 = "On Clicked (AddRemoveSetButton)"!() -> (OutputDelegate, then, Button:Common Button Base Object Reference) {is=title K2Node_ComponentBoundEvent}
   > n2 = Cast To WBP_RewardTrack_TopNav_Button!(
    execute
    Object=GetTabButtonBaseByViewModel!(
     self:UIKit Tab Group Object Reference=Get SetSelector() -> (SetSelector:WBP UIKit Tab Group Object Reference)
     TabViewModel:MVVM Base Viewmodel Object Reference=Get CurrentlySelectedCategory() -> (CurrentlySelectedCategory:Season Pass Category VM Object Reference)
    ) -> (ReturnValue:Common Button Base Object Reference)
   ) -> (then, CastFailed, AsWBP Reward Track Top Nav Button:WBP Reward Track Top Nav Button Object Reference) {is=title K2Node_DynamicCast}
   n3 = Get Is Equipped!(self:WBP Reward Track Top Nav Button Object Reference=@n2) -> (Is Equipped:bool)
   n4 = NOT Boolean(A=@n3)
   > SetEquipped!(execute, self:WBP Reward Track Top Nav Button Object Reference=@n2, IsEquipped:bool=@n4) -> (then)
   > UpdateAddRemoveButtonLabel!(execute, self, ShowAdd:bool=@n4) -> (then)
   > n5 = Branch(Condition=@n3)
   > FireEventElementInteractedByName!(execute, Element Name:String=Button_Add, WidgetCallingFrom:Widget Object Reference=@n1.Button) -> (then, Success:bool)
   > Fire Event Element Interacted(Element=@n1.Button)
   > Add(
    TargetArray:Array of Season Pass Category VM Object References=Get SetsToClaim() -> (SetsToClaim:Array of Season Pass Category VM Object References)
    NewItem:"Season Pass Category VM Object Reference (by ref)"=Get CurrentlySelectedCategory()
   ) {is=title K2Node_CallArrayFunction}
   > n6 = UpdateFinalSetButtonState!(execute, self, SetRemoved:bool=false) -> (then)
   SetCategoryAndUpdate!(execute=[@n6.then, @n7.then], self, Category:Season Pass Category VM Object Reference=Get CurrentlySelectedCategory()) -> (then)
   > UpdateClaimInfo()
   > Set with Broadcast SetsToClaimCount!(execute, SetsToClaimCount:int=Length(TargetArray:Array of Season Pass Category VM Object References=Get SetsToClaim())) -> (then, Output_Get:int=0)
   FireEventElementInteractedByName(execute=@n5.else, Element Name=Button_Remove, WidgetCallingFrom=@n1.Button)
   > Fire Event Element Interacted(
    Element=@n1.Button
    AdditionalInfo=Get Analytics Info Pass Category Name!(
     MakeFlowAnalyticsAdditionalInfo:"Flow Analytics Additional Info Structure (by ref)"=Make Flow Analytics Additional Info!() -> (FlowAnalyticsAdditionalInfo:Flow Analytics Additional Info Structure) {is=title K2Node_MakeStruct}
     Key:String=SelectedCategory
     SeasonPassSelectedVM:Season Pass Selection VM Object Reference=Get SeasonPassSelectionVM() -> (SeasonPassSelectionVM:Season Pass Selection VM Object Reference)
    ) -> (PassName:Flow Analytics Additional Info Structure) {is=title K2Node_MacroInstance}
   )
   > Remove Item!(
    execute
    TargetArray:Array of Season Pass Category VM Object References=Get SetsToClaim()
    Item:"Season Pass Category VM Object Reference (by ref)"=Get CurrentlySelectedCategory()
   ) -> (then, ReturnValue:bool)
   > n7 = UpdateFinalSetButtonState(SetRemoved=true)
   n8 = "On UIKit Tab Selected (SetSelector)"!() -> (OutputDelegate, then, TabID:Name=None, TabVM:MVVM Base Viewmodel Object Reference) {is=title K2Node_ComponentBoundEvent}
   > n9 = Cast To WBP_RewardTrack_TopNav_Button(Object=GetTabButtonBaseByViewModel(self=Get SetSelector(), TabViewModel=@n8.TabVM))
   > UpdateAddRemoveButtonLabel(ShowAdd=NOT Boolean(A=Get Is Equipped(self=@n9)))
   > n10 = Set CurrentlySelectedCategory(CurrentlySelectedCategory:Season Pass Category VM Object Reference=Get SeasonPassCategoryVM!(self:WBP Reward Track Top Nav Button Object Reference=@n9) -> (SeasonPassCategoryVM:Season Pass Category VM Object Reference)) -> (Output_Get:Season Pass Category VM Object Reference)
   > n11 = Branch(Condition=Is Empty!(TargetArray:Array of Season Pass Reward VM Object References=Get ItemList() -> (ItemList:Array of Season Pass Reward VM Object References)) -> (ReturnValue:bool))
   > n12 = Set SelectedClaimSetsCategory!(
    execute
    SelectedClaimSetsCategory:Season Pass Category VM Object Reference=@n10
    self:Season Pass Selection VM Object Reference=Get SeasonPassSelectionVM()
   ) -> (then, Output_Get:Season Pass Category VM Object Reference)
   UpdateAddRemovedButtonEnabled!(execute=[@n11.else, @n12.then], self, Button:WBP Reward Track Top Nav Button Object Reference=@n9) -> (then)
   n13 = Get SetSelector()
   > Fire Event Element Interacted(
    Element=GetTabButtonBaseByID!(
     self:Common Tab List Widget Base Object Reference=@n13
     TabNameID:Name=GetActiveTab!(self:Common Tab List Widget Base Object Reference=@n13) -> (ReturnValue:Name=None)
    ) -> (ReturnValue:Common Button Base Object Reference)
    AdditionalInfo=Get Analytics Info Pass Category Name(
     MakeFlowAnalyticsAdditionalInfo=Make Flow Analytics Additional Info()
     Key=SelectedCategory
     SeasonPassSelectedVM=Get SeasonPassSelectionVM()
    )
   )
   n14 = "On Clicked (Button_BuyLevels)"!() -> (OutputDelegate, then, Button:Common Button Base Object Reference) {is=title K2Node_ComponentBoundEvent}
   > Fire Event Element Interacted(
    Element=@n14.Button
    AdditionalInfo=Get Analytics Info Pass Name!(
     MakeFlowAnalyticsAdditionalInfo:"Flow Analytics Additional Info Structure (by ref)"=Make Flow Analytics Additional Info()
     Key:String
     SeasonPassSelectedVM:Season Pass Selection VM Object Reference=Get SeasonPassSelectionVM()
    ) -> (PassName:Flow Analytics Additional Info Structure) {is=title K2Node_MacroInstance}
   )
   > n15 = Create Widget!(
    execute:Exec
    Class:User Widget Class Reference=&/WBP_RewardTrack_BuyLevels.WBP_RewardTrack_BuyLevels_C
    OwningPlayer:Player Controller Object Reference
    SeasonPassVM:Season Pass VM Object Reference=Get SeasonPassVM() -> (SeasonPassVM:Season Pass VM Object Reference)
   ) -> (then:Exec, ReturnValue:WBP Reward Track Buy Levels Object Reference)
   > TryPushWidgetAsModalPanel!(execute, PushWidget:User Widget Object Reference=@n15) -> (then)
   > n16 = SetLevelsToBuy!(execute, self:WBP Reward Track Buy Levels Object Reference=@n15, Amount:int=Get NeedToClaim() -> (NeedToClaim:int=0), IsAdditive:bool=false) -> (then)
   n17 = Refresh On Deactivating Buy Rewards() {is=title K2Node_CustomEvent}
   n18 = Assign On Widget Deactivated!(execute:Exec=@n16.then, self:Common Activatable Widget Object Reference=@n15, Delegate:"Delegate (by ref)"=@n17) -> (then:Exec) {is=title K2Node_AssignDelegate}
   n19 = On Deactivating Buy Rewards Changing Tab() {is=title K2Node_CustomEvent}
   Bind Event to On Going to Buy VBucks Page(execute=@n18.then, self:WBP Reward Track Buy Levels Object Reference=@n15, Delegate=@n19) {is=title K2Node_AddDelegate}
   n20 = "On Clicked (Button_ClaimSets)"!() -> (OutputDelegate, then, Button:Common Button Base Object Reference) {is=title K2Node_ComponentBoundEvent}
   > Fire Event Element Interacted(
    Element=@n20.Button
    AdditionalInfo=Get Analytics Info Pass Name(MakeFlowAnalyticsAdditionalInfo=Make Flow Analytics Additional Info(), SeasonPassSelectedVM=Get SeasonPassSelectionVM())
   )
   > n21 = PurchaseItems!(
    execute:Exec
    Items:Array of Season Pass Reward VM Object References=Get ItemList()
    SeasonPass:Season Pass VM Object Reference=Get SeasonPassVM()
    bTryPurchasePrerequisites:bool=true
   ) -> (then:Exec, OnOperationCompleted:Exec, Success:bool=~) {is=title K2Node_AsyncAction}
   > SetIsEnabled(self=@n20.Button, bInIsEnabled=false)
   Call Finish Claim Event Dispatcher!(execute:Exec=@n21.OnOperationCompleted, self:WBP Reward Track Claim Sets Object Reference) -> (then:Exec) {is=title K2Node_CallDelegate}
   > VM_SetupCategoryTabs!(
    execute
    self
    Sets:Array of Season Pass Category VM Object References=Get Categories!(self:Season Pass VM Object Reference=Get SeasonPassVM()) -> (Categories:Array of Season Pass Category VM Object References)
   ) -> (then)
   > SetIsEnabled(bInIsEnabled=true)
   > Branch(Condition=@n21)
   n22 = Get SetsToClaim()
   > n23 = Create Widget!(
    execute
    Class=/SeasonPass/Widgets/Reboot/SetComplete/WBP_UI_BPReboot_SetComplete.WBP_UI_BPReboot_SetComplete_C
    OwningPlayer
    SeasonPassCategoryVM:Season Pass Category VM Object Reference="Get (a copy)"(Array:Array of Season Pass Category VM Object References=@n22) -> (Output:Season Pass Category VM Object Reference)
    NumSetsCompleted:int=Length(TargetArray=@n22)
   ) -> (then, ReturnValue:WBP UI BPReboot Set Complete Object Reference)
   > TryPushWidgetAsModalPanel(PushWidget=@n23)
   > n24 = Branch(Condition=IsMainSetComplete!(SeasonPass:Season Pass VM Object Reference=Get SeasonPassVM()) -> (IsComplete:bool))
   n25 = AnimationFinishing_Event() {is=title K2Node_CustomEvent}
   Assign Animation Finishing!(execute:Exec=@n24.then, self:WBP UI BPReboot Set Complete Object Reference=@n23, Delegate:"Delegate (by ref)"=@n25) -> (then:Exec) {is=title K2Node_AssignDelegate}
   Initialize() {is=title K2Node_CustomEvent}
   > PlayAnimation(InAnimation=Get Opening() -> (Opening:Widget Animation Object Reference))
   > SetIsEnabled(bInIsEnabled=true)
   > ResetOnReopen()
   > n26 = Branch(Condition=HasRewardsLeftToClaim!() -> (HasRewards:bool))
   n27 = Get VisibilitySwitcherIsClaimed() -> (VisibilitySwitcherIsClaimed:Common Visibility Switcher Object Reference)
   > SetActiveWidgetIndex(self:Common Visibility Switcher Object Reference=@n27, Index=0)
   n28 = Get ItemTileView() -> (ItemTileView:Fort Item Tile View Object Reference)
   > ClearListItems(self=@n28)
   > Clear Selection(self:List View Object Reference=@n28)
   SetActiveWidgetIndex(execute=@n26.else, self=@n27, Index=1)
   n29 = SetCategoryAndUpdate!() -> (OutputDelegate, then, Category:Season Pass Category VM Object Reference) {is=title K2Node_CustomEvent}
   > UpdateItemCards!(execute, self, Category:Season Pass Category VM Object Reference=@n29.Category) -> (then)
   n30 = Get ItemTileView()
   > SetSelectedIndex!(execute, self:List View Object Reference=@n30, Index:int=0) -> (then)
   > NavigateToIndex!(execute, self:List View Object Reference=@n30, Index:int=0) -> (then)
   n31 = Get ItemList()
   > n32 = Branch(Condition=Is Empty(TargetArray=@n31))
   > Set SelectedReward!(execute, SelectedReward:Season Pass Reward VM Object Reference, self:Season Pass Selection VM Object Reference=Get SeasonPassSelectionVM()) -> (then, Output_Get:Season Pass Reward VM Object Reference)
   Set SelectedReward(
    execute=@n32.else
    SelectedReward="Get (a copy)"(Array:Array of Season Pass Reward VM Object References=@n31) -> (Output:Season Pass Reward VM Object Reference)
    self=Get SeasonPassSelectionVM()
   )
   Event PreConstruct()
   > SetMaxDesiredHeight!(
    execute
    self:Size Box Object Reference=Get SizeBoxTitle!(self:WBP UIKit Item Description Base Object Reference=Get WBP_UIKit_DescriptionWithSubtitle() -> (WBP_UIKit_DescriptionWithSubtitle:WBP UIKit Description With Subtitle Object Reference)) -> (SizeBoxTitle:Size Box Object Reference)
    InMaxDesiredHeight:float=250
   ) -> (then)
   > SetClipping!(
    execute
    self:Widget Object Reference=Get ScrollBox_Tabs!(self:WBP UIKit Tab Group Base Object Reference=Get SetSelector()) -> (ScrollBox_Tabs:Scroll Box Object Reference)
    InClipping:EWidgetClipping Enum=Inherit
   ) -> (then)
   > SetCompactUIStyle() {is=title K2Node_CallFunction}
   > SetupVisuals()
   Call Quick Close Claim Pages Dispatcher!(execute:Exec=@n19.then, self:WBP Reward Track Claim Sets Object Reference) -> (then:Exec) {is=title K2Node_CallDelegate}
   n33 = Create Widget!(execute=@n25.then, Class=/SeasonPass/Widgets/Reboot/BattlePassComplete/BP_BattlePassCompleteAnim.BP_BattlePassCompleteAnim_C, OwningPlayer) -> (then, ReturnValue:BP Battle Pass Complete Anim Object Reference)
   > PushContentWidget!(execute, PushWidget:Widget Object Reference=@n33) -> (then)
   UpdateClaimInfo(execute=@n34.then)
   Get Analytics Info Pass Category Name(
    MakeFlowAnalyticsAdditionalInfo=Make Flow Analytics Additional Info()
    Key=SelectedCategory
    SeasonPassSelectedVM=Get SeasonPassSelectionVM()
   )
   n34 = UpdateItemCards(execute=@n17.then, Category=Get CurrentlySelectedCategory())
  ]}
 FinishClaimEventDispatcher {nodes [FinishClaimEventDispatcher() {is=title K2Node_FunctionEntry}]}
 GetAnalyticsInfo {nodes [
   GetAnalyticsInfo() {is=title K2Node_FunctionEntry}
   n1 = GetDisplayName!(self:Season Pass VM Object Reference=Get SeasonPassVM()) -> (ReturnValue:Text)
   > Return Node!(
    execute
    ReturnValue:Flow Analytics Additional Info Structure="Add String (Non-Localized)"!(
     Info:"Flow Analytics Additional Info Structure (by ref)"=Make Flow Analytics Additional Info()
     Key:String=PassName
     Value:String=Get Source String from Text!(Text:Text=@n1) -> (ReturnValue:String) {is=title K2Node_CallFunction}
    ) -> (ReturnValue:Flow Analytics Additional Info Structure)
   )
  ]}
 GetCTAButtonEnabled {nodes [
   n1 = GetCTAButtonEnabled!() -> (then, OwnPass:bool=~, AssetsLoaded:bool=~, SetsBeingClaimed:int) {is=title K2Node_FunctionEntry}
   > Return Node!(execute, ShouldEnable:bool=AND Boolean!(A=@n1, B=@n1.AssetsLoaded, C:bool="integer > integer"(A=@n1.SetsBeingClaimed)) -> (ReturnValue))
  ]}
 GetFirstPurchaseablePage {nodes [
   GetFirstPurchaseablePage() {is=title K2Node_FunctionEntry}
   n1 = Get SeasonPassCategoryVM!() -> (SeasonPassCategoryVM)
   > n2 = Is Valid(InputObject=@n1)
   Return Node!(execute=[@n3.Completed, @n2.Is Not Valid], Page:int=-1)
   > n3 = For Each Loop with Break!(
    Exec:Exec
    Array:Array of Season Pass Page VM Object References=Get Pages!(self:Season Pass Category VM Object Reference=@n1) -> (Pages:Array of Season Pass Page VM Object References)
    Break:Exec
   ) -> (LoopBody:Exec, Array Element:Season Pass Page VM Object Reference, Array Index:int, Completed:Exec)
   > n4 = For Each Loop with Break(Array:Array of Season Pass Track VM Object References=Get Tracks!(self:Season Pass Page VM Object Reference=@n3) -> (Tracks:Array of Season Pass Track VM Object References)) -> (Array Element:Season Pass Track VM Object Reference)
   > n5 = For Each Loop with Break(Array:Array of Season Pass Reward VM Object References=Get Rewards!(self:Season Pass Track VM Object Reference=@n4) -> (Rewards:Array of Season Pass Reward VM Object References)) -> (Array Element:Season Pass Reward VM Object Reference)
   > n6 = Branch(Condition=Get bIsRewardOwned!(self:Season Pass Reward VM Object Reference=@n5) -> (bIsRewardOwned:bool))
   Return Node(execute=@n6.else, Page=@n3.Array Index)
  ]}
 GetIsClaimingAlmostAllSets {nodes [
   GetIsClaimingAlmostAllSets() {is=title K2Node_FunctionEntry}
   n1 = Get SetsToClaim()
   n2 = GetTabCount!(self:Common Tab List Widget Base Object Reference=Get SetSelector()) -> (ReturnValue:int=0)
   > Return Node!(execute, ReturnValue:bool="integer >= integer"(A=Length(TargetArray=@n1), B=int - int(A=@n2)))
  ]}
 GetLastUnlockablePage {nodes [
   GetLastUnlockablePage() {is=title K2Node_FunctionEntry}
   n1 = Get SeasonPassCategoryVM()
   > n2 = Is Valid(InputObject=@n1)
   Return Node!(execute=@n2.Is Not Valid, LastUnlockablePage:int=-1)
   > n3 = For Each Loop with Break(Array:Array of Season Pass Page VM Object References=Get Pages(self=@n1)) -> (Array Element:Season Pass Page VM Object Reference)
   n4 = Get Pages(self=@n1)
   Return Node(execute=@n3.Completed, LastUnlockablePage=int - int(A=Length(TargetArray:Array of Season Pass Page VM Object References=@n4)))
   > Branch(Condition=Get bIsTimeLocked!(self:Season Pass Element Base VM Object Reference=@n3) -> (bIsTimeLocked:bool))
   > Return Node(LastUnlockablePage=int - int(A=@n3.Array Index))
  ]}
 HandleSetsToUnlock {nodes [
   n1 = HandleSetsToUnlock!() -> (then, SetCurrency:int) {is=title K2Node_FunctionEntry}
   n2 = Get WBP_SeasonPass_RemainingSetsToUnlock() -> (WBP_SeasonPass_RemainingSetsToUnlock:WBP Season Pass Remaining Sets To Unlock Object Reference)
   > SetText!(execute, self:WBP Season Pass Remaining Sets To Unlock Object Reference=@n2, SetsToUnlock:int=@n1) -> (then)
   > SetVisibility(self=@n2, InVisibility=Select(Option 0=Collapsed, Option 1=SelfHitTestInvisible, Index="integer > integer"(A=@n1)))
  ]}
 HasRewardsLeftToClaim {nodes [
   HasRewardsLeftToClaim() {is=title K2Node_FunctionEntry}
   > Set CurrentGroup(CurrentGroup:String=GetGroupName!(self:Season Pass Element Base VM Object Reference=Get SeasonPassCategoryVM()) -> (ReturnValue:String)) -> (Output_Get:String)
   > n1 = For Each Loop(Array:Array of Season Pass Category VM Object References=Get Categories(self=Get SeasonPassVM())) -> (Array Element:Season Pass Category VM Object Reference)
   Return Node!(execute=@n1.Completed, HasRewards:bool=false)
   n2 = GetGroupName(self=@n1)
   n3 = Get CurrentGroup() -> (CurrentGroup:String)
   n4 = Get bIsFullyClaimed!(self:Season Pass Category VM Object Reference=@n1) -> (bIsFullyClaimed:bool)
   > Branch(Condition=AND Boolean!(A="Equal Exactly (String)"(A=@n2, B=@n3), B=NOT Boolean(A=@n4)) -> (ReturnValue))
   > Return Node(HasRewards=true)
  ]}
 NeedsButtonOnPassCategory {nodes [
   NeedsButtonOnPassCategory() {is=title K2Node_FunctionEntry}
   n1 = Get SeasonPassCategoryVM()
   > n2 = Is Valid(InputObject=@n1)
   Return Node!(execute=@n2.Is Not Valid, Value:bool=false)
   > n3 = For Each Loop with Break(Array=Get Pages(self=@n1))
   Return Node(execute=@n3.Completed, Value=false)
   > n4 = For Each Loop with Break(Array:Array of Season Pass Track VM Object References=Get Tracks(self=@n3)) -> (Array Element:Season Pass Track VM Object Reference)
   > n5 = For Each Loop with Break(Array:Array of Season Pass Reward VM Object References=Get Rewards(self=@n4)) -> (Array Element:Season Pass Reward VM Object Reference)
   > n6 = Branch(Condition=Get bIsRewardOwned(self=@n5))
   n7 = Branch(
    execute=@n6.else
    Condition=OR Boolean(
     A=Get bIsQuestDriven!(self:Season Pass Reward VM Object Reference=@n5) -> (bIsQuestDriven:bool)
     B=IsTimeLocked!(self:Season Pass Element Base VM Object Reference=@n5) -> (ReturnValue:bool)
    )
   )
   > Return Node(Value=false)
   Return Node(execute=@n7.else, Value=true)
  ]}
 QuickCloseClaimPagesDispatcher {nodes [QuickCloseClaimPagesDispatcher() {is=title K2Node_FunctionEntry}]}
 ResetOnReopen {nodes [
   ResetOnReopen() {is=title K2Node_FunctionEntry}
   > Set TopNavFinalSetButton(TopNavFinalSetButton:WBP Reward Track Top Nav Button Object Reference) -> (Output_Get:WBP Reward Track Top Nav Button Object Reference)
   n1 = Get SeasonPassVM()
   > SetText!(
    execute
    self:WBP UIKit Highlight Base Object Reference=Get Status_ToClaim() -> (Status_ToClaim:WBP UIKit Highlight Object Reference)
    Text:Text=Format Text!(
     Format="LOCTEXT(\"{rewards} TO CLAIM\")"
     rewards:int=GetNumPurchaseableRewards!(self:Season Pass Element Base VM Object Reference=@n1) -> (ReturnValue:int=0)
    ) -> (Result)
   ) -> (then)
   > Set Text(Text="To Text (Integer)"(Value=0), self=Get Text_TotalRewardsValue() -> (Text_TotalRewardsValue:Common Text Object Reference))
   > Clear(TargetArray:Array of Season Pass Category VM Object References=Get SetsToClaim())
   > Set with Broadcast SetsToClaimCount(SetsToClaimCount=0) -> (Output_Get=0)
   > SetText!(
    execute
    self:Fort CTAButton Object Reference=Get AddRemoveSetButton() -> (AddRemoveSetButton:WBP UIKit Button Quiet Object Reference)
    ButtonText:"Text (by ref)"=Format Text!(Format="LOCTEXT(\"Add Set\")") -> (Result)
   ) -> (then)
   n2 = Get SeasonPassVM()
   > Is Valid(InputObject=@n2)
   > VM_SetupCategoryTabs(Sets=Get Categories(self=@n2))
   n3 = Get SetSelector()
   n4 = GetTabButtonBaseByViewModel(self=@n3, TabViewModel=Get SeasonPassCategoryVM())
   > n5 = Is Valid(InputObject=@n4)
   SelectTabByIndex!(execute=@n5.Is Not Valid, self:UIKit Tab Group Object Reference=@n3, Index:int=0, bSuppressClickFeedback:bool=true) -> (then, ReturnValue:bool)
   SetIsSelected!(execute=@n5.Is Valid, self:Common Button Base Object Reference=@n4, InSelected:bool=true, bGiveClickFeedback:bool=false) -> (then)
  ]}
 SetBlackBackplateCorner {nodes [
   SetBlackBackplateCorner() {is=title K2Node_FunctionEntry}
   > n1 = GetDynamicMaterial(self=Get Backplate!(self:WBP UIKit Backplate Base Object Reference=Get Backplate_Black() -> (Backplate_Black:WBP UIKit Backplate Object Reference)) -> (Backplate:Image Object Reference))
   > SetVectorParameterValue(
    self=@n1
    ParameterName=Corner Radius
    Value=Make Linear Color!(R:float=0, G:float=20, B:float=0, A:float=20) -> (LinearColor:Linear Color Structure) {is=title K2Node_MakeStruct}
   )
  ]}
 SetCompactUIBackToTop {nodes [
   SetCompactUIBackToTop() {is=title K2Node_FunctionEntry}
   n1 = Get WBP_MPLocker_BackToTopButton() -> (WBP_MPLocker_BackToTopButton:WBP MPLocker Back To Top Button Object Reference)
   > SetPadding!(
    execute
    self:Size Box Slot Object Reference=SlotAsSizeBoxSlot!(Widget:Widget Object Reference=@n1) -> (ReturnValue:Size Box Slot Object Reference)
    InPadding_Left:float=4
    InPadding_Top:float=0
    InPadding_Right:float=0
    InPadding_Bottom:float=150
   ) -> (then)
   n2 = Get WBP_UIKit_Button_BackToTop_Touch!(self:WBP MPLocker Back To Top Button Object Reference=@n1) -> (WBP_UIKit_Button_BackToTop_Touch:WBP UIKit Button Quiet Object Reference)
   > SetVisibility(self=Get Texts!(self:WBP UIKit Button CTA Base Object Reference=@n2) -> (Texts:Vertical Box Object Reference), InVisibility=Collapsed)
   > SetIconMaterial!(
    execute
    self:WBP UIKit Button CTA Base Object Reference=@n2
    IconMaterial:Material Instance Object Reference=/FortUILibrary/UIKit/Buttons/MaterialInstances/GenericButtons/MI_GenericIconDeepBlue
   ) -> (then)
   > SetIcon!(execute, self:WBP UIKit Button CTA Base Object Reference=@n2, Icon:Object Reference=/Game/UI/CodeReferenced/T_UI_DirectionArrow) -> (then)
   > Set SmallIconSize!(execute, SmallIconSize:Vector 2D Structure=(X=32,Y=32), self:WBP UIKit Button CTA Base Object Reference=@n2) -> (then, Output_Get:Vector 2D Structure)
   > Set MobileSmallHeight!(execute, MobileSmallHeight:int=120, self:WBP UIKit Button CTA Base Object Reference=@n2) -> (then, Output_Get:int=0)
   > Set!(
    execute
    IconPadding_Left:float=-16
    IconPadding_Top:float=4
    IconPadding_Right:float=-16
    IconPadding_Bottom:float=16
    self:WBP UIKit Button CTA Base Object Reference=@n2
   ) -> (then, Output_Get:Margin Structure)
  ]}
 SetCompactUIStyle {nodes [
   SetCompactUIStyle() {is=title K2Node_FunctionEntry}
   > n1 = Set with Broadcast IsUsingCompactUI!(execute, IsUsingCompactUI:bool=IsUsingCompactUI()) -> (then, Output_Get:bool)
   > Branch(Condition=@n1)
   > SetCompactUIBackToTop() {is=title K2Node_CallFunction}
   > Set UseArrowNavButtons!(execute, UseArrowNavButtons:bool=true, self:WBP UIKit Tab Group Base Object Reference=Get SetSelector()) -> (then, Output_Get:bool)
   > Set with Broadcast IsTabletRatio!(
    execute
    IsTabletRatio:bool=CompareToScreenAspectRatio!(Comparison:EAspectRatioComparison Enum=LessThanOrEqualTo, AspectRatio:EAspectRatio Enum=Ratio_3_2) -> (ReturnValue:bool)
   ) -> (then, Output_Get:bool)
  ]}
 SetupVisuals {nodes [
   SetupVisuals() {is=title K2Node_FunctionEntry}
   > SetBlackBackplateCorner() {is=title K2Node_CallFunction}
  ]}
 UpdateAddRemoveButtonLabel {nodes [
   n1 = UpdateAddRemoveButtonLabel!() -> (then, ShowAdd:bool=~) {is=title K2Node_FunctionEntry}
   > SetText(
    self=Get AddRemoveSetButton()
    ButtonText=Select(
     Option 0:"Text (by ref)"=Format Text(Format="LOCTEXT(\"Remove Set\")")
     Option 1:"Text (by ref)"=Format Text(Format="LOCTEXT(\"Add Set\")")
     Index=@n1
    ) -> (ReturnValue:"Text (by ref)")
   )
  ]}
 UpdateAddRemovedButtonEnabled {nodes [
   n1 = UpdateAddRemovedButtonEnabled!() -> (then, Button:WBP Reward Track Top Nav Button Object Reference) {is=title K2Node_FunctionEntry}
   > SetIsEnabled(
    self=Get AddRemoveSetButton()
    bInIsEnabled=NOT Boolean(A=GetIsLocked!(self:WBP Reward Track Top Nav Button Object Reference=@n1, ViewmodelIsLocked:bool=false, IsOverridingLocked:bool=true) -> (IsLocked:bool))
   )
  ]}
 UpdateClaimInfo {nodes [
   UpdateClaimInfo() {is=title K2Node_FunctionEntry}
   > Set NumRewards(NumRewards:int=Length(TargetArray:Array of Season Pass Reward VM Object References=Get ItemList())) -> (Output_Get:int=0)
   > n1 = Sequence!(execute) -> (then_0, then_1, then_2:Exec)
   n2 = Branch(execute=@n1.then_1, Condition="integer <= integer"(A=Get NumToBuy() -> (NumToBuy:int=0)))
   n3 = Get SeasonPassVM()
   n4 = Get Status_ToClaim()
   > SetText!(
    execute
    self:WBP UIKit Highlight Base Object Reference=@n4
    Text:Text=Format Text!(Format="LOCTEXT(\"{rewards} TO CLAIM\")", rewards:int=GetNumPurchaseableRewards(self=@n3) -> (ReturnValue=0)) -> (Result)
   ) -> (then)
   > SetType!(execute, self:WBP UIKit Highlight Base Object Reference=@n4, Type:E_UI_HighlightType Enum=NewEnumerator0) -> (then)
   > HandleSetsToUnlock!(
    execute
    self
    SetCurrency:int=GetNumUnlockableDirectChildren!(self:Season Pass Element Base VM Object Reference=Get SeasonPassVM()) -> (ReturnValue:int=0)
   ) -> (then) {is=title K2Node_CallFunction}
   n5 = Get HB_Buttons() -> (HB_Buttons:Common Visibility Switcher Object Reference)
   > SetActiveWidgetIndex(self=@n5, Index=0)
   n6 = Get FortPreloadAssetsVM() -> (FortPreloadAssetsVM:Fort Preload Assets VM Object Reference)
   > SetAssetsToPreload!(
    execute
    self:Fort Preload Assets VM Object Reference=@n6
    AssetsToPreload:Array of Soft Object References=Make Array!() -> (Array:Array of Soft Object References)
    bInstallSoftReferences:bool=false
   ) -> (then)
   Set NeedToClaim(execute=@n2.else, NeedToClaim:int=Get NumToBuy() -> (NumToBuy=0)) -> (Output_Get:int=0)
   > SetText(self=@n4, Text=Format Text(Format="LOCTEXT(\"NEED {rewards} MORE LEVELS TO CLAIM SETS\")", rewards=Get NumToBuy() -> (NumToBuy=0)))
   > SetType(self=@n4, Type=NewEnumerator3)
   > SetVisibility(self=Get WBP_SeasonPass_RemainingSetsToUnlock(), InVisibility=Collapsed)
   > SetActiveWidgetIndex(self=@n5, Index=1)
   > SetAssetsToPreload(self=@n6, AssetsToPreload=Get PreloadAssets() -> (PreloadAssets:Array of Soft Object References), bInstallSoftReferences=false)
   Set Text(execute=@n1.then_2, Text="To Text (Integer)"(Value=Get NumRewards() -> (NumRewards:int=0)), self=Get Text_TotalRewardsValue())
   n7 = For Each Loop(Exec=@n1.then_0, Array=Get SetsToClaim())
   > Set MinLevelRequired(MinLevelRequired:int="Max (Integer)"(
     A=GetCategoryUnlockLevelRequirement!(self:Season Pass Category VM Object Reference=@n7) -> (ReturnValue:int=0)
     B=Get MinLevelRequired() -> (MinLevelRequired:int=0)
    )) -> (Output_Get:int=0)
   n8 = int - int(
    A=Get MinLevelRequired() -> (MinLevelRequired=0)
    B=GetCurrentLevel!(self:Season Pass VM Object Reference=Get SeasonPassVM()) -> (ReturnValue:int=0)
   )
   n9 = int - int(A=Get NumRewards() -> (NumRewards=0), B=GetNumPurchaseableRewards(self=Get SeasonPassVM()) -> (ReturnValue=0))
   Set NumToBuy(
    execute=@n7.Completed
    NumToBuy:int="Max (Integer)"(A="Clamp (Integer)"(Value=@n8, Min=0, Max=@n8) -> (ReturnValue=0), B="Clamp (Integer)"(Value=@n9, Min=0, Max=@n9) -> (ReturnValue=0))
   ) -> (Output_Get:int=0)
  ]}
 UpdateFinalSetButtonState {nodes [
   UpdateFinalSetButtonState!() -> (then, SetRemoved:bool=~) {is=title K2Node_FunctionEntry}
   > n1 = Get!(execute:Exec) -> (then:Exec, else:Exec, TopNavFinalSetButton:WBP Reward Track Top Nav Button Object Reference)
   > n2 = Sequence!(execute) -> (then_0, then_1)
   n3 = Get SetsToClaim()
   n4 = Get SeasonPassCategoryVM!(self:WBP Reward Track Top Nav Button Object Reference=@n1) -> (SeasonPassCategoryVM)
   > Branch(Condition=AND Boolean(
     A=Contains Item(TargetArray:Array of Season Pass Category VM Object References=@n3, ItemToFind:"Season Pass Category VM Object Reference (by ref)"=@n4)
     B=Get SetRemoved() -> (SetRemoved:bool)
    ))
   n5 = Get TopNavFinalSetButton() -> (TopNavFinalSetButton:WBP Reward Track Top Nav Button Object Reference)
   > SetEquipped(self=@n5, IsEquipped=false)
   > Remove Item(TargetArray=Get SetsToClaim(), Item=Get SeasonPassCategoryVM(self=@n5))
   SetLocked!(
    execute=@n2.then_1
    self:WBP Reward Track Top Nav Button Object Reference=@n1
    Locked:bool=NOT Boolean(A=GetIsClaimingAlmostAllSets!() -> (ReturnValue:bool) {is=title K2Node_CallFunction})
   ) -> (then)
  ]}
 UpdateItemCards {nodes [
   UpdateItemCards!() -> (then, Category:Season Pass Category VM Object Reference) {is=title K2Node_FunctionEntry}
   > Clear(TargetArray:Array of Season Pass Reward VM Object References=Get ItemList())
   > n1 = For Each Loop(Array=Get SetsToClaim())
   n2 = Get ItemList()
   Set List Items!(execute=@n1.Completed, self:List View Object Reference=Get ItemTileView(), InListItems:Array of Object References=@n2) -> (then) {is=title K2Node_CallFunction}
   n3 = Length(TargetArray=@n2)
   > SetVisibility(
    self=Get Text_NoRewardsAdded() -> (Text_NoRewardsAdded:Common Text Object Reference)
    InVisibility=Select(Option 0:ESlateVisibility Enum=SelfHitTestInvisible, Option 1:ESlateVisibility Enum=Collapsed, Index="integer > integer"(A=@n3)) -> (ReturnValue:ESlateVisibility Enum)
   )
   n4 = For Each Loop with Break(Exec=@n1.LoopBody, Array:Array of Season Pass Page VM Object References=Get Pages(self=@n1)) -> (Array Element:Season Pass Page VM Object Reference)
   > n5 = For Each Loop with Break(Array:Array of Season Pass Track VM Object References=Get Tracks(self=@n4)) -> (Array Element:Season Pass Track VM Object Reference)
   > n6 = For Each Loop with Break(Array:Array of Season Pass Reward VM Object References=Get Rewards(self=@n5)) -> (Array Element:Season Pass Reward VM Object Reference)
   > n7 = Branch(Condition=OR Boolean!(A=Get bIsRewardOwned(self=@n6), B=IsTimeLocked(self=@n6), C:bool=Get bIsQuestDriven(self=@n6)) -> (ReturnValue))
   Add(
    execute=@n7.else
    TargetArray:Array of Season Pass Reward VM Object References=Get ItemList()
    NewItem:"Season Pass Reward VM Object Reference (by ref)"=@n6
   )
   > Add(
    TargetArray:Array of Soft Object References=Get PreloadAssets()
    NewItem:"Soft Object Reference (by ref)"=GetButtonIcon!(self:Season Pass Reward VM Object Reference=@n6) -> (ReturnValue:Soft Object Reference)
   )
  ]}
 VM_PreviewItemChange {nodes [
   n1 = VM_PreviewItemChange!() -> (then, Preview Item VM:Fort Item VM Object Reference) {is=title K2Node_FunctionEntry}
   > Is Valid(InputObject=@n1)
   > Set FortItemDetailsVM!(
    execute
    FortItemDetailsVM:Fort Item Details VM Object Reference=Get ItemDetails!(self:Fort Item VM Object Reference=@n1) -> (ItemDetails:Fort Item Details VM Object Reference)
    self:WBP UIKit Description With Subtitle Object Reference=Get WBP_UIKit_DescriptionWithSubtitle()
   ) -> (then, Output_Get:Fort Item Details VM Object Reference)
   n2 = Get SeasonPassSelectionVM()
   n3 = Get SelectedReward!(self:Season Pass Selection VM Object Reference=@n2) -> (SelectedReward:Season Pass Reward VM Object Reference)
   n4 = AND Boolean(
    A=IsValid(Object=Get SelectedClaimSetsCategory!(self:Season Pass Selection VM Object Reference=@n2) -> (SelectedClaimSetsCategory:Season Pass Category VM Object Reference))
    B=NOT Boolean(A=IsValid(Object=@n3))
   )
   > SetVisibility(self=Get WBP_UIKit_DescriptionWithSubtitle(), InVisibility=Select(Option 0=Hidden, Option 1=Visible, Index=NOT Boolean(A=@n4)))
  ]}
 VM_SetupCategoryTabs {nodes [
   VM_SetupCategoryTabs!() -> (then, Sets:Array of Season Pass Category VM Object References) {is=title K2Node_FunctionEntry}
   > n1 = Sequence!(execute) -> (then_0, then_1, then_2:Exec)
   > RemoveAllTabs(self:Common Tab List Widget Base Object Reference=Get SetSelector())
   > Set CurrentGroup(CurrentGroup=GetGroupName(self=Get SeasonPassCategoryVM!() -> (SeasonPassCategoryVM)))
   > n2 = For Each Loop(Array=Get Sets() -> (Sets:Array of Season Pass Category VM Object References))
   n3 = Get bIsFullyClaimed(self=@n2)
   n4 = GetGroupName(self=@n2)
   n5 = Get CurrentGroup()
   > Branch(Condition=AND Boolean(A="Equal Exactly (String)"(A=@n4, B=@n5), B=NOT Boolean(A=@n3)))
   > AddTab!(
    execute
    self:WBP UIKit Tab Group Base Object Reference=Get SetSelector()
    Tab VM:MVVM Base Viewmodel Object Reference=@n2
    Tab ID:Name=String To Name(InString=GetDisplayName!(Object:Object Reference=@n2) -> (ReturnValue:String))
    Index:int=-1
   ) -> (then)
   n6 = For Each Loop(Exec=@n1.then_1, Array=Get Sets())
   > n7 = Cast To WBP_RewardTrack_TopNav_Button(Object=GetTabButtonBaseByViewModel(self=Get SetSelector(), TabViewModel=@n6))
   > SetEquipped(self=@n7, IsEquipped=false)
   > Set IsOverridingLockedState!(execute, IsOverridingLockedState:bool=true, self:WBP Reward Track Top Nav Button Object Reference=@n7) -> (then, Output_Get:bool)
   > SetLocked(self=@n7, Locked=false)
   > Set TopNavFinalSetButton(TopNavFinalSetButton=@n7)
   n8 = GetIsClaimingAlmostAllSets()
   n9 = Get TopNavFinalSetButton()
   n10 = Get Pages(self=Get SeasonPassCategoryVM!(self:WBP Reward Track Top Nav Button Object Reference=@n9) -> (SeasonPassCategoryVM))
   n11 = "Get (a copy)"(
    Array:Array of Season Pass Page VM Object References=@n10
    Dimension 1=int - int(A=Length(TargetArray:Array of Season Pass Page VM Object References=@n10))
   ) -> (Output:Season Pass Page VM Object Reference)
   n12 = Get Rewards(self="Get (a copy)"(Array:Array of Season Pass Track VM Object References=Get Tracks(self=@n11)) -> (Output:Season Pass Track VM Object Reference))
   n13 = "Get (a copy)"(
    Array:Array of Season Pass Reward VM Object References=@n12
    Dimension 1=int - int(A=Length(TargetArray:Array of Season Pass Reward VM Object References=@n12))
   ) -> (Output:Season Pass Reward VM Object Reference)
   SetLocked(
    execute=@n1.then_2
    self=@n9
    Locked=AND Boolean(
     A=NOT Boolean(A=@n8)
     B=Get bDoesRequireProgressForUnlock!(self:Season Pass Reward VM Object Reference=@n13) -> (bDoesRequireProgressForUnlock:bool)
    )
   )
  ]}
]
variables [
 CompactUI_BackToTopPadding {type=Margin, default=<ltrb 20 0 0 0>}
 CurrentlySelectedCategory {type=SeasonPassCategoryVM, default=null}
 Default_ItemDescriptionPadding {type=Margin, default=<ltrb 56 0 0 0>}
 FinishClaimEventDispatcher {}
 IsTabletRatio {type=boolean}
 IsUsingCompactUI {type=boolean}
 ItemList {type="array<SeasonPassRewardVM>", default []}
 NeedToClaim {type=integer}
 PreloadAssets {type="array<Object>", default []}
 QuickCloseClaimPagesDispatcher {}
 SetsToClaim {type="array<SeasonPassCategoryVM>", default []}
 SetsToClaimCount {type=integer}
 TabletRatioUI_BackToTopPadding {type=Margin, default=<ltrb 20 0 0 80>}
 TabletRatioUI_ItemDescriptionPadding {type=Margin, default=<ltrb 56 0 0 80>}
 TopNavFinalSetButton {type=WBP_RewardTrack_TopNav_Button_C, default=null}
]
widgets {
 info {inheritedWidgetCount=0, parentClass=&CommonActivatableWidget, rootWidgetClass=&SizeBox, widgetCount=31}
 widgets [
  MainSizeBox {widgetClassPath=&SizeBox, bIsVariable=true, widget=MainSizeBox, properties {bOverride_WidthOverride=true, widthOverride=760}}
  VisibilitySwitcherIsClaimed {
   widgetClassPath=&CommonVisibilitySwitcher
   bIsVariable=true
   parent=MainSizeBox
   widget=VisibilitySwitcherIsClaimed
   slot {class=SizeBoxSlot}
  }
  Overlay_1 {widgetClassPath=&Overlay, parent=VisibilitySwitcherIsClaimed, widget=Overlay_1, slot {class=CommonVisibilitySwitcherSlot}}
  Backplate_Dark {
   widgetClassPath=&WBP_UIKit_Backplate
   bIsVariable=true
   parent=Overlay_1
   widget=Backplate_Dark
   properties {brightness=Dark, cornerRadius=Interactive Component}
   slot {class=OverlaySlot, properties {horizontalAlignment=HAlign_Fill, verticalAlignment=VAlign_Fill}}
  }
  VB_All {
   widgetClassPath=&VerticalBox
   parent=Overlay_1
   widget=VB_All
   slot {class=OverlaySlot, properties {horizontalAlignment=HAlign_Fill, verticalAlignment=VAlign_Fill}}
  }
  SizeBox_88 {
   widgetClassPath=&SizeBox
   parent=VB_All
   widget=SizeBox_88
   properties {bOverride_WidthOverride=true, widthOverride=760}
   slot {class=VerticalBoxSlot, properties {padding=<ltrb 0 0 0 16>, size {sizeRule=Fill, value=1}}}
  }
  Overlay_4 {widgetClassPath=&Overlay, parent=SizeBox_88, widget=Overlay_4, slot {class=SizeBoxSlot}}
  VerticalBox_71 {
   widgetClassPath=&VerticalBox
   parent=Overlay_4
   widget=VerticalBox_71
   slot {class=OverlaySlot, properties {horizontalAlignment=HAlign_Fill, verticalAlignment=VAlign_Fill, padding=<ltrb 0 16 0 0>}}
  }
  SetSelector {
   widgetClassPath=/FortUILibrary/UIKit/Tabs/WBP_UIKit_TabGroup
   bIsVariable=true
   parent=VerticalBox_71
   widget=SetSelector
   properties {
    arrowButtonsClass=/FortniteCoreUI/Frontend/WBP_UIKit_Button_TabNavArrow_FrontEnd.WBP_UIKit_Button_TabNavArrow_FrontEnd_C
    backplateBrightness=Obscure
    backplateCornerRadius=Circle
    buttonClass=&/WBP_RewardTrack_TopNav_Button.WBP_RewardTrack_TopNav_Button_C
    promptBackplateBrightness=Dark
    tabSpacing=4
    usePromptBackplate=false
    backplatePadding=<ltrb 24 24 24 24>
    initialTabsData [
     {
      bEnableTabId=false
      tabId=None
      tabViewModel=&self:WidgetTree.SetSelector.UIKitSelectableButtonViewModel_0
     }
     {
      bEnableTabId=false
      tabId=None
      tabViewModel=&self:WidgetTree.SetSelector.UIKitSelectableButtonViewModel_2
     }
     {
      bEnableTabId=false
      tabId=None
      tabViewModel=&self:WidgetTree.SetSelector.UIKitSelectableButtonViewModel_1
     }
    ]
    tabGroupPadding=<ltrb 20 20 20 20>
   }
   slot {class=VerticalBoxSlot, properties {horizontalAlignment=HAlign_Center}}
  }
  AddRemoveSetButton {
   widgetClassPath=/FortUILibrary/UIKit/Buttons/WBP_UIKit_Button_Quiet
   bIsVariable=true
   parent=VerticalBox_71
   widget=AddRemoveSetButton
   properties {text=Add Set, triggeringInputAction {dataTable=&InputActionDataTable, rowName=GenericFaceButtonLeft}}
   slot {class=VerticalBoxSlot, properties {horizontalAlignment=HAlign_Center}}
  }
  Overlay_0 {
   widgetClassPath=&Overlay
   parent=VerticalBox_71
   widget=Overlay_0
   properties {clipping=ClipToBounds}
   slot {class=VerticalBoxSlot, properties {padding=<ltrb -16 16 -8 0>, size {sizeRule=Fill, value=1}}}
  }
  ItemTileView {
   widgetClassPath=/Script/FortniteUI.FortItemTileView
   bIsVariable=true
   parent=Overlay_0
   widget=ItemTileView
   properties {
    bEntrySizeIncludesEntrySpacing=false
    clipping=Inherit
    entryHeight=160
    entryWidgetClass=/SeasonPass/Widgets/SeasonPassClaimPagesReward.SeasonPassClaimPagesReward_C
    horizontalEntrySpacing=16
    navigation=&self:WidgetTree.ItemTileView.WidgetNavigation_1
    numDesignerPreviewEntries=12
    scrollIntoViewAlignment=TopOrLeft
    tileAlignment=LeftAligned
    verticalEntrySpacing=16
    scrollBarStyle {
     thickness=10
     draggedThumbImage {
      bIsDynamicallyLoaded=false
      drawAs=Box
      imageType=FullColor
      mirroring=NoMirror
      resourceName=../../../Engine/Content/Slate/Common/Scrollbar_Thumb.png
      resourceObject=/FortUILibrary/Foundation/Art_ToolKit/Tools/M_UI_Circle_White
      tiling=NoTile
      imageSize=<10 10>
      margin=<ltrb 0.5 0.5 0.5 0.5>
      outlineSettings {
       bUseBrushTransparency=false
       roundingType=HalfHeightRadius
       width=0
       color {colorUseRule=UseColor_Specified, specifiedColor=<rgba 0 0 0 0>}
       cornerRadii=<0 0 0 0>
      }
      tintColor {colorUseRule=UseColor_Specified, specifiedColor=<rgba 1 1 1 1>}
      uVRegion {bIsValid=false, max=<0 0>, min=<0 0>}
     }
     horizontalBackgroundImage {
      bIsDynamicallyLoaded=false
      drawAs=NoDrawType
      imageType=NoImage
      mirroring=NoMirror
      resourceName=None
      resourceObject=null
      tiling=NoTile
      imageSize=<0 0>
      margin=<ltrb 0 0 0 0>
      outlineSettings {
       bUseBrushTransparency=false
       roundingType=HalfHeightRadius
       width=0
       color {colorUseRule=UseColor_Specified, specifiedColor=<rgba 0 0 0 0>}
       cornerRadii=<0 0 0 0>
      }
      tintColor {colorUseRule=UseColor_Specified, specifiedColor=<rgba 1 1 1 1>}
      uVRegion {bIsValid=false, max=<0 0>, min=<0 0>}
     }
     horizontalBottomSlotImage {
      bIsDynamicallyLoaded=false
      drawAs=Image
      imageType=FullColor
      mirroring=NoMirror
      resourceName=../../../Engine/Content/Slate/Common/Scrollbar_Background_Horizontal.png
      resourceObject=null
      tiling=NoTile
      imageSize=<8 8>
      margin=<ltrb 0 0 0 0>
      outlineSettings {
       bUseBrushTransparency=false
       roundingType=HalfHeightRadius
       width=0
       color {colorUseRule=UseColor_Specified, specifiedColor=<rgba 0 0 0 0>}
       cornerRadii=<0 0 0 0>
      }
      tintColor {colorUseRule=UseColor_Specified, specifiedColor=<rgba 1 1 1 1>}
      uVRegion {bIsValid=false, max=<0 0>, min=<0 0>}
     }
     horizontalTopSlotImage {
      bIsDynamicallyLoaded=false
      drawAs=Image
      imageType=FullColor
      mirroring=NoMirror
      resourceName=../../../Engine/Content/Slate/Common/Scrollbar_Background_Horizontal.png
      resourceObject=null
      tiling=NoTile
      imageSize=<8 8>
      margin=<ltrb 0 0 0 0>
      outlineSettings {
       bUseBrushTransparency=false
       roundingType=HalfHeightRadius
       width=0
       color {colorUseRule=UseColor_Specified, specifiedColor=<rgba 0 0 0 0>}
       cornerRadii=<0 0 0 0>
      }
      tintColor {colorUseRule=UseColor_Specified, specifiedColor=<rgba 1 1 1 1>}
      uVRegion {bIsValid=false, max=<0 0>, min=<0 0>}
     }
     hoveredThumbImage {
      bIsDynamicallyLoaded=false
      drawAs=Box
      imageType=FullColor
      mirroring=NoMirror
      resourceName=../../../Engine/Content/Slate/Common/Scrollbar_Thumb.png
      resourceObject=&M_UI_Circle_White
      tiling=NoTile
      imageSize=<10 10>
      margin=<ltrb 0.5 0.5 0.5 0.5>
      outlineSettings {
       bUseBrushTransparency=false
       roundingType=HalfHeightRadius
       width=0
       color {colorUseRule=UseColor_Specified, specifiedColor=<rgba 0 0 0 0>}
       cornerRadii=<0 0 0 0>
      }
      tintColor {colorUseRule=UseColor_Specified, specifiedColor=<rgba 1 1 1 1>}
      uVRegion {bIsValid=false, max=<0 0>, min=<0 0>}
     }
     normalThumbImage {
      bIsDynamicallyLoaded=false
      drawAs=Box
      imageType=FullColor
      mirroring=NoMirror
      resourceName=../../../Engine/Content/Slate/Common/Scrollbar_Thumb.png
      resourceObject=&M_UI_Circle_White
      tiling=NoTile
      imageSize=<10 10>
      margin=<ltrb 0.5 0.5 0.5 0.5>
      outlineSettings {
       bUseBrushTransparency=false
       roundingType=HalfHeightRadius
       width=0
       color {colorUseRule=UseColor_Specified, specifiedColor=<rgba 0 0 0 0>}
       cornerRadii=<0 0 0 0>
      }
      tintColor {colorUseRule=UseColor_Specified, specifiedColor=<rgba 1 1 1 1>}
      uVRegion {bIsValid=false, max=<0 0>, min=<0 0>}
     }
     verticalBackgroundImage {
      bIsDynamicallyLoaded=false
      drawAs=Box
      imageType=NoImage
      mirroring=NoMirror
      resourceName=None
      resourceObject=&M_UI_Circle_White
      tiling=NoTile
      imageSize=<10 10>
      margin=<ltrb 0.5 0.5 0.5 0.5>
      outlineSettings {
       bUseBrushTransparency=false
       roundingType=HalfHeightRadius
       width=0
       color {colorUseRule=UseColor_Specified, specifiedColor=<rgba 0 0 0 0>}
       cornerRadii=<0 0 0 0>
      }
      tintColor {colorUseRule=UseColor_Specified, specifiedColor=<rgba 1 1 1 0.2>}
      uVRegion {bIsValid=false, max=<0 0>, min=<0 0>}
     }
     verticalBottomSlotImage {
      bIsDynamicallyLoaded=false
      drawAs=NoDrawType
      imageType=FullColor
      mirroring=NoMirror
      resourceName=../../../Engine/Content/Slate/Common/Scrollbar_Background_Vertical.png
      resourceObject=null
      tiling=NoTile
      imageSize=<8 8>
      margin=<ltrb 0 0 0 0>
      outlineSettings {
       bUseBrushTransparency=false
       roundingType=HalfHeightRadius
       width=0
       color {colorUseRule=UseColor_Specified, specifiedColor=<rgba 0 0 0 0>}
       cornerRadii=<0 0 0 0>
      }
      tintColor {colorUseRule=UseColor_Specified, specifiedColor=<rgba 1 1 1 1>}
      uVRegion {bIsValid=false, max=<0 0>, min=<0 0>}
     }
     verticalTopSlotImage {
      bIsDynamicallyLoaded=false
      drawAs=NoDrawType
      imageType=FullColor
      mirroring=NoMirror
      resourceName=../../../Engine/Content/Slate/Common/Scrollbar_Background_Vertical.png
      resourceObject=null
      tiling=NoTile
      imageSize=<8 8>
      margin=<ltrb 0 0 0 0>
      outlineSettings {
       bUseBrushTransparency=false
       roundingType=HalfHeightRadius
       width=0
       color {colorUseRule=UseColor_Specified, specifiedColor=<rgba 0 0 0 0>}
       cornerRadii=<0 0 0 0>
      }
      tintColor {colorUseRule=UseColor_Specified, specifiedColor=<rgba 1 1 1 1>}
      uVRegion {bIsValid=false, max=<0 0>, min=<0 0>}
     }
    }
    widgetStyle {backgroundBrush {
      bIsDynamicallyLoaded=false
      drawAs=Box
      imageType=NoImage
      mirroring=NoMirror
      resourceName=None
      resourceObject=&M_UI_Circle_White
      tiling=NoTile
      imageSize=<10 10>
      margin=<ltrb 0.5 0.5 0.5 0.5>
      outlineSettings {
       bUseBrushTransparency=false
       roundingType=HalfHeightRadius
       width=0
       color {colorUseRule=UseColor_Specified, specifiedColor=<rgba 0 0 0 0>}
       cornerRadii=<0 0 0 0>
      }
      tintColor {colorUseRule=UseColor_Specified, specifiedColor=<rgba 1 1 1 0.2>}
      uVRegion {bIsValid=false, max=<0 0>, min=<0 0>}
     }}
   }
   slot {class=OverlaySlot, properties {horizontalAlignment=HAlign_Fill, verticalAlignment=VAlign_Fill, padding=<ltrb 24 16 16 16>}}
  }
  Text_NoRewardsAdded {
   widgetClassPath=&CommonTextBlock
   bIsVariable=true
   parent=Overlay_0
   widget=Text_NoRewardsAdded
   properties {
    autoWrapText=true
    justification=Center
    mobileFontSizeMultiplier=1.125
    scrollStyle=&TextScrollStyle-Base_C
    text=Add at least one Set from above to claim
    font {
     bForceMonospaced=false
     bMaterialIsStencil=false
     fontMaterial=null
     fontObject=&HeadingNow
     letterSpacing=0
     monospacedWidth=1
     size=18
     skewAmount=0
     typefaceFontName=75 Medium
     outlineSettings {
      bApplyOutlineToDropShadows=false
      bMiteredCorners=false
      bSeparateFillAlpha=false
      outlineMaterial=null
      outlineSize=0
      outlineColor=<rgba 0 0 0 1>
     }
    }
    shadowOffset=<0 0>
   }
   slot {class=OverlaySlot, properties {horizontalAlignment=HAlign_Fill, verticalAlignment=VAlign_Center, padding=<ltrb 0 0 8 0>}}
  }
  FortVisualAttachment_1 {
   widgetClassPath=&FortVisualAttachment
   parent=Overlay_4
   widget=FortVisualAttachment_1
   properties {contentAnchor=<0 1>}
   slot {class=OverlaySlot, properties {horizontalAlignment=HAlign_Right, verticalAlignment=VAlign_Bottom}}
  }
  WBP_MPLocker_BackToTopButton {
   widgetClassPath=/Locker/MPL/WBP_MPLocker_BackToTopButton
   bIsVariable=true
   parent=FortVisualAttachment_1
   widget=WBP_MPLocker_BackToTopButton
   properties {targetList=&self:WidgetTree.ItemTileView}
   slot {class=SizeBoxSlot, properties {horizontalAlignment=HAlign_Center, padding=<ltrb 4 0 0 150>}}
  }
  Overlay_2 {widgetClassPath=&Overlay, parent=VB_All, widget=Overlay_2, slot {class=VerticalBoxSlot}}
  Backplate_Black {
   widgetClassPath=&WBP_UIKit_Backplate
   bIsVariable=true
   parent=Overlay_2
   widget=Backplate_Black
   properties {backplateMaterial=/SeasonPass/Materials/Backplates/MI_UIKit_ClaimPagesBackplate, brightness=Obscure}
   slot {class=OverlaySlot, properties {horizontalAlignment=HAlign_Fill, verticalAlignment=VAlign_Fill}}
  }
  VerticalBox_1 {
   widgetClassPath=&VerticalBox
   parent=Overlay_2
   widget=VerticalBox_1
   slot {class=OverlaySlot, properties {horizontalAlignment=HAlign_Fill, padding=<ltrb 16 0 16 16>}}
  }
  HorizontalBox_0 {
   widgetClassPath=&HorizontalBox
   parent=VerticalBox_1
   widget=HorizontalBox_0
   slot {class=VerticalBoxSlot, properties {horizontalAlignment=HAlign_Center, verticalAlignment=VAlign_Center, padding=<ltrb 0 16 0 0>}}
  }
  WBP_SeasonPass_RemainingSetsToUnlock {
   widgetClassPath=&/WBP_SeasonPass_RemainingSetsToUnlock
   bIsVariable=true
   parent=HorizontalBox_0
   widget=WBP_SeasonPass_RemainingSetsToUnlock
   slot {class=HorizontalBoxSlot, properties {padding=<ltrb 0 0 16 0>}}
  }
  Status_ToClaim {
   widgetClassPath=/FortUILibrary/UIKit/Highlight/WBP_UIKit_Highlight
   bIsVariable=true
   parent=HorizontalBox_0
   widget=Status_ToClaim
   properties {highlightText=2 to Claim, highlightType=Attention}
   slot {class=HorizontalBoxSlot}
  }
  HorizontalBox_1 {
   widgetClassPath=&HorizontalBox
   parent=VerticalBox_1
   widget=HorizontalBox_1
   slot {class=VerticalBoxSlot, properties {horizontalAlignment=HAlign_Center, padding=<ltrb 0 24 0 0>}}
  }
  Text_TotalRewardsLabel {
   widgetClassPath=&CommonTextBlock
   bIsVariable=true
   parent=HorizontalBox_1
   widget=Text_TotalRewardsLabel
   properties {
    justification=Center
    mobileFontSizeMultiplier=1.15
    scrollStyle=&TextScrollStyle-Base_C
    text=Rewards being claimed:
    textTransformPolicy=ToUpper
    colorAndOpacity {colorUseRule=UseColor_Specified, specifiedColor=<rgba 1 1 1 0.6>}
    font {
     bForceMonospaced=false
     bMaterialIsStencil=false
     fontMaterial=null
     fontObject=&HeadingNow
     letterSpacing=0
     monospacedWidth=1
     size=25.5
     skewAmount=0
     typefaceFontName=86 Bold Italic
     outlineSettings {
      bApplyOutlineToDropShadows=false
      bMiteredCorners=false
      bSeparateFillAlpha=false
      outlineMaterial=null
      outlineSize=0
      outlineColor=<rgba 0 0 0 1>
     }
    }
    shadowOffset=<0 0>
   }
   slot {class=HorizontalBoxSlot, properties {verticalAlignment=VAlign_Center, padding=<ltrb 0 0 8 0>}}
  }
  Text_TotalRewardsValue {
   widgetClassPath=&CommonTextBlock
   bIsVariable=true
   parent=HorizontalBox_1
   widget=Text_TotalRewardsValue
   properties {
    justification=Center
    mobileFontSizeMultiplier=1.15
    scrollStyle=&TextScrollStyle-Base_C
    text=0
    textTransformPolicy=ToUpper
    font {
     bForceMonospaced=false
     bMaterialIsStencil=false
     fontMaterial=null
     fontObject=&HeadingNow
     letterSpacing=0
     monospacedWidth=1
     size=25.5
     skewAmount=0
     typefaceFontName=86 Bold Italic
     outlineSettings {
      bApplyOutlineToDropShadows=false
      bMiteredCorners=false
      bSeparateFillAlpha=false
      outlineMaterial=null
      outlineSize=0
      outlineColor=<rgba 0 0 0 1>
     }
    }
    shadowOffset=<0 0>
   }
   slot {class=HorizontalBoxSlot, properties {verticalAlignment=VAlign_Center}}
  }
  HB_Buttons {
   widgetClassPath=&CommonVisibilitySwitcher
   bIsVariable=true
   parent=VerticalBox_1
   widget=HB_Buttons
   properties {bOverride_Cursor=true}
   slot {class=VerticalBoxSlot, properties {padding=<ltrb 0 12 0 0>}}
  }
  Button_ClaimSets {
   widgetClassPath=/FortUILibrary/UIKit/Buttons/WBP_UIKit_Button_Loud
   bIsVariable=true
   parent=HB_Buttons
   widget=Button_ClaimSets
   properties {
    bIsEnabled=false
    bIsFocusable=false
    holdData=/FortUILibrary/Input/PurchaseInteraction_HoldData.PurchaseInteraction_HoldData_C
    navigation=&self:WidgetTree.Button_ClaimSets.WidgetNavigation_1
    text=Claim Sets
    visibility=Visible
    triggeringInputAction {dataTable=&InputActionDataTable, rowName=Confirm}
   }
   slot {class=CommonVisibilitySwitcherSlot}
  }
  Button_BuyLevels {
   widgetClassPath=&WBP_UIKit_Button_Loud
   bIsVariable=true
   parent=HB_Buttons
   widget=Button_BuyLevels
   properties {
    bHideInputActionWithKeyboard=true
    bIsFocusable=false
    navigation=&self:WidgetTree.Button_BuyLevels.WidgetNavigation_1
    text=Buy Levels
    visibility=Visible
    triggeringInputAction {dataTable=&InputActionDataTable, rowName=Confirm}
   }
   slot {class=CommonVisibilitySwitcherSlot}
  }
  FortVisualAttachment_0 {
   widgetClassPath=&FortVisualAttachment
   parent=VB_All
   widget=FortVisualAttachment_0
   properties {contentAnchor=<0 1>}
   slot {class=VerticalBoxSlot, properties {horizontalAlignment=HAlign_Right, verticalAlignment=VAlign_Bottom}}
  }
  SizeBox_ItemDescription {
   widgetClassPath=&SizeBox
   parent=FortVisualAttachment_0
   widget=SizeBox_ItemDescription
   properties {bOverride_WidthOverride=true, widthOverride=800}
   slot {class=SizeBoxSlot, properties {padding=<ltrb 56 0 0 0>}}
  }
  WBP_UIKit_DescriptionWithSubtitle {
   widgetClassPath=/FortUILibrary/UIKit/Items/WBP_UIKit_DescriptionWithSubtitle
   bIsVariable=true
   parent=SizeBox_ItemDescription
   widget=WBP_UIKit_DescriptionWithSubtitle
   properties {hideSubtitle=true}
   slot {class=SizeBoxSlot, properties {padding=<ltrb 0 0 0 16>}}
  }
  Text_AllRewardsClaimed {
   widgetClassPath=&CommonTextBlock
   bIsVariable=true
   parent=VisibilitySwitcherIsClaimed
   widget=Text_AllRewardsClaimed
   properties {
    autoWrapText=true
    justification=Center
    text=No rewards left to claim
    textTransformPolicy=ToUpper
    font {
     bForceMonospaced=false
     bMaterialIsStencil=false
     fontMaterial=null
     fontObject=&HeadingNow
     letterSpacing=0
     monospacedWidth=1
     size=25.5
     skewAmount=0
     typefaceFontName=86 Bold Italic
     outlineSettings {
      bApplyOutlineToDropShadows=false
      bMiteredCorners=false
      bSeparateFillAlpha=false
      outlineMaterial=null
      outlineSize=0
      outlineColor=<rgba 0 0 0 1>
     }
    }
    shadowOffset=<0 0>
   }
   slot {class=CommonVisibilitySwitcherSlot, properties {verticalAlignment=VAlign_Center}}
  }
 ]
}
