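# Season Pass reward track "included" popup, a custom banner widget derived from WBP_CustomBanner_Base that shows a title and body.
# A grid pairs a material-driven portrait image with a dark UIKit backplate behind the stacked text blocks.
# SetInformation fills both texts.
# PreConstruct calls SetCompactUIStyle, which widens text wrapping under compact UI, then sets placeholder text at design time only.
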WidgetBlueprint /SeasonPass/Widgets/Reboot/WBP_RewardTrack_IncludedPopup
graphs [
 EventGraph {nodes [
   n1 = Event PreConstruct()
   > Parent: PreConstruct(IsDesignTime=@n1.IsDesignTime)
   > SetCompactUIStyle() {is=title K2Node_CallFunction}
   > Branch(Condition=@n1.IsDesignTime)
   > "SetText (Text)"(
    self=[Get Text_Title() -> (Text_Title:Common Text Object Reference), Get Text_Body() -> (Text_Body:Common Text Object Reference)]
    InText="INVTEXT(\"[PH] Text\")"
   )
   Event Construct()
   > Parent: Construct()
   n2 = Event Tick()
   > Parent: Tick!(execute, MyGeometry:Geometry Structure=@n2.MyGeometry, InDeltaTime:float=@n2.InDeltaTime) -> (then) {is=title K2Node_CallParentFunction}
  ]}
 SetCompactUIStyle {nodes [
   SetCompactUIStyle() {is=title K2Node_FunctionEntry}
   > Branch(Condition=IsUsingCompactUI())
   > SetWrapTextWidth!(execute, self:Common Text Object Reference=[Get Text_Title(), Get Text_Body()], InWrapTextAt:int=425) -> (then)
  ]}
 SetInformation {nodes [
   n1 = SetInformation!() -> (then, Title:Text, Body:Text) {is=title K2Node_FunctionEntry}
   > "SetText (Text)"(self=Get Text_Title(), InText=@n1)
   > "SetText (Text)"(self=Get Text_Body(), InText=@n1.Body)
  ]}
]
variables []
widgets {
 info {
  inheritedWidgetCount=0
  parentClass=/FortniteCoreUI/MultiProduct/WBP_CustomBanner_Base.WBP_CustomBanner_Base_C
  rootWidgetClass=&GridPanel
  widgetCount=8
 }
 widgets [
  GridPanel_1 {widgetClassPath=&GridPanel, widget=GridPanel_1, properties {bOverride_Cursor=true, columnFill [0,1], rowFill [1]}}
  Spacer_Image {
   widgetClassPath=&Spacer
   parent=GridPanel_1
   widget=Spacer_Image
   properties {size=<100 100>}
   slot {class=GridSlot, properties {columnSpan=1, layer=1, rowSpan=1}}
  }
  Image {
   widgetClassPath=&Image
   bIsVariable=true
   parent=GridPanel_1
   widget=Image
   properties {
    visibility=HitTestInvisible
    brush {
     bIsDynamicallyLoaded=false
     drawAs=Image
     imageType=FullColor
     mirroring=NoMirror
     resourceName=None
     resourceObject=/SeasonPass/Widgets/Materials/M_UI_Upsell_TopNotification_Portrait
     tiling=NoTile
     imageSize=<96 96>
     margin=<ltrb 0 0 0 0>
     outlineSettings {
      bUseBrushTransparency=false
      roundingType=HalfHeightRadius
      width=0
      color {colorUseRule=UseColor_Specified, specifiedColor=<rgba 0 0 0 0>}
      cornerRadii=<0 0 0 0>
     }
     tintColor {colorUseRule=UseColor_Specified, specifiedColor=<rgba 1 1 1 1>}
     uVRegion {bIsValid=false, max=<0 0>, min=<0 0>}
    }
   }
   slot {class=GridSlot, properties {columnSpan=1, horizontalAlignment=HAlign_Center, layer=1, rowSpan=1, verticalAlignment=VAlign_Center}}
  }
  Overlay_0 {
   widgetClassPath=&Overlay
   parent=GridPanel_1
   widget=Overlay_0
   properties {flowDirectionPreference=LeftToRight}
   slot {class=GridSlot, properties {column=1, columnSpan=1, horizontalAlignment=HAlign_Left, rowSpan=1, verticalAlignment=VAlign_Center}}
  }
  WBP_UIKit_Backplate {
   widgetClassPath=&WBP_UIKit_Backplate
   parent=Overlay_0
   widget=WBP_UIKit_Backplate
   properties {brightness=Dark}
   slot {class=OverlaySlot, properties {horizontalAlignment=HAlign_Fill, verticalAlignment=VAlign_Fill, padding=<ltrb -53 0 0 0>}}
  }
  VerticalBox_45 {
   widgetClassPath=&VerticalBox
   bIsVariable=true
   parent=Overlay_0
   widget=VerticalBox_45
   slot {class=OverlaySlot, properties {horizontalAlignment=HAlign_Fill, padding=<ltrb 4 12 16 12>}}
  }
  Text_Title {
   widgetClassPath=&CommonTextBlock
   bIsVariable=true
   parent=VerticalBox_45
   widget=Text_Title
   properties {
    mobileFontSizeMultiplier=1.4
    textTransformPolicy=ToUpper
    wrapTextAt=400
    font {
     bForceMonospaced=false
     bMaterialIsStencil=false
     fontMaterial=null
     fontObject=&HeadingNow
     letterSpacing=-30
     monospacedWidth=1
     size=12.75
     skewAmount=0
     typefaceFontName=75 Medium Italic
     outlineSettings {
      bApplyOutlineToDropShadows=false
      bMiteredCorners=false
      bSeparateFillAlpha=false
      outlineMaterial=null
      outlineSize=0
      outlineColor=<rgba 0 0 0 1>
     }
    }
    shadowOffset=<0 0>
   }
   slot {class=VerticalBoxSlot}
  }
  Text_Body {
   widgetClassPath=&CommonTextBlock
   bIsVariable=true
   parent=VerticalBox_45
   widget=Text_Body
   properties {
    mobileFontSizeMultiplier=1.4
    wrapTextAt=400
    colorAndOpacity {colorUseRule=UseColor_Specified, specifiedColor=<rgba 1 1 1 0.7>}
    font {
     bForceMonospaced=false
     bMaterialIsStencil=false
     fontMaterial=null
     fontObject=&HeadingNow
     letterSpacing=0
     monospacedWidth=1
     size=11.25
     skewAmount=0
     typefaceFontName=74 Regular
     outlineSettings {
      bApplyOutlineToDropShadows=false
      bMiteredCorners=false
      bSeparateFillAlpha=false
      outlineMaterial=null
      outlineSize=0
      outlineColor=<rgba 0 0 0 1>
     }
    }
    shadowOffset=<0 0>
   }
   slot {class=VerticalBoxSlot, properties {padding=<ltrb 0 4 0 0>}}
  }
 ]
}
