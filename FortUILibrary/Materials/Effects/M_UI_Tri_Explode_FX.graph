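# Additive UI material that draws a burst of drifting, slowly spinning triangles, tinted between Grad and Emphasis colors.
# Centered UVs, scaled by ExplosionScale and rotated by DirectionRotationFraction, feed four panning, rotating samples of different triangle-texture channels.
# A parabolic falloff across the rotated axis, tuned by TriScale, thresholds each sample through SmoothStep.
# The summed layers, partly weighted by a directional radial mask, drive the color lerp; the Aspect divide uses one component twice.
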Material /FortUILibrary/Materials/Effects/M_UI_Tri_Explode_FX
graph {
 properties {bUsedWithStaticMesh=false, blendMode=BLEND_Additive, materialDomain=MD_UI, pixelDepthOffsetMode=PDOM_Legacy, usedShadingModels=Default Lit}
 nodes [
  n1 = MaterialFunctionCall!(Rotation_Fraction=ScalarParameter() {parameterName=DirectionRotationFraction}, TexCoord=NamedRerouteUsage()) {materialFunction=/FortUILibrary/Foundation/Art_ToolKit/AngledBoxes/MF_UI_Rotator_Fractions}
  n2 = Panner(Coordinate=@n1.Result, Speed=AppendVector(A=ScalarParameter() {defaultValue=0.001, parameterName=Speed1}, B=Constant() {r=0.3})) {speedX=0.25}
  n3 = ComponentMask(None=@n1.Result) {r=true}
  n4 = Multiply(A=@n3, B=OneMinus(None=@n3))
  n5 = Multiply(A=Constant() {r=2}, B=ScalarParameter() {defaultValue=1, parameterName=TriScale})
  n6 = Saturate(None=OneMinus(None=Multiply(A=@n4, B=@n5) {constB=4}))
  n7 = MaterialFunctionCall(
   Alpha=TextureSampleParameter2D(UVs=Rotator(Coordinate=@n2) {speed=0.0001}) {parameterName=TextureMask, samplerType=SAMPLERTYPE_LinearColor, texture=/Game/UI/Foundation/Items/Textures/T_Triangle_More}.G
   Min=@n6
  ) {materialFunction=&SmoothStep}
  n8 = Multiply(A=Subtract(A=Multiply(A=@n7.Result)) {constB=0.65}) {constB=8}
  n9 = Clamp(None=@n8)
  n10 = Panner(Coordinate=@n1.Result, Speed=AppendVector(A=ScalarParameter() {defaultValue=0.001, parameterName=Speed1}, B=Constant() {r=0.3})) {speedX=0.25}
  n11 = MaterialFunctionCall(
   Alpha=TextureSampleParameter2D(UVs=Rotator(Coordinate=@n10) {speed=0.0002}) {parameterName=TextureMask, samplerType=SAMPLERTYPE_LinearColor, texture=&T_Triangle_More}.B
   Min=@n6
  ) {materialFunction=&SmoothStep}
  n12 = Multiply(A=Subtract(A=Multiply(A=@n11.Result)) {constB=0.65}) {constB=8}
  n13 = Clamp(None=@n12)
  n14 = Panner(Coordinate=@n1.Result, Speed=AppendVector(A=ScalarParameter() {defaultValue=0.001, parameterName=Speed2}, B=Constant() {r=0.3})) {speedX=0.15}
  n15 = MaterialFunctionCall(
   Alpha=TextureSampleParameter2D(UVs=Rotator(Coordinate=@n14) {speed=-0.0002}) {parameterName=TextureMask, samplerType=SAMPLERTYPE_LinearColor, texture=&T_Triangle_More}.G
   Min=@n6
  ) {materialFunction=&SmoothStep}
  n16 = Multiply(A=Subtract(A=Multiply(A=@n15.Result)) {constB=0.65}) {constB=8}
  n17 = Panner(Coordinate=@n1.Result, Speed=AppendVector(A=ScalarParameter() {defaultValue=0.001, parameterName=Speed3}, B=Constant() {r=0.3})) {speedX=0.5}
  n18 = MaterialFunctionCall(
   Alpha=TextureSampleParameter2D(UVs=Rotator(Coordinate=@n17) {speed=-0.0003}) {parameterName=TextureMask, samplerType=SAMPLERTYPE_LinearColor, texture=&T_Triangle_More}.R
   Min=@n6
  ) {materialFunction=&SmoothStep}
  n19 = Multiply(A=Subtract(A=Multiply(A=@n18.Result)) {constB=0.3}) {constB=8}
  n20 = Add(A=Add(A=Add(A=@n9, B=@n13), B=Clamp(None=@n16)), B=Clamp(None=@n19))
  n21 = NamedRerouteUsage()
  n22 = Saturate(None=ScalarParameter() {defaultValue=1, parameterName=MaskRadius})
  n23 = Saturate(None=ScalarParameter() {defaultValue=1.7317, parameterName=MaskDesity})
  n24 = Multiply(A=Multiply(
    A=@n3
    B=MaterialFunctionCall(UVs=@n21, Radius=@n22, Density=@n23) {materialFunction=&RadialGradientExponential}.RadialGradientExponential
   )) {constB=3}
  n25 = Add(A=Add(A=Clamp(None=@n8), B=Clamp(None=@n12)), B=Clamp(None=@n16) {maxDefault=0.5})
  n26 = Clamp(None=@n19) {maxDefault=0.05}
  LinearInterpolate(
   A=VectorParameter() {parameterName=Grad, defaultValue=<rgba 0.5642 0.1018 0.8542 1>}.RGB
   B=VectorParameter() {parameterName=Emphasis, defaultValue=<rgba 1 0.5019 0.7858 1>}.RGB
   Alpha=Add(A=LinearInterpolate(B=@n20, Alpha=@n24), B=Add(A=@n25, B=@n26))
  )
  n27 = MaterialFunctionCall(Float2=MaterialFunctionCall() {materialFunction=&GetUserInterfaceUV}.Pixel Size) {materialFunction=&BreakOutFloat2Components}
  NamedRerouteDeclaration(None=Divide(A=@n27.R, B=@n27.R)) {name=Aspect, nodeColor=<rgba 0.6875 1 0 1>}
  n28 = Multiply(
   A=Subtract(A=TextureCoordinate()) {constB=0.5}
   B=MaterialFunctionCall(X=NamedRerouteUsage(), Y=ScalarParameter() {defaultValue=1, parameterName=Param_3}) {materialFunction=&MakeFloat2}.Result
  )
  NamedRerouteDeclaration(None=Divide(A=Add(A=@n28) {constB=0.5}, B=ScalarParameter() {defaultValue=1, parameterName=ExplosionScale})) {name=CorrectedUV, nodeColor=<rgba 0.9922 0 1 1>}
 ]
}
