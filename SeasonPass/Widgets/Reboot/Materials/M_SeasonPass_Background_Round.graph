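# Translucent UI material for a rounded background shape in the Season Pass widgets.
# It reads widget size from texture coordinate 3 and insets width by a height-scaled HorizontalPadding, clamped so the shape is never thinner than a circle.
# That ratio sizes a fully rounded mask function, whose SDF feeds an anti-aliasing function.
# A color parameter and a separate circle function are also present, with no connections shown.
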Material /SeasonPass/Widgets/Reboot/Materials/M_SeasonPass_Background_Round
graph {
 properties {
  bAutomaticallySetUsageInEditor=false
  bUsedWithStaticMesh=false
  blendMode=BLEND_Translucent
  materialDomain=MD_UI
  usedShadingModels=Default Lit
 }
 nodes [
  n1 = Constant() {r=1}
  MaterialFunctionCall(Size=@n1) {materialFunction=&MF_UI_Circle}
  n2 = TextureCoordinate() {coordinateIndex=3}
  n3 = ComponentMask(None=@n2) {g=true}
  n4 = Multiply(A=@n3, B=ScalarParameter() {defaultValue=-0.3, parameterName=HorizontalPadding}) {constB=-0.3}
  n5 = ComponentMask(None=@n2) {r=true}
  n6 = Divide(A=Max(A=@n3, B=Add(A=@n4, B=@n5)) {desc=can't be thinner than a circle}, B=@n5)
  MaterialFunctionCall(
   DistanceField=MaterialFunctionCall(Size=AppendVector(A=@n6, B=@n1), Corner Radius=@n1, Corners Modulated by Size=Constant()) {materialFunction=&MF_UI_CoreUI_RoundedMask}.SDF
   Threshold=Constant()
  ) {materialFunction=&MF_UI_AntiAliasing}
  VectorParameter() {group=0 - Coloring, parameterName=Color 1, defaultValue=<rgba 0.9301 1 0.0103 1>}
 ]
}
