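# UI material for the Season Pass Reboot reward track's game mode overlay, drawn with alpha-composite blending.
# It builds a corner fade, a positionable logo clipped to a box, and a rounded-corner shape mask, shared through named reroutes and blended by lerps.
# An optional quest triangle comes from a skewed, rotated three-sided polygon SDF with anti-aliasing and a two-color gradient, hidden by default through zero opacity.
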Material /SeasonPass/Widgets/Reboot/Materials/M_UI_BPReboot_RewardTrack_GameModeOverlay
graph {
 properties {
  bAutomaticallySetUsageInEditor=false
  bUsedWithStaticMesh=false
  blendMode=BLEND_AlphaComposite
  materialDomain=MD_UI
  usedShadingModels=Default Lit
 }
 nodes [
  n1 = Multiply(A=NamedRerouteUsage(), B=NamedRerouteUsage())
  LinearInterpolate(A=Multiply(A=@n1) {constB=0.55}, Alpha=Add(A=NamedRerouteUsage(), B=NamedRerouteUsage()))
  n2 = Constant3Vector() {constant=<rgba 0.001 0.0005 0.0104 1>}
  LinearInterpolate(
   A=LinearInterpolate(A=LinearInterpolate(B=@n2, Alpha=@n1) {constB=0}, B=NamedRerouteUsage(), Alpha=NamedRerouteUsage())
   B=NamedRerouteUsage()
   Alpha=NamedRerouteUsage()
  )
  n3 = MaterialFunctionCall() {materialFunction=&MF_UI_UVScaling}
  n4 = OneMinus(None=ComponentMask(None=@n3.Scaled To Fill) {r=true})
  n5 = ComponentMask(None=@n3.Scaled to Fit) {g=true}
  NamedRerouteDeclaration(None=Power(Base=Multiply(A=@n4, B=@n5)) {constExponent=0.5}) {name=Fade, nodeColor=<rgba 0 0 0 1>}
  n6 = MaterialFunctionCall(UV=Add(A=TextureCoordinate(), B=Constant2Vector() {g=-0.5, r=0.5})) {materialFunction=&MF_UI_UVScaling}
  n7 = Constant2Vector() {g=1, r=2}
  n8 = ScalarParameter() {defaultValue=-0.285, parameterName=Pos X}
  n9 = ScalarParameter() {defaultValue=0.318, parameterName=Pos Y}
  n10 = MaterialFunctionCall(
   UVs=Add(
    A=MaterialFunctionCall(UVs=@n6.Scaled to Fit, Texture Scale=@n7) {materialFunction=&ScaleUVsByCenter}.UVs
    B=AppendVector(A=@n8, B=@n9)
   )
   Texture Scale=ScalarParameter() {defaultValue=0.45, parameterName=Scale}
  ) {materialFunction=&ScaleUVsByCenter}
  n11 = TextureSampleParameter2D(UVs=@n10.UVs) {parameterName=LogoTexture}
  n12 = MaterialFunctionCall(UVs=@n10.UVs, Size=Constant() {r=1}) {materialFunction=&MF_UI_Box}
  NamedRerouteDeclaration(None=Multiply(A=@n11.A, B=@n12.Fill)) {name=Logo Alpha, nodeColor=<rgba 0 0 0 1>}
  n13 = ScalarParameter() {defaultValue=20, group=Layout, parameterName=TopLeftCornerRadius, sortPriority=4}
  n14 = ScalarParameter() {defaultValue=20, group=Layout, parameterName=BottomLeftCornerRadius, sortPriority=7}
  n15 = ScalarParameter() {defaultValue=20, group=Layout, parameterName=TopRightCornerRadius, sortPriority=5}
  n16 = ScalarParameter() {defaultValue=20, group=Layout, parameterName=BottomRightCornerRadius, sortPriority=6}
  n17 = ScalarParameter() {
   defaultValue=1
   desc="If enabled reads corner radius values in pixels.\r\nbool"
   group=Layout
   parameterName=UseValueInPixels
   sliderMax=1
   sortPriority=1
  }
  NamedRerouteDeclaration(None=MaterialFunctionCall(
    Corner Radius=MaterialFunctionCall(R=@n13, G=@n14, B=@n15, A=@n16) {materialFunction=&AppendMany}.RGBA
    Radius in Pixels=@n17
   ) {materialFunction=&MF_UI_CoreUI_RoundedMask}.Shape Mask) {name=Shape Mask, nodeColor=<rgba 0 0 0 1>}
  n18 = ScalarParameter() {defaultValue=1.8, group=Quest Triangle, parameterName=TriangleGradPosX}
  n19 = ScalarParameter() {group=Quest Triangle, parameterName=TriangleGradPosY}
  n20 = MaterialFunctionCall(
   UVs=Add(A=MaterialFunctionCall() {materialFunction=&MF_UI_UVScaling}.ScaleY, B=AppendVector(A=@n18, B=@n19))
   "Rotation Angle (0-1)"=ScalarParameter() {defaultValue=-0.3, group=Quest Triangle, parameterName=TriangleGradRotAngle}
  ) {materialFunction=&CustomRotator}
  NamedRerouteDeclaration(None=LinearInterpolate(
    A=VectorParameter() {parameterName=TringleGradColor1, defaultValue=<rgba 0.0729 0.0626 0.0554 1>}.RGB
    B=VectorParameter() {parameterName=TringleGradColor2, defaultValue=<rgba 0.0152 0.0123 0.0103 1>}.RGB
    Alpha=MaterialFunctionCall(UV Channel=@n20.Rotated Values) {materialFunction=&LinearGradient}.UGradient
   )) {name=TringleColor, nodeColor=<rgba 0 0 0 1>}
  n21 = Add(
   A=MaterialFunctionCall() {materialFunction=&MF_UI_UVScaling}.ScaleY
   B=AppendVector(
    A=ScalarParameter() {defaultValue=0.7494, group=Quest Triangle, parameterName=QuestTrianglePositionX}
    B=ScalarParameter() {defaultValue=-0.15, group=Quest Triangle, parameterName=QuestTrianglePositionY}
   )
  )
  n22 = ScalarParameter() {defaultValue=0.425, group=Quest Triangle, parameterName=QuestTriangleRotation}
  n23 = ScalarParameter() {defaultValue=1.45, group=Quest Triangle, parameterName=QuestTringleScale}
  n24 = AppendVector(
   A=ScalarParameter() {defaultValue=2.3, group=Quest Triangle, parameterName=QuestTriangleScaleX}
   B=ScalarParameter() {defaultValue=1.7, group=Quest Triangle, parameterName=QuestTriangleScaleY}
  )
  n25 = MaterialFunctionCall!(
   UV=MaterialFunctionCall(
    UVs=MaterialFunctionCall(UVs=@n21, "Rotation Angle (0-1)"=@n22) {materialFunction=&CustomRotator}.Rotated Values
    Texture Scale=Multiply(A=@n23, B=@n24)
   ) {materialFunction=&ScaleUVsByCenter}.UVs
   Skew Angle=ScalarParameter() {defaultValue=14, group=Quest Triangle, parameterName=QuestTriangleSkew}
  ) {materialFunction=/FortUILibrary/MaterialFunctions/Basics/MF_UI_Skew}
  n26 = Constant() {r=3}
  n27 = Saturate(None=MaterialFunctionCall(DistanceField=MaterialFunctionCall!(UVs=@n25.Skewed UVs, Radius, Sides=@n26, InnerAngle) {materialFunction=/FortUILibrary/MaterialFunctions/Shapes/SDFs/MF_SDF_RegularPolygon}.SDF) {materialFunction=&MF_UI_AntiAliasing}.Inner)
  n28 = ScalarParameter() {group=Quest Triangle, parameterName=QuestTringleOpcaity}
  n29 = Multiply(A=NamedRerouteUsage(), B=Saturate(None=Multiply(A=@n27, B=@n28)))
  NamedRerouteDeclaration(None=Multiply(A=NamedRerouteUsage(), B=NamedRerouteDeclaration(None=@n29) {name=TriangleMask, nodeColor=<rgba 0 0 0 1>})) {name=QuestTriangle, nodeColor=<rgba 0 0 0 1>}
  NamedRerouteDeclaration(None=Multiply(A=@n11.RGB, B=@n12.Fill)) {name=Logo Color, nodeColor=<rgba 0 0 0 1>}
 ]
}
