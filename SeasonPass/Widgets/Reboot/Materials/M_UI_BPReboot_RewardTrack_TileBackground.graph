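# Translucent UI material for the tile background of the Season Pass Reboot reward track.
# IsPremium blends DisabledColor to CharacterColor, which feeds an eased, hue-shifted gradient.
# Three procedural anti-aliased triangles, built from rotated, scaled, mirrored UVs, form a bolt overlay controlled by Bolt Opacity, with an optional additive ColorWash tint.
# A rounded-rectangle mask supplies the tile shape, with per-corner radii in pixels by default.
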Material /SeasonPass/Widgets/Reboot/Materials/M_UI_BPReboot_RewardTrack_TileBackground
graph {
 properties {
  bAutomaticallySetUsageInEditor=false
  bUsedWithStaticMesh=false
  blendMode=BLEND_Translucent
  materialDomain=MD_UI
  usedShadingModels=Default Lit
 }
 nodes [
  Constant() {r=1}
  n1 = NamedRerouteUsage()
  n2 = LinearInterpolate(A=NamedRerouteUsage(), Alpha=NamedRerouteUsage())
  n3 = NamedRerouteUsage()
  n4 = LinearInterpolate(
   A=@n1
   B=Multiply(A=Power(Base=@n1)) {constB=0.6}
   Alpha=Multiply(A=LinearInterpolate(A=@n2, Alpha=@n3), B=ScalarParameter() {defaultValue=0.3, parameterName=Bolt Opacity}) {constB=0.5}
  )
  LinearInterpolate(
   A=@n4
   B=Add(A=@n4, B=VectorParameter() {parameterName=ColorWashColor, defaultValue=<rgba 0.7982 0.5754 1 1>}.RGB)
   Alpha=ScalarParameter() {parameterName=ColorWash}
  )
  n5 = ScalarParameter() {defaultValue=20, group=Layout, parameterName=TopLeftCornerRadius, sortPriority=4}
  n6 = ScalarParameter() {defaultValue=20, group=Layout, parameterName=BottomLeftCornerRadius, sortPriority=7}
  n7 = ScalarParameter() {defaultValue=20, group=Layout, parameterName=TopRightCornerRadius, sortPriority=5}
  n8 = ScalarParameter() {defaultValue=20, group=Layout, parameterName=BottomRightCornerRadius, sortPriority=6}
  n9 = ScalarParameter() {
   defaultValue=1
   desc="If enabled reads corner radius values in pixels.\r\nbool"
   group=Layout
   parameterName=UseValueInPixels
   sliderMax=1
   sortPriority=1
  }
  NamedRerouteDeclaration(None=MaterialFunctionCall(
    Corner Radius=MaterialFunctionCall(R=@n5, G=@n6, B=@n7, A=@n8) {materialFunction=&AppendMany}.RGBA
    Radius in Pixels=@n9
   ) {materialFunction=&MF_UI_CoreUI_RoundedMask}.Shape Mask) {name=Shape Mask, nodeColor=<rgba 0 0 0 1>}
  n10 = Constant() {r=0.02}
  n11 = LinearInterpolate(
   A=VectorParameter() {parameterName=DisabledColor, defaultValue=<rgba 0.1042 0.1042 0.1042 1>}.RGB
   B=VectorParameter() {parameterName=CharacterColor, defaultValue=<rgba 0.1714 0.0561 0.521 1>}.RGB
   Alpha=ScalarParameter() {defaultValue=1, parameterName=IsPremium}
  )
  n12 = Power(Base=@n11) {constExponent=0.5}
  n13 = Distance(A=Add(A=TextureCoordinate(), B=Constant2Vector() {g=-0.5}), B=Constant() {r=0.5})
  NamedRerouteDeclaration(None=LinearInterpolate(
    A=MaterialFunctionCall(Hue Shift Percentage=@n10, Texture=@n12) {materialFunction=&HueShift}.Result
    B=@n11
    Alpha=MaterialFunctionCall("Value (0-1)"=@n13) {materialFunction=&MF_UI_EaseCurves}.EaseInOutSine
   )) {name=Gradient, nodeColor=<rgba 0 0 0 1>}
  n14 = MaterialFunctionCall() {materialFunction=&MF_UI_UVScaling}
  n15 = Constant() {r=0.385}
  n16 = MaterialFunctionCall!(
   UV=MaterialFunctionCall(
    UVs=MaterialFunctionCall(UVs=@n14.Scaled To Fill, "Rotation Angle (0-1)"=@n15) {materialFunction=&CustomRotator}.Rotated Values
    Texture Scale=Constant2Vector() {g=2, r=1}
   ) {materialFunction=&ScaleUVsByCenter}.UVs
   Mirror Y
   Swizzle UV
  ) {materialFunction=/FortUILibrary/MaterialFunctions/Basics/MF_UI_Symmetry_2x1_New}
  n17 = ComponentMask(None=@n16.Symmetry) {r=true}
  n18 = OneMinus(None=ComponentMask(None=@n16.Symmetry) {g=true})
  NamedRerouteDeclaration(None=MaterialFunctionCall(DistanceField=SmoothStep(Min=@n17, Value=@n18), Threshold=Constant() {r=0.01}) {materialFunction=&MF_UI_AntiAliasing}.Inner) {name=Main Triangle, nodeColor=<rgba 0 0 0 1>}
  n19 = TextureCoordinate()
  n20 = Constant2Vector() {r=0.5}
  n21 = Add(
   A=MaterialFunctionCall(UV=Add(A=@n19, B=@n20)) {materialFunction=&MF_UI_UVScaling}.Scaled to Fit
   B=Constant2Vector() {g=0.5, r=-0.2}
  )
  n22 = Constant() {r=0.43}
  n23 = MaterialFunctionCall(UV=MaterialFunctionCall(
    UVs=MaterialFunctionCall(UVs=@n21, "Rotation Angle (0-1)"=@n22) {materialFunction=&CustomRotator}.Rotated Values
    Texture Scale=Constant2Vector() {g=0.75, r=0.3}
   ) {materialFunction=&ScaleUVsByCenter}.UVs) {materialFunction=&MF_UI_Symmetry_2x1_New}
  n24 = ComponentMask(None=@n23.Symmetry) {r=true}
  n25 = OneMinus(None=ComponentMask(None=@n23.Symmetry) {g=true})
  NamedRerouteDeclaration(None=MaterialFunctionCall(DistanceField=SmoothStep(Min=@n24, Value=@n25), Threshold=Constant() {r=0.01}) {materialFunction=&MF_UI_AntiAliasing}.Inner) {name=Left Triangle, nodeColor=<rgba 0 0 0 1>}
  n26 = TextureCoordinate()
  n27 = Constant2Vector() {r=-0.5}
  n28 = Add(
   A=MaterialFunctionCall(UV=Add(A=@n26, B=@n27)) {materialFunction=&MF_UI_UVScaling}.Scaled to Fit
   B=Constant2Vector() {g=-0.5, r=0.3}
  )
  n29 = Constant() {r=-0.155}
  n30 = MaterialFunctionCall(UV=MaterialFunctionCall(
    UVs=MaterialFunctionCall(UVs=@n28, "Rotation Angle (0-1)"=@n29) {materialFunction=&CustomRotator}.Rotated Values
    Texture Scale=Constant2Vector() {g=0.6, r=0.2}
   ) {materialFunction=&ScaleUVsByCenter}.UVs) {materialFunction=&MF_UI_Symmetry_2x1_New}
  n31 = ComponentMask(None=@n30.Symmetry) {r=true}
  n32 = OneMinus(None=ComponentMask(None=@n30.Symmetry) {g=true})
  NamedRerouteDeclaration(None=MaterialFunctionCall(DistanceField=SmoothStep(Min=@n31, Value=@n32), Threshold=Constant() {r=0.01}) {materialFunction=&MF_UI_AntiAliasing}.Inner) {name=Right Triangle, nodeColor=<rgba 0 0 0 1>}
  NamedRerouteUsage()
 ]
}
