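# Translucent UI material for the season pass Reboot widgets, built from a tinted gradient, a rounded-corner mask and a triangle shape.
# An offset UV drives a vertical gradient, which multiplies the dimmed Color parameter and, separately, a reroute usage.
# CoreUI RoundedMask builds the named Shape Mask from per-corner radius parameters with a pixel-unit toggle.
# The same offset UV feeds a WIP triangle SDF, whose inverted SmoothStep edge feeds no other node shown.
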Material /SeasonPass/Widgets/Reboot/Materials/M_UI_BP_Triangle
graph {
 properties {bAutomaticallySetUsageInEditor=false, blendMode=BLEND_Translucent, materialDomain=MD_UI, usedShadingModels=Default Lit}
 nodes [
  n1 = Add(A=TextureCoordinate(), B=Constant2Vector() {g=-0.17, r=0.25})
  n2 = MaterialFunctionCall(UV Channel=@n1) {materialFunction=&LinearGradient}
  Multiply(A=@n2.VGradient, B=NamedRerouteUsage())
  Multiply(A=Divide(A=VectorParameter() {parameterName=Color, defaultValue=<rgba 1 1 1 1>}.RGB) {constB=8}, B=@n2.VGradient)
  n3 = ScalarParameter() {defaultValue=20, group=Layout, parameterName=TopLeftCornerRadius, sortPriority=4}
  n4 = ScalarParameter() {defaultValue=20, group=Layout, parameterName=BottomLeftCornerRadius, sortPriority=7}
  n5 = ScalarParameter() {defaultValue=20, group=Layout, parameterName=TopRightCornerRadius, sortPriority=5}
  n6 = ScalarParameter() {defaultValue=20, group=Layout, parameterName=BottomRightCornerRadius, sortPriority=6}
  n7 = ScalarParameter() {
   defaultValue=1
   desc="If enabled reads corner radius values in pixels.\r\nbool"
   group=Layout
   parameterName=UseValueInPixels
   sliderMax=1
   sortPriority=1
  }
  NamedRerouteDeclaration(None=MaterialFunctionCall(
    Corner Radius=MaterialFunctionCall(R=@n3, G=@n4, B=@n5, A=@n6) {materialFunction=&AppendMany}.RGBA
    Radius in Pixels=@n7
   ) {materialFunction=&MF_UI_CoreUI_RoundedMask}.Shape Mask) {name=Shape Mask, nodeColor=<rgba 0 0 0 1>}
  SmoothStep(Value=MaterialFunctionCall!(
    UVs=@n1
    Corner 01=Constant2Vector() {g=-1, r=-1}
    Corner 02=Constant2Vector() {g=1, r=-1}
    Corner 03=Constant2Vector() {r=1.24}
    Glow Min
    Glow Max
    Outline Thickness
   ) {materialFunction=/FortUILibrary/MaterialFunctions/Shapes/MF_UI_WIP_Triangle}.SDF) {constMax=0, constMin=0.02}
 ]
}
