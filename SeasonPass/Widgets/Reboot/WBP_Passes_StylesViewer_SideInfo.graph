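# Side info panel for the season pass styles viewer, showing a style reward's claimed, lock, coming-soon, pass-required and progress states.
# SetSeasonRewardVM refreshes visibility, progress, quest button and lock text from the reward view model.
# Helper functions resolve rewards from variant items, falling back to the parent cosmetic.
# Buttons open the reward's quest category or async-push the pass purchase screen, injecting the season pass view model through MVVM.
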WidgetBlueprint /SeasonPass/Widgets/Reboot/WBP_Passes_StylesViewer_SideInfo
graphs [
 EventGraph {nodes [
   n1 = "On Clicked (Button_Quests)"!() -> (OutputDelegate, then, Button:Common Button Base Object Reference) {is=title K2Node_ComponentBoundEvent}
   > Fire Event Element Interacted(Element=@n1.Button)
   n2 = GetReward!(self:Season Pass Reward VM Object Reference=Get SeasonPassRewardVM() -> (SeasonPassRewardVM:Season Pass Reward VM Object Reference)) -> (ReturnValue:Athena Season Item Entry Base Object Reference)
   > NavigateToQuestScreen!(
    execute
    Context:Widget Object Reference=Self-Reference()
    StartingCategoryTag:"Gameplay Tag Structure (by ref)"="Get (a ref)"(
     Array:Array of Gameplay Tag Structures=GetQuestCategoryTags!(Context:Widget Object Reference, RewardEntry:Athena Season Item Entry Base Object Reference=@n2) -> (ReturnValue_GameplayTags:Array of Gameplay Tag Structures)
     Dimension 1=0
    ) -> (Output:"Gameplay Tag Structure (by ref)")
   ) -> (then)
   n3 = "On Clicked (LoudButton_BuySeasonPass)"!() -> (OutputDelegate, then, Button:Common Button Base Object Reference) {is=title K2Node_ComponentBoundEvent}
   > Fire Event Element Interacted(Element=@n3.Button)
   > n4 = PushPrimaryContentSoftClassAsync!(
    execute:Exec
    SoftWidgetClass:Common Activatable Widget Soft Class Reference=GetActivatableWidgetForAlias!(WidgetAlias:Gameplay Tag Structure=(TagName="UI.Alias.PassPurchase")) -> (ReturnValue:Common Activatable Widget Soft Class Reference)
   ) -> (then:Exec, OnPrimaryContentPushed:Exec, ActivatableWidget:Common Activatable Widget Object Reference) {is=title K2Node_AsyncAction}
   SetViewModelByClass!(
    execute=@n4.OnPrimaryContentPushed
    self:MVVM View Object Reference=GetExtension!(self:User Widget Object Reference=@n4, ExtensionType:User Widget Extension Class Reference=/Script/ModelViewViewModel.MVVMView) -> (ReturnValue:MVVM View Object Reference)
    NewValue:Notify Field Value Changed Interface=Get SeasonPassVM() -> (SeasonPassVM:Season Pass VM Object Reference)
   ) -> (then, ReturnValue:bool)
   Event PreConstruct()
   > SetCompactUIStyle() {is=title K2Node_CallFunction}
  ]}
 GetClaimedStateWidgetIndex {nodes [
   n1 = GetClaimedStateWidgetIndex!() -> (then, SeasonPassVM:Season Pass VM Object Reference, VariantItemVM:Fort Variant Item VM Object Reference) {is=title K2Node_FunctionEntry}
   > Return Node!(execute, ClaimedWidgetIndex:int=1)
   n2 = Is Valid(InputObject=FindSeasonPassRewardForVariant!(self:Season Pass VM Object Reference=@n1, VariantItemVM:Fort Variant Item VM Object Reference=@n1.VariantItemVM) -> (ReturnValue:Season Pass Reward VM Object Reference))
   Return Node(execute=@n2.Is Not Valid, ClaimedWidgetIndex=0)
   > Return Node(ClaimedWidgetIndex=1)
  ]}
 GetIsMainPassFullyClaimed {nodes [
   n1 = GetIsMainPassFullyClaimed!() -> (then, SeasonPassVM:Season Pass Element Base VM Object Reference) {is=title K2Node_FunctionEntry}
   > Set NumberOfTotalRewardsInPass(NumberOfTotalRewardsInPass:int=0) -> (Output_Get:int=0)
   > Set NumberOfTotalOwnedRewardsInPass(NumberOfTotalOwnedRewardsInPass:int=0) -> (Output_Get:int=0)
   > n2 = For Each Loop(Array:Array of Season Pass Element Base VM Object References=GetChildGroup!(self:Season Pass Element Base VM Object Reference=@n1, GroupName:String=Default) -> (ReturnValue:Array of Season Pass Element Base VM Object References)) -> (Array Element:Season Pass Element Base VM Object Reference)
   Return Node!(
    execute=@n2.Completed
    Output_Get:bool="integer >= integer"(
     A=Get FullyClaimedMainCategorySets() -> (FullyClaimedMainCategorySets:int=0)
     B=Get NumberOfMainCategorySets() -> (NumberOfMainCategorySets:int=0)
    )
   )
   > n3 = Cast To SeasonPassCategoryVM!(execute, Object=@n2) -> (then, CastFailed, AsSeason Pass Category VM:Season Pass Category VM Object Reference) {is=title K2Node_DynamicCast}
   > Increment Int(Value=Get NumberOfMainCategorySets() -> (NumberOfMainCategorySets=0))
   > Set NumberOfTotalOwnedRewardsInPass(NumberOfTotalOwnedRewardsInPass=int + int(
     A=Get NumberOfTotalOwnedRewardsInPass() -> (NumberOfTotalOwnedRewardsInPass:int=0)
     B=GetOwnedRewardsCount!(self:Season Pass Category VM Object Reference=@n3) -> (ReturnValue:int=0)
    )) -> (Output_Get=0)
   > Set NumberOfTotalRewardsInPass(NumberOfTotalRewardsInPass=int + int(
     A=Get NumberOfTotalRewardsInPass() -> (NumberOfTotalRewardsInPass:int=0)
     B=GetTotalRewards!(self:Season Pass Category VM Object Reference=@n3) -> (ReturnValue:int=0)
    )) -> (Output_Get=0)
   > Branch(Condition=Get bIsFullyClaimed!(self:Season Pass Category VM Object Reference=@n3) -> (bIsFullyClaimed:bool))
   > Increment Int(Value=Get FullyClaimedMainCategorySets() -> (FullyClaimedMainCategorySets=0))
  ]}
 GetRequiresAndNotOwnsPass {nodes [
   GetRequiresAndNotOwnsPass() {is=title K2Node_FunctionEntry}
   n1 = Get SeasonPassRewardVM()
   n2 = Get bDoesOwnsPass!(self:Season Pass VM Object Reference=Get SeasonPassVM()) -> (bDoesOwnsPass:bool)
   > Return Node!(
    execute
    RequiresAndOwnsPass:bool=AND Boolean(A=NOT Boolean(A=@n2), B=IsSeasonPassRequired!(self:Season Pass Reward VM Object Reference=@n1) -> (ReturnValue:bool))
   )
  ]}
 GetSeasonPassRewardVM {nodes [
   n1 = GetSeasonPassRewardVM!() -> (then, SeasonPassVM:Season Pass VM Object Reference, VariantItemVM:Fort Variant Item VM Object Reference) {is=title K2Node_FunctionEntry}
   n2 = FindSeasonPassRewardForVariant(self=@n1, VariantItemVM=@n1.VariantItemVM)
   > n3 = Is Valid(InputObject=@n2)
   Return Node!(
    execute=@n3.Is Not Valid
    SeasonPassRewardVM:Season Pass Reward VM Object Reference=FindSeasonPassReward!(
     self:Season Pass VM Object Reference=@n1
     ItemVM:Fort Item VM Object Reference=Get CosmeticItem!(self:Fort Variant Item VM Object Reference=@n1.VariantItemVM) -> (CosmeticItem:Fort Item VM Object Reference)
    ) -> (ReturnValue:Season Pass Reward VM Object Reference)
   )
   > Return Node(SeasonPassRewardVM=@n2)
  ]}
 GetTextForSeasonPassRewards {nodes [
   GetTextForSeasonPassRewards() {is=title K2Node_FunctionEntry}
   n1 = IsIncludedWithPass!(self:Season Pass Reward VM Object Reference=Get SeasonPassRewardVM()) -> (ReturnValue:bool)
   n2 = MakeLiteralText(Value="LOCTEXT(\"Claim all other {0} rewards.\")")
   n3 = Get Title!(self:Season Pass Category VM Object Reference=GetOwnerCategory!(self:Season Pass Reward VM Object Reference=Get SeasonPassRewardVM()) -> (ReturnValue:Season Pass Category VM Object Reference)) -> (Title:Text)
   > Return Node!(
    execute
    Text:Text=Select(
     Option 0:"Text (by ref)"=MakeLiteralText(Value="LOCTEXT(\"Get the Battle Pass to claim the included style.\")")
     Option 1:"Text (by ref)"=Format Text!(Format=@n2, 0:Text=@n3) -> (Result)
     Index=NOT Boolean(A=@n1)
    ) -> (ReturnValue:"Text (by ref)")
   )
  ]}
 IsBonus {nodes [
   IsBonus() {is=title K2Node_FunctionEntry}
   > n1 = Get!(execute:Exec) -> (then:Exec, else:Exec, SeasonPassRewardVM:Season Pass Reward VM Object Reference)
   n2 = GetOwnerCategory(self=@n1)
   > Return Node!(
    execute
    ReturnValue:bool="Equal Exactly (String)"(A=GetGroupName!(self:Season Pass Element Base VM Object Reference=@n2) -> (ReturnValue:String), B=Bonus)
   )
  ]}
 SetBattlePassButtonText {nodes [
   n1 = SetBattlePassButtonText!() -> (then, inText:Text) {is=title K2Node_FunctionEntry}
   > SetText!(
    execute
    self:Fort CTAButton Object Reference=Get LoudButton_BuySeasonPass() -> (LoudButton_BuySeasonPass:WBP UIKit Button Loud Object Reference)
    ButtonText:"Text (by ref)"=@n1
   ) -> (then)
  ]}
 SetComingSoonVisibility {nodes [
   n1 = SetComingSoonVisibility!() -> (then, InIsTimeLock:bool=~) {is=title K2Node_FunctionEntry}
   > SetVisibility(
    self=Get Tag_ComingSoon() -> (Tag_ComingSoon:Border Object Reference)
    InVisibility=Select(Option 0:ESlateVisibility Enum=Collapsed, Option 1:ESlateVisibility Enum=SelfHitTestInvisible, Index=@n1) -> (ReturnValue:ESlateVisibility Enum)
   )
  ]}
 SetCompactUIStyle {nodes [
   SetCompactUIStyle() {is=title K2Node_FunctionEntry}
   > Branch(Condition=IsUsingCompactUI())
   > Branch(Condition=CompareToScreenAspectRatio!(Comparison:EAspectRatioComparison Enum=GreaterThan, AspectRatio:EAspectRatio Enum=Ratio_4_3) -> (ReturnValue:bool))
   > SetPadding!(
    execute
    self:Vertical Box Slot Object Reference=[
     SlotAsVerticalBoxSlot(Widget=Get Text_LockCondition() -> (Text_LockCondition:Common Rich Text Block Object Reference))
     SlotAsVerticalBoxSlot(Widget=Get ProgressHorizontalBox() -> (ProgressHorizontalBox:Horizontal Box Object Reference))
     SlotAsVerticalBoxSlot(Widget=Get Button_Quests() -> (Button_Quests:WBP UIKit Button Regular Object Reference))
     SlotAsVerticalBoxSlot(Widget=Get Tag_Prerequisite() -> (Tag_Prerequisite:Border Object Reference))
     SlotAsVerticalBoxSlot(Widget=Get Tag_ComingSoon())
    ]
    InPadding_Left:float=0
    InPadding_Top:float=0
    InPadding_Right:float=0
    InPadding_Bottom:float=Get CompactUIPadding() -> (CompactUIPadding:double)
   ) -> (then)
  ]}
 SetLockText {nodes [
   SetLockText!() -> (then, InLockText:Text) {is=title K2Node_FunctionEntry}
   > n1 = Get()
   > Get!(execute) -> (then, else, SeasonPassVM:Season Pass VM Object Reference)
   n2 = GetRequiresAndNotOwnsPass!() -> (RequiresAndOwnsPass:bool) {is=title K2Node_CallFunction}
   n3 = IsTimeLocked!(self:Season Pass Element Base VM Object Reference=@n1) -> (ReturnValue:bool)
   n4 = GetOwnerCategory(self=@n1)
   n5 = Select(
    Option 0:Text=Get InLockText() -> (InLockText:Text)
    Option 1:Text=Get StatusText!(self:Season Pass Category VM Object Reference=@n4) -> (StatusText:Text)
    Index=AND Boolean(A=Get bIsLocked!(self:Season Pass Element Base VM Object Reference=@n4) -> (bIsLocked:bool), B=NOT Boolean(A=@n3))
   ) -> (ReturnValue:Text)
   n6 = TextIsEmpty(InText=@n5)
   > SetText!(
    execute
    self:Rich Text Block Object Reference=Get Text_LockCondition()
    InText:"Text (by ref)"=Select(
     Option 0:"Text (by ref)"=@n5
     Option 1:"Text (by ref)"=GetTextForSeasonPassRewards!() -> (Text:Text) {is=title K2Node_CallFunction}
     Index=AND Boolean(A=@n6, B=@n2)
    ) -> (ReturnValue:"Text (by ref)")
   ) -> (then)
  ]}
 SetLockTextVisibility {nodes [
   SetLockTextVisibility() {is=title K2Node_FunctionEntry}
   > n1 = Get!(execute) -> (then, else, SeasonPassRewardVM:Season Pass Reward VM Object Reference)
   n2 = Bitwise AND(
    A:int=Get PrerequisiteHeaders!(self:Season Pass Element Base VM Object Reference=@n1) -> (PrerequisiteHeaders:int=0)
    B:int="To Integer (Byte)"!(InByte:Byte=Literal enum ESeasonPassRewardHeader!(Enum:ESeasonPassRewardHeader Enum=SeasonPassRequired) -> (ReturnValue:ESeasonPassRewardHeader Enum) {is=title K2Node_EnumLiteral}) -> (ReturnValue:int=0)
   ) -> (ReturnValue:int=0) {is=title K2Node_CommutativeAssociativeBinaryOperator}
   n3 = NOT Boolean(A=AND Boolean(A=IsBonus!() -> (ReturnValue:bool) {is=title K2Node_CallFunction}, B="Not Equal (Integer)"(A=@n2)))
   n4 = Get Text_LockCondition()
   n5 = NOT Boolean(A=TextIsEmpty(InText=Get Text(self:Rich Text Block Object Reference=@n4)))
   > SetVisibility(
    self=Get LockedInfo() -> (LockedInfo:Horizontal Box Object Reference)
    InVisibility=Select(Option 0:ESlateVisibility Enum=Collapsed, Option 1:ESlateVisibility Enum=HitTestInvisible, Index=AND Boolean(A=@n3, B=@n5)) -> (ReturnValue:ESlateVisibility Enum)
   )
  ]}
 SetProgressInfo {nodes [
   n1 = SetProgressInfo!() -> (then, ProgressCount:int, ProgressTotal:int) {is=title K2Node_FunctionEntry}
   > Set Text(Text="To Text (Integer)"(Value=@n1), self=Get ProgressTextX() -> (ProgressTextX:Common Text Object Reference))
   > Set Text(
    Text=Format Text!(Format="LOCTEXT(\"/{CountForUnlock}\")", CountForUnlock:int=Get ProgressTotal() -> (ProgressTotal:int=0)) -> (Result)
    self=Get ProgressTextOf() -> (ProgressTextOf:Common Text Object Reference)
   )
   > SetProgress!(
    execute
    self:WBP UIKit Progress Bar Base Object Reference=[
     Get ProgressBar() -> (ProgressBar:WBP UIKit Progress Bar Object Reference)
     Get ProgressBar_Alter() -> (ProgressBar_Alter:WBP UIKit Progress Bar Object Reference)
    ]
    Progress:double=float / float(A:int=Get ProgressCount() -> (ProgressCount:int=0), B:int=Get ProgressTotal() -> (ProgressTotal=0))
   ) -> (then)
   > Set Visibility(Visibility=Visible, self=Get ProgressBarSwitcher() -> (ProgressBarSwitcher:Common Visibility Switcher Object Reference))
  ]}
 SetProgressbarVisibility {nodes [
   SetProgressbarVisibility() {is=title K2Node_FunctionEntry}
   > Get()
   > n1 = Sequence()
   n2 = Get SeasonPassRewardVM()
   n3 = Cast To AthenaSeasonItemEntryQuest!(Object=GetReward(self=@n2)) -> (AsQuest:Quest Object Reference, bSuccess:bool=~) {is=title K2Node_DynamicCast}
   n4 = IsIncludedWithPass(self=@n2)
   > Set DisplayProgressBar(DisplayProgressBar:bool=AND Boolean!(
     A=Get bDoesRequireProgressForUnlock!(self:Season Pass Reward VM Object Reference=@n2) -> (bDoesRequireProgressForUnlock:bool)
     B=NOT Boolean(A=@n4)
     C:bool=NOT Boolean(A=@n3.bSuccess)
    ) -> (ReturnValue)) -> (Output_Get:bool)
   SetVisibility(
    execute=@n1.then_1
    self=Get ProgressHorizontalBox()
    InVisibility=Select(Option 0=Collapsed, Option 1=SelfHitTestInvisible, Index=Get DisplayProgressBar() -> (DisplayProgressBar:bool))
   )
   n5 = Get SeasonPassRewardVM()
   > SetProgressInfo!(
    execute
    self
    ProgressCount:int=Get ProgressForUnlock!(self:Season Pass Reward VM Object Reference=@n5) -> (ProgressForUnlock:int=0)
    ProgressTotal:int=Get ProgressCountForUnlock!(self:Season Pass Reward VM Object Reference=@n5) -> (ProgressCountForUnlock:int=0)
   ) -> (then) {is=title K2Node_CallFunction}
  ]}
 SetQuestsButtonVisibility {nodes [
   SetQuestsButtonVisibility() {is=title K2Node_FunctionEntry}
   > n1 = Get()
   n2 = GetReward(self=@n1)
   > SetVisibility(
    self=Get Button_Quests()
    InVisibility="To Visibility (Boolean)"!(
     bIsVisible:bool=Cast To AthenaSeasonItemEntryQuest(Object=@n2) -> (bSuccess=~).bSuccess
     TrueVisibility:ESlateVisibility Enum=SelfHitTestInvisible
     FalseVisibility:ESlateVisibility Enum=Collapsed
    ) -> (ReturnValue:ESlateVisibility Enum=Visible)
   )
  ]}
 SetSeasonPassRequirementsVisibility {nodes [
   n1 = SetSeasonPassRequirementsVisibility!() -> (then, bRewardRequeresBattlePass:bool=~) {is=title K2Node_FunctionEntry}
   > n2 = Get!(execute) -> (then, else, SeasonPassVM:Season Pass VM Object Reference)
   n3 = NOT Boolean(A=Get bDoesOwnsPass(self=@n2))
   > SetVisibility(
    self=[Get Tag_Prerequisite(), Get LoudButton_BuySeasonPass()]
    InVisibility=Select(Option 0=Collapsed, Option 1=SelfHitTestInvisible, Index=AND Boolean!(A=@n3, B=@n1) -> (ReturnValue))
   )
  ]}
 SetSeasonRewardVM {nodes [
   SetSeasonRewardVM!() -> (then, SeasonPassRewardVM:Season Pass Reward VM Object Reference) {is=title K2Node_FunctionEntry}
   > UpdateVisibility()
   > SetProgressbarVisibility() {is=title K2Node_CallFunction}
   > SetQuestsButtonVisibility() {is=title K2Node_CallFunction}
   > n1 = Get!(execute) -> (then, else, SeasonPassRewardVM:Season Pass Reward VM Object Reference)
   > SetLockText!(execute, self, InLockText:Text=Get LockStatusText!(self:Season Pass Element Base VM Object Reference=@n1) -> (LockStatusText:Text)) -> (then) {is=title K2Node_CallFunction}
   > SetLockTextVisibility() {is=title K2Node_CallFunction}
  ]}
 UpdateProgressbarColors {nodes [
   n1 = UpdateProgressbarColors!() -> (then, bShouldDisplayAlternative:bool=~) {is=title K2Node_FunctionEntry}
   > SetActiveWidget(
    self=Get ProgressBarSwitcher()
    Widget=Select(
     Option 0:WBP UIKit Progress Bar Object Reference=Get ProgressBar()
     Option 1:WBP UIKit Progress Bar Object Reference=Get ProgressBar_Alter()
     Index=@n1
    ) -> (ReturnValue:WBP UIKit Progress Bar Object Reference)
   )
   > SetColorAndOpacity!(
    execute
    self=Get ProgressTextX()
    InColorAndOpacity_SpecifiedColor:Linear Color Structure=SelectColor(A=(R=1,G=1,B=1,A=1), B=(R=0.9301,G=1,B=0.0103,A=1), bPickA=@n1)
    InColorAndOpacity_ColorUseRule:ESlateColorStylingMode Enum
   ) -> (then)
  ]}
 UpdateVisibility {nodes [
   UpdateVisibility() {is=title K2Node_FunctionEntry}
   > n1 = Get()
   > SetVisibility(InVisibility=Select(Option 0:ESlateVisibility Enum=Collapsed, Option 1:ESlateVisibility Enum=SelfHitTestInvisible, Index=IsValid(Object=@n1)) -> (ReturnValue:ESlateVisibility Enum))
  ]}
]
variables [
 IsBonusCategory {type=boolean}
 NumberOfTotalOwnedRewardsInPass {type=integer}
 NumberOfTotalRewardsInPass {type=integer}
]
widgets {
 info {inheritedWidgetCount=0, parentClass=&UserWidget, rootWidgetClass=&GridPanel, widgetCount=21}
 widgets [
  Root {widgetClassPath=&GridPanel, widget=Root, properties {columnFill [1]}}
  ClaimedStatusSwitcher {
   widgetClassPath=&CommonVisibilitySwitcher
   bIsVariable=true
   parent=Root
   widget=ClaimedStatusSwitcher
   properties {activeWidgetIndex=1}
   slot {class=GridSlot, properties {horizontalAlignment=HAlign_Center, padding=<ltrb 0 0 0 12>}}
  }
  Grid_ClaimedWithBP {
   widgetClassPath=&GridPanel
   bIsVariable=true
   parent=ClaimedStatusSwitcher
   widget=Grid_ClaimedWithBP
   slot {class=CommonVisibilitySwitcherSlot}
  }
  SeasonPassBadgeIcon {
   widgetClassPath=/SeasonPass/Widgets/SeasonPassBadgeIcon
   bIsVariable=true
   parent=Grid_ClaimedWithBP
   widget=SeasonPassBadgeIcon
   properties {size=<108 108>}
   slot {class=GridSlot, properties {horizontalAlignment=HAlign_Center}}
  }
  Text_LockCondition_1 {
   widgetClassPath=&CommonTextBlock
   bIsVariable=true
   parent=Grid_ClaimedWithBP
   widget=Text_LockCondition_1
   properties {
    bAutoCollapseWithEmptyText=true
    text="Claimed with Battle Pass!"
    colorAndOpacity {colorUseRule=UseColor_Specified, specifiedColor=<rgba 1 1 1 0.65>}
    font {
     bForceMonospaced=false
     bMaterialIsStencil=false
     fontMaterial=null
     fontObject=&HeadingNow
     letterSpacing=0
     monospacedWidth=1
     size=24
     skewAmount=0
     typefaceFontName=86 Bold
     outlineSettings {
      bApplyOutlineToDropShadows=false
      bMiteredCorners=false
      bSeparateFillAlpha=false
      outlineMaterial=null
      outlineSize=0
      outlineColor=<rgba 0 0 0 1>
     }
    }
    shadowOffset=<0 0>
   }
   slot {class=GridSlot, properties {row=1, padding=<ltrb 0 12 0 0>}}
  }
  Claimed_Text {
   widgetClassPath=&CommonTextBlock
   bIsVariable=true
   parent=ClaimedStatusSwitcher
   widget=Claimed_Text
   properties {
    justification=Center
    text=CLAIMED
    colorAndOpacity {colorUseRule=UseColor_Specified, specifiedColor=<rgba 0.0395 1 0.2307 1>}
    font {
     bForceMonospaced=false
     bMaterialIsStencil=false
     fontMaterial=null
     fontObject=&HeadingNow
     letterSpacing=0
     monospacedWidth=1
     size=22.5
     skewAmount=0
     typefaceFontName=86 Bold Italic
     outlineSettings {
      bApplyOutlineToDropShadows=false
      bMiteredCorners=false
      bSeparateFillAlpha=false
      outlineMaterial=null
      outlineSize=0
      outlineColor=<rgba 0 0 0 1>
     }
    }
    shadowOffset=<0 0>
   }
   slot {class=CommonVisibilitySwitcherSlot, properties {verticalAlignment=VAlign_Center}}
  }
  Tag_Prerequisite {
   widgetClassPath=&Border
   bIsVariable=true
   parent=Root
   widget=Tag_Prerequisite
   properties {
    horizontalAlignment=HAlign_Center
    background {
     bIsDynamicallyLoaded=false
     drawAs=Image
     imageType=FullColor
     mirroring=NoMirror
     resourceName=None
     resourceObject=/FortUILibrary/Systems/Badge/Materials/M_UIKit_Badge_Background
     tiling=NoTile
     imageSize=<32 32>
     margin=<ltrb 0 0 0 0>
     outlineSettings {
      bUseBrushTransparency=false
      roundingType=HalfHeightRadius
      width=0
      color {colorUseRule=UseColor_Specified, specifiedColor=<rgba 0 0 0 0>}
      cornerRadii=<0 0 0 0>
     }
     tintColor {colorUseRule=UseColor_Specified, specifiedColor=<rgba 1 0.063 0.3372 1>}
     uVRegion {bIsValid=false, max=<0 0>, min=<0 0>}
    }
    padding=<ltrb 8 2 8 2>
   }
   slot {class=GridSlot, properties {row=1, padding=<ltrb 0 0 0 16>}}
  }
  Text_Prequisite {
   widgetClassPath=&CommonTextBlock
   bIsVariable=true
   parent=Tag_Prerequisite
   widget=Text_Prequisite
   properties {
    justification=Center
    mobileFontSizeMultiplier=1.14
    scrollStyle=&TextScrollStyle-Base_C
    text=Battle pass required
    textTransformPolicy=ToUpper
    font {
     bForceMonospaced=false
     bMaterialIsStencil=false
     fontMaterial=null
     fontObject=&HeadingNow
     letterSpacing=0
     monospacedWidth=1
     size=15.75
     skewAmount=0
     typefaceFontName=76 Bold Italic
     outlineSettings {
      bApplyOutlineToDropShadows=false
      bMiteredCorners=false
      bSeparateFillAlpha=false
      outlineMaterial=null
      outlineSize=0
      outlineColor=<rgba 0 0 0 1>
     }
    }
    shadowOffset=<0 0>
   }
   slot {class=BorderSlot, properties {horizontalAlignment=HAlign_Center, padding=<ltrb 8 2 8 2>}}
  }
  Tag_ComingSoon {
   widgetClassPath=&Border
   bIsVariable=true
   parent=Root
   widget=Tag_ComingSoon
   properties {
    horizontalAlignment=HAlign_Center
    background {
     bIsDynamicallyLoaded=false
     drawAs=Image
     imageType=FullColor
     mirroring=NoMirror
     resourceName=None
     resourceObject=&M_UIKit_Badge_Background
     tiling=NoTile
     imageSize=<32 32>
     margin=<ltrb 0 0 0 0>
     outlineSettings {
      bUseBrushTransparency=false
      roundingType=HalfHeightRadius
      width=0
      color {colorUseRule=UseColor_Specified, specifiedColor=<rgba 0 0 0 0>}
      cornerRadii=<0 0 0 0>
     }
     tintColor {colorUseRule=UseColor_Specified, specifiedColor=<rgba 1 0.063 0.3372 1>}
     uVRegion {bIsValid=false, max=<0 0>, min=<0 0>}
    }
    padding=<ltrb 8 2 8 2>
   }
   slot {class=GridSlot, properties {row=2, padding=<ltrb 0 0 0 16>}}
  }
  Text_ComingSoon {
   widgetClassPath=&CommonTextBlock
   parent=Tag_ComingSoon
   widget=Text_ComingSoon
   properties {
    justification=Center
    mobileFontSizeMultiplier=1.14
    scrollStyle=&TextScrollStyle-Base_C
    text=Coming Soon
    textTransformPolicy=ToUpper
    font {
     bForceMonospaced=false
     bMaterialIsStencil=false
     fontMaterial=null
     fontObject=&HeadingNow
     letterSpacing=0
     monospacedWidth=1
     size=15.75
     skewAmount=0
     typefaceFontName=76 Bold Italic
     outlineSettings {
      bApplyOutlineToDropShadows=false
      bMiteredCorners=false
      bSeparateFillAlpha=false
      outlineMaterial=null
      outlineSize=0
      outlineColor=<rgba 0 0 0 1>
     }
    }
    shadowOffset=<0 0>
   }
   slot {class=BorderSlot, properties {horizontalAlignment=HAlign_Center, padding=<ltrb 8 2 8 2>}}
  }
  LockedInfo {
   widgetClassPath=&HorizontalBox
   bIsVariable=true
   parent=Root
   widget=LockedInfo
   slot {class=GridSlot, properties {row=3, padding=<ltrb 0 0 0 16>}}
  }
  LockedIndicator {
   widgetClassPath=/FortUILibrary/UIKit/StatusIndicator/WBP_UIKit_StatusIndicator
   bIsVariable=true
   parent=LockedInfo
   widget=LockedIndicator
   properties {type=Utility, icon=/FortUILibrary/Textures/MaterialInstances/MI_UIKit_Icon_Locked, iconSize=Compact, visibility=Collapsed}
   slot {class=HorizontalBoxSlot, properties {verticalAlignment=VAlign_Top, padding=<ltrb 0 0 4 0>}}
  }
  Text_LockCondition {
   widgetClassPath=&CommonRichTextBlock
   bIsVariable=true
   parent=LockedInfo
   widget=Text_LockCondition
   properties {
    autoWrapText=true
    bIsScrollingEnabled=false
    bOverrideDefaultStyle=true
    lineHeightPercentage=1.3
    mobileTextBlockScale=1.1
    text="[PH] Lock condition... <Strong>Text</> [PH] Lock condition.."
    textStyleSet=/SeasonPass/Widgets/Reboot/Data/DT_UI_BP_RichTextStyles
    defaultTextStyleOverride {
     overflowPolicy=Clip
     transformPolicy=None
     colorAndOpacity {colorUseRule=UseColor_Specified, specifiedColor=<rgba 1 1 1 0.75>}
     font {
      bForceMonospaced=false
      bMaterialIsStencil=false
      fontMaterial=null
      fontObject=&HeadingNow
      letterSpacing=0
      monospacedWidth=1
      size=15.75
      skewAmount=0
      typefaceFontName=74 Regular
      outlineSettings {
       bApplyOutlineToDropShadows=false
       bMiteredCorners=false
       bSeparateFillAlpha=false
       outlineMaterial=null
       outlineSize=0
       outlineColor=<rgba 0 0 0 1>
      }
     }
     highlightColor {colorUseRule=UseColor_Specified, specifiedColor=<rgba 0 0 0 1>}
     highlightShape {
      bIsDynamicallyLoaded=false
      drawAs=Image
      imageType=NoImage
      mirroring=NoMirror
      resourceName=None
      resourceObject=null
      tiling=NoTile
      imageSize=<32 32>
      margin=<ltrb 0 0 0 0>
      outlineSettings {
       bUseBrushTransparency=false
       roundingType=HalfHeightRadius
       width=0
       color {colorUseRule=UseColor_Specified, specifiedColor=<rgba 0 0 0 0>}
       cornerRadii=<0 0 0 0>
      }
      tintColor {colorUseRule=UseColor_Specified, specifiedColor=<rgba 1 1 1 1>}
      uVRegion {bIsValid=false, max=<0 0>, min=<0 0>}
     }
     selectedBackgroundColor {colorUseRule=UseColor_Foreground, specifiedColor=<rgba 1 0 1 1>}
     shadowColorAndOpacity=<rgba 0 0 0 1>
     shadowOffset=<0 0>
     strikeBrush {
      bIsDynamicallyLoaded=false
      drawAs=Image
      imageType=NoImage
      mirroring=NoMirror
      resourceName=None
      resourceObject=null
      tiling=NoTile
      imageSize=<32 32>
      margin=<ltrb 0 0 0 0>
      outlineSettings {
       bUseBrushTransparency=false
       roundingType=HalfHeightRadius
       width=0
       color {colorUseRule=UseColor_Specified, specifiedColor=<rgba 0 0 0 0>}
       cornerRadii=<0 0 0 0>
      }
      tintColor {colorUseRule=UseColor_Specified, specifiedColor=<rgba 1 1 1 1>}
      uVRegion {bIsValid=false, max=<0 0>, min=<0 0>}
     }
     underlineBrush {
      bIsDynamicallyLoaded=false
      drawAs=Image
      imageType=NoImage
      mirroring=NoMirror
      resourceName=None
      resourceObject=null
      tiling=NoTile
      imageSize=<32 32>
      margin=<ltrb 0 0 0 0>
      outlineSettings {
       bUseBrushTransparency=false
       roundingType=HalfHeightRadius
       width=0
       color {colorUseRule=UseColor_Specified, specifiedColor=<rgba 0 0 0 0>}
       cornerRadii=<0 0 0 0>
      }
      tintColor {colorUseRule=UseColor_Specified, specifiedColor=<rgba 1 1 1 1>}
      uVRegion {bIsValid=false, max=<0 0>, min=<0 0>}
     }
    }
   }
   slot {class=HorizontalBoxSlot, properties {size {sizeRule=Fill, value=1}}}
  }
  ProgressHorizontalBox {
   widgetClassPath=&HorizontalBox
   bIsVariable=true
   parent=Root
   widget=ProgressHorizontalBox
   properties {visibility=Collapsed}
   slot {class=GridSlot, properties {row=4, verticalAlignment=VAlign_Bottom, padding=<ltrb 0 0 0 16>}}
  }
  ProgressBarSwitcher {
   widgetClassPath=&CommonVisibilitySwitcher
   bIsVariable=true
   parent=ProgressHorizontalBox
   widget=ProgressBarSwitcher
   slot {class=HorizontalBoxSlot, properties {verticalAlignment=VAlign_Center, padding=<ltrb 0 0 16 0>, size {sizeRule=Fill, value=1}}}
  }
  ProgressBar {
   widgetClassPath=/FortUILibrary/UIKit/ProgressBar/WBP_UIKit_ProgressBar
   bIsVariable=true
   parent=ProgressBarSwitcher
   widget=ProgressBar
   properties {
    height=16
    linearMaterial=/SeasonPass/Materials/ProgressBar/MI_UI_NexusProgerssBar
    outline_PX=6
    progressValue=0.5
    useCircular=false
    useOutline=true
    emptyColor=<rgba 0.3573 0.3573 0.3573 0.305>
   }
   slot {class=CommonVisibilitySwitcherSlot}
  }
  ProgressBar_Alter {
   widgetClassPath=&WBP_UIKit_ProgressBar
   bIsVariable=true
   parent=ProgressBarSwitcher
   widget=ProgressBar_Alter
   properties {
    height=16
    linearMaterial=/SeasonPass/Materials/ProgressBar/MI_UI_NexusProgerssBar_Exp
    outline_PX=6
    progressValue=0.5
    useCircular=false
    useOutline=true
    emptyColor=<rgba 0.3573 0.3573 0.3573 0.305>
   }
   slot {class=CommonVisibilitySwitcherSlot}
  }
  ProgressTextX {
   widgetClassPath=&CommonTextBlock
   bIsVariable=true
   parent=ProgressHorizontalBox
   widget=ProgressTextX
   properties {
    text=0
    colorAndOpacity {colorUseRule=UseColor_Specified, specifiedColor=<rgba 0.9301 1 0.0103 1>}
    font {
     bForceMonospaced=false
     bMaterialIsStencil=false
     fontMaterial=null
     fontObject=&NotoSans
     letterSpacing=0
     monospacedWidth=1
     size=24
     skewAmount=0
     typefaceFontName=NotoSans-Regular
     outlineSettings {
      bApplyOutlineToDropShadows=false
      bMiteredCorners=false
      bSeparateFillAlpha=false
      outlineMaterial=null
      outlineSize=0
      outlineColor=<rgba 0 0 0 1>
     }
    }
    shadowOffset=<0 0>
   }
   slot {class=HorizontalBoxSlot}
  }
  ProgressTextOf {
   widgetClassPath=&CommonTextBlock
   bIsVariable=true
   parent=ProgressHorizontalBox
   widget=ProgressTextOf
   properties {
    text=/100
    font {
     bForceMonospaced=false
     bMaterialIsStencil=false
     fontMaterial=null
     fontObject=&NotoSans
     letterSpacing=0
     monospacedWidth=1
     size=24
     skewAmount=0
     typefaceFontName=NotoSans-Regular
     outlineSettings {
      bApplyOutlineToDropShadows=false
      bMiteredCorners=false
      bSeparateFillAlpha=false
      outlineMaterial=null
      outlineSize=0
      outlineColor=<rgba 0 0 0 1>
     }
    }
    shadowOffset=<0 0>
   }
   slot {class=HorizontalBoxSlot}
  }
  Button_Quests {
   widgetClassPath=&WBP_UIKit_Button_Regular
   bIsVariable=true
   parent=Root
   widget=Button_Quests
   properties {
    bHideInputActionWithKeyboard=true
    bIsFocusable=false
    hoveredSound=/Game/Sounds/Fort_UI_Sounds/FrontEnd/BattlePass/BattlePass_Button_Hover_Tile_Default_Cue
    mobileFontSizeMultiplierOverride=1.33
    overrideDefaultHeight=true
    overrideDefaultMobileFontSizeMultiplier=true
    pressedSound=/Game/Sounds/Fort_UI_Sounds/FrontEnd/BattlePass/BattlePass_Button_Click_Whoosh_Cue
    text=Go to Quests
    triggeringInputAction {dataTable=&InputActionDataTable, rowName=GenericFaceButtonTop}
   }
   slot {class=GridSlot, properties {row=6, verticalAlignment=VAlign_Top, padding=<ltrb 0 0 0 16>}}
  }
  LoudButton_BuySeasonPass {
   widgetClassPath=/FortUILibrary/UIKit/Buttons/WBP_UIKit_Button_Loud
   bIsVariable=true
   parent=Root
   widget=LoudButton_BuySeasonPass
   properties {
    bHideInputActionWithKeyboard=true
    bIsFocusable=false
    hoveredSound=&BattlePass_Button_Hover_Tile_Default_Cue
    mobileFontSizeMultiplierOverride=1.33
    mobileMinHeight=80
    overrideDefaultHeight=true
    overrideDefaultMobileFontSizeMultiplier=true
    pressedSound=/Game/Sounds/Fort_UI_Sounds/FrontEnd/BattlePass/BattlePass_Button_Confirm_Buy_BattlePass_Cue
    text="[PH] GET THE SEASON PASS"
    triggeringInputAction {dataTable=/SeasonPass/Data/SeasonPassInputActionDataTable, rowName=PurchaseBattlePass_S23}
   }
   slot {class=GridSlot, properties {row=7}}
  }
 ]
}
